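# Unlit translucent energy sphere material for mesh particles in the GroundPortal FX set.
# Two-sided facing-based rim and center glows plus panning electric noise, faded at both V ends, form a mask driving color and brightness.
# Color defaults to MF_BlasterCommon's DarkColor at both lerp ends, with a switch to a two-color gradient.
# Dynamic parameters vary noise and glow; opacity math combines particle alpha, noise-based dissolve, inverse Fresnel and depth fade.
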Material /PistolCanary/Assets/FX/GroundPortal/M_EnergySphere
graph {
 properties {
  bUsedWithMeshParticles=true
  bUsedWithNanite=true
  bUsedWithNiagaraMeshParticles=true
  blendMode=BLEND_Translucent
  pixelDepthOffsetMode=PDOM_Legacy
  shadingModel=MSM_Unlit
  usedShadingModels=Unlit
 }
 nodes [
  n1 = MaterialFunctionCall!(TimeScale, ElementColor, HeadEnhanceScale) {materialFunction=/Game/Effects/Fort_Effects/Materials/_Functions/MF_BlasterCommon}
  n2 = DotProduct(A=Multiply(A=VertexNormalWS(), B=TwoSidedSign()), B=CameraVectorWS())
  n3 = ScalarParameter() {defaultValue=5, parameterName=FresExp_Outside}
  n4 = Multiply(A=@n3, B=DynamicParameter() {parameterIndex=1, paramNames [FresExpInsideMult, FresExpOutsideMult, InsideMult, NoiseMult]}.FresExpOutsideMult)
  n5 = ScalarParameter() {parameterName=UseAdditionalDynamicParams}
  n6 = Multiply(
   A=Power(Base=OneMinus(None=@n2), Exp=LinearInterpolate(A=@n3, B=@n4, Alpha=@n5))
   B=ScalarParameter() {defaultValue=0.5, parameterName=OutsideMult}
  )
  n7 = ScalarParameter() {defaultValue=10, parameterName=FresExp_Inside}
  n8 = Multiply(A=@n7, B=DynamicParameter() {parameterIndex=1, paramNames [FresExpInsideMult, FresExpOutsideMult, InsideMult, NoiseMult]}.FresExpInsideMult)
  n9 = ScalarParameter() {parameterName=UseAdditionalDynamicParams}
  n10 = ScalarParameter() {defaultValue=0.1, parameterName=InsideMult}
  n11 = DynamicParameter() {parameterIndex=1, paramNames [FresExpInsideMult, FresExpOutsideMult, InsideMult, NoiseMult]}
  n12 = Multiply(
   A=Power(Base=@n2, Exp=LinearInterpolate(A=@n7, B=@n8, Alpha=@n9))
   B=LinearInterpolate(A=@n10, B=Multiply(A=@n10, B=@n11.InsideMult), Alpha=ScalarParameter() {parameterName=UseAdditionalDynamicParams})
  )
  n13 = LinearInterpolate(Alpha=ComponentMask(None=TextureCoordinate()) {g=true}) {constA=-1}
  n14 = Constant3Vector() {constant=<rgba 0 0 1 0>}
  n15 = OneMinus(None=Abs(None=DotProduct(A=@n13, B=@n14)))
  n16 = ScalarParameter() {defaultValue=2, parameterName=TopfadePower}
  n17 = DynamicParameter() {paramNames [randomOffset,NoiseScaleOT,Dissolve,Power]}
  n18 = Add(A=TextureCoordinate(), B=@n17.randomOffset)
  n19 = AppendVector(A=ScalarParameter() {defaultValue=1, parameterName=NoiseTileX}, B=ScalarParameter() {defaultValue=1, parameterName=NoiseTileY})
  n20 = TextureSample(UVs=Panner(Coordinate=Multiply(A=@n18, B=@n19)) {speedX=0.5, speedY=0.5}) {samplerType=SAMPLERTYPE_LinearColor, texture=/Game/Characters/Player/Male/Medium/Bodies/M_MED_SRDrift/Textures/T_ElectricNoise}
  n21 = ScalarParameter() {defaultValue=2, parameterName=NoiseOpacMult}
  n22 = DynamicParameter() {parameterIndex=1, paramNames [FresExpInsideMult, FresExpOutsideMult, InsideMult, NoiseMult]}
  n23 = Multiply(
   A=Multiply(A=Power(Base=@n15, Exp=@n16), B=Power(Base=@n20.B, Exp=@n17.Power) {constExponent=0.5})
   B=LinearInterpolate(A=@n21, B=Multiply(A=@n21, B=@n22.NoiseMult), Alpha=ScalarParameter() {parameterName=UseAdditionalDynamicParams})
  )
  n24 = Saturate(None=Add(A=Add(A=@n6, B=@n12), B=@n23))
  n25 = StaticSwitchParameter(
   True=LinearInterpolate(A=@n1.DarkColor, B=@n1.DarkColor, Alpha=@n24)
   False=LinearInterpolate(
    A=VectorParameter() {parameterName=Color, defaultValue=<rgba 0.1 3 16 1>}.RGB
    B=VectorParameter() {parameterName=ColorB, defaultValue=<rgba 0.9583 0.2576 0.229 0>}.RGB
    Alpha=@n24
   )
  ) {defaultValue=true, parameterName=UseBlasterColorPicker}
  n26 = LinearInterpolate(A=ScalarParameter() {parameterName=ColorBrightnessLow}, B=ScalarParameter() {defaultValue=2, parameterName=ColorBrightnessHigh}, Alpha=@n24)
  n27 = ParticleColor()
  MaterialFunctionCall(Emissive=Multiply(A=Multiply(A=@n25, B=@n26), B=@n27.RGB)) {materialFunction=&MF_FX_EmissiveAdjustments}
  n28 = Multiply(A=@n27.A, B=@n24)
  n29 = ScalarParameter() {defaultValue=0.1, desc=A2B, parameterName=ClipRange}
  n30 = LinearInterpolate(
   A=Multiply(A=@n29) {constB=-1}
   Alpha=OneMinus(None=DynamicParameter() {parameterIndex=3, paramNames [Param1,Param2,Dissolve,Param4]}.Dissolve) {desc=Inverse Alpha}
  )
  n31 = Saturate(None=Divide(A=Subtract(A=@n23, B=@n30), B=@n29))
  n32 = Fresnel() {baseReflectFraction=0, exponent=3}
  Multiply(A=Multiply(A=Multiply(A=@n28, B=@n31), B=OneMinus(None=@n32)), B=DepthFade() {fadeDistanceDefault=512})
  OneMinus(None=@n24) {desc=Mask}
  ScalarParameter() {defaultValue=0.5, desc=A, parameterName=Cliptest, sliderMax=1}
 ]
}
