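# Translucent two-sided light-cone material in the PistolCanary GroundPortal effects, flagged for sprite and mesh particles.
# A switch picks particle color or LightConeColor, with the RGB going through an emissive adjustment function.
# Opacity math combines ramp texture, alpha, view-angle fade and brightness, then optional depth fade (skipped on ES2/ES3_1) and near-camera fade, masked by a smoothstepped Fresnel.
# A DanceClub switch selects MF_Sneezy_DanceClub or a constant.
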Material /PistolCanary/Assets/FX/GroundPortal/M_PistolCanary_LightCone
graph {
 properties {
  bUsedWithMeshParticles=true
  bUsedWithNiagaraMeshParticles=true
  bUsedWithParticleSprites=true
  blendMode=BLEND_Translucent
  pixelDepthOffsetMode=PDOM_Legacy
  previewMesh=/Game/Effects/Fort_Effects/Mesh/Rays/S_Billboard_LightCone
  twoSided=true
  usedShadingModels=Default Lit
 }
 nodes [
  n1 = ParticleColor()
  n2 = VectorParameter() {parameterName=LightConeColor, defaultValue=<rgba 1 0.9051 0.6744 0.5>}
  n3 = StaticSwitchParameter(True=@n1.RGB, False=@n2.RGB) {parameterName=UseParticleColor}
  MaterialFunctionCall(Emissive=@n3) {materialFunction=&MF_FX_EmissiveAdjustments}
  n4 = MaterialFunctionCall(Distance=ScalarParameter() {parameterName=NearFadeDistance}) {materialFunction=&Near_Camera_Fade}
  n5 = TextureSampleParameter2D() {parameterName=Main Tex, texture=/Game/Effects/Fort_Effects/Textures/Ramps/T_Ramp_Vert_Soft_Fade}
  n6 = StaticSwitchParameter(True=@n1.A, False=@n2.A) {parameterName=UseParticleColor}
  n7 = ScalarParameter() {defaultValue=4, parameterName=FadeSensitivity}
  n8 = Multiply(
   A=Multiply(
    A=Multiply(A=@n5.G, B=@n6)
    B=MaterialFunctionCall("FadeSensitivity (optional)"=@n7) {materialFunction=&FadeBasedOnViewAngle}.Result
   )
   B=ScalarParameter() {defaultValue=0.25, parameterName=Brightness}
  )
  n9 = Multiply(A=@n8, B=DepthFade(Opacity=Constant() {r=1}, FadeDistance=ScalarParameter() {defaultValue=15, parameterName=DepthFade-Distance}))
  n10 = StaticSwitchParameter(True=FeatureLevelSwitch(Default=@n9, ES2=@n8, ES3_1=@n8, SM4_REMOVED=@n9), False=@n8) {parameterName=UseDepthFade}
  n11 = Fresnel(ExponentIn=ScalarParameter() {defaultValue=0.01, parameterName=CenterHole}) {exponent=10}
  Multiply(A=StaticSwitchParameter(True=Multiply(A=@n4.Result, B=@n10), False=@n10) {parameterName=UseNearFade}, B=Clamp(None=SmoothStep(Value=@n11)))
  Reroute(None=@n3)
  StaticSwitchParameter(True=MaterialFunctionCall() {materialFunction=/Game/Environments/Sets/DigSite/Materials/MF_Sneezy_DanceClub}.Result, False=Constant()) {parameterName=DanceClub}
 ]
}
