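# Unlit translucent particle material for the ground portal's glowing ring, usable on sprites and Niagara mesh particles.
# Particle color feeds the emissive adjustment function.
# For opacity, smoke-disk textures in radial coordinates, panned, rotated and noise-distorted via dynamic parameters and UseDistortion, are smoothstepped.
# Two sphere masks cut that into a ring, then particle alpha, depth fade and view-angle fade apply.
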Material /PistolCanary/Assets/FX/GroundPortal/M_Portal_RingGlow_01
graph {
 properties {
  bUsedWithNiagaraMeshParticles=true
  bUsedWithNiagaraSprites=true
  bUsedWithParticleSprites=true
  blendMode=BLEND_Translucent
  pixelDepthOffsetMode=PDOM_Legacy
  shadingModel=MSM_Unlit
  usedShadingModels=Unlit
 }
 nodes [
  n1 = ParticleColor()
  MaterialFunctionCall(Emissive=@n1.RGB) {materialFunction=&MF_FX_EmissiveAdjustments}
  n2 = Add(A=TextureCoordinate(), B=Constant2Vector() {g=-0.5, r=-0.5})
  n3 = Constant2Vector()
  n4 = Multiply(A=DynamicParameter() {defaultValue=<rgba -1 0 0 1>, paramNames [ExplosionPan, DistortAmount, RotateAmount, DistortPanSpeed]}.RotateAmount) {constB=-1}
  n5 = TextureCoordinate()
  n6 = Multiply(
   A=Constant2Vector() {g=1, r=1}
   B=DynamicParameter() {defaultValue=<rgba -1 0 0 1>, paramNames [ExplosionPan, DistortAmount, RotateAmount, DistortPanSpeed]}.DistortPanSpeed
  ) {constB=0.25}
  n7 = TextureSampleParameter2D(UVs=MaterialFunctionCall(UVs=Panner(Coordinate=@n5, Speed=@n6) {speedX=-1, speedY=1}, Texture Scale=Constant() {r=0.5}) {materialFunction=&ScaleUVsByCenter}.UVs) {parameterName=DistortTexture2, texture=/Game/Effects/Fort_Effects/Textures/T_Smooth_tile}
  n8 = Multiply(
   A=@n7.G
   B=DynamicParameter() {defaultValue=<rgba -1 0 0 1>, paramNames [ExplosionPan, DistortAmount, RotateAmount, DistortPanSpeed]}.DistortAmount
  ) {constB=0.25}
  n9 = Add(
   A=LinearInterpolate(
    A=@n2
    B=MaterialFunctionCall(UVs=@n2, Rotation Center=@n3, "Rotation Angle (0-1)"=@n4) {materialFunction=&CustomRotator}.Rotated Values
    Alpha=ScalarParameter() {parameterName=UseDistortion}
   ) {constB=0}
   B=LinearInterpolate(B=@n8, Alpha=ScalarParameter() {parameterName=UseDistortion}) {constB=0}
  )
  n10 = TextureSample(UVs=Panner(
    Coordinate=MaterialFunctionCall(Vector or UVs=@n9) {materialFunction=&VectorToRadialValue}.Radial Coordinates
    Time=DynamicParameter() {defaultValue=<rgba -1 0 0 1>, paramNames [ExplosionPan, DistortAmount, RotateAmount, DistortPanSpeed]}.ExplosionPan
   ) {speedY=-0.25}) {texture=/Game/Effects/Fort_Effects/Textures/Smoke/T_Smoke_Bomb_Disk2}
  n11 = DynamicParameter() {defaultValue=<rgba -1 0 0 1>, paramNames [ExplosionPan, DistortAmount, RotateAmount, DistortPanSpeed]}
  n12 = MaterialFunctionCall(
   UVs=Add(A=TextureCoordinate(), B=Constant2Vector() {g=-0.5, r=-0.5})
   Rotation Center=Constant2Vector()
   "Rotation Angle (0-1)"=DynamicParameter() {defaultValue=<rgba -1 0 0 1>, paramNames [ExplosionPan, DistortAmount, RotateAmount, DistortPanSpeed]}.RotateAmount
  ) {materialFunction=&CustomRotator}
  n13 = Panner(
   Coordinate=MaterialFunctionCall(Vector or UVs=Add(A=@n12.Rotated Values, B=@n8)) {materialFunction=&VectorToRadialValue}.Radial Coordinates
   Time=DynamicParameter() {defaultValue=<rgba -1 0 0 1>, paramNames [ExplosionPan, DistortAmount, RotateAmount, DistortPanSpeed]}.ExplosionPan
  ) {speedY=-0.25}
  n14 = Constant2Vector() {g=1, r=0.5}
  n15 = Power(Base=TextureSample(UVs=MaterialFunctionCall(UVs=@n13, Texture Scale=@n14) {materialFunction=&ScaleUVsByCenter}.UVs) {texture=&T_Smoke_Bomb_Disk2}.G) {constExponent=1}
  n16 = Constant() {r=1}
  n17 = LinearInterpolate(
   A=@n10.G
   B=Multiply(A=If(A=@n11.DistortAmount, "A > B"=@n15, "A == B"=@n16, "A < B"=@n16), B=@n10.G)
   Alpha=ScalarParameter() {parameterName=UseDistortion}
  )
  n18 = ScalarParameter() {defaultValue=0.1, parameterName=Min}
  n19 = ScalarParameter() {defaultValue=1, parameterName=Max}
  n20 = SphereMask(A=TextureCoordinate(), B=Constant2Vector() {g=0.5, r=0.5}, Radius=Constant() {r=0.2}, Hardness=Constant() {r=0.4})
  n21 = Multiply(
   A=Multiply(
    A=MaterialFunctionCall(Alpha=@n17, Min=@n18, Max=@n19) {materialFunction=&SmoothStep}.Result
    B=OneMinus(None=@n20)
   )
   B=SphereMask(A=TextureCoordinate(), B=Constant2Vector() {g=0.5, r=0.5}, Radius=Constant() {r=0.4}, Hardness=Constant() {r=-1})
  )
  Multiply(
   A=DepthFade(Opacity=Multiply(A=@n1.A, B=@n21), FadeDistance=ScalarParameter() {defaultValue=50, parameterName=DepthFade})
   B=MaterialFunctionCall("FadeSensitivity (optional)"=ScalarParameter() {parameterName="FadeSensitivity (optional) (S)"}) {materialFunction=&FadeBasedOnViewAngle}.Result
  )
  n22 = Add(A=TextureCoordinate(), B=Constant2Vector() {g=-0.5, r=-0.5})
  n23 = MaterialFunctionCall(
   UVs=MaterialFunctionCall(Vector or UVs=Rotator(Coordinate=@n22) {centerX=0, centerY=0, speed=4}) {materialFunction=&VectorToRadialValue}.Radial Coordinates
   Texture Scale=Constant() {r=0.5}
  ) {materialFunction=&ScaleUVsByCenter}
  Multiply(A=TextureSample(UVs=@n23.UVs) {texture=&T_Smooth_tile}.G, B=@n7.G) {constB=0.25}
  Panner(
   Coordinate=@n23.UVs
   Time=DynamicParameter() {defaultValue=<rgba -1 0 0 1>, paramNames [ExplosionPan, DistortAmount, RotateAmount, DistortPanSpeed]}.ExplosionPan
  ) {speedY=-0.25}
 ]
}
