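# Niagara system for the PistolCanary ground portal's charge effect, currently showing only pink-red lightning.
# Lightning_Primary001, the sole enabled emitter, scatters short-lived meshes across a flattened sphere at twice User.Lightning Spawn Rate, which defaults to zero.
# Multi-key color curves make bolts flicker, while dynamic material parameters animate them, scaled by User.WidthMult and User.ColorMult.
# Energy wave, glow, light, ring, trail and smoke emitters remain in the asset but are disabled.
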NiagaraSystem /PistolCanary/Assets/FX/GroundPortal/NS_PistolCanary_PortalCharge
dependencies {
 usedDataInterfaces [
  &NiagaraDataInterfaceCurve
  &NiagaraDataInterfaceColorCurve
  &NiagaraDataInterfaceArrayColor
  &NiagaraDataInterfaceVector2DCurve
  &NiagaraDataInterfaceSpriteRendererInfo
  &NiagaraDataInterfaceArrayFloat
 ]
 usedDynamicInputs [
  &Vector2DFromCurve
  &RandomRangeVector2D
  &Add_Float
 ]
 usedModules [
  &self:SystemSpawnScript.SystemScriptSource.SystemScriptGraph.NiagaraNodeAssignment_0.SetVariables_#1
  &SystemState
  &self:SystemSpawnScript.SystemScriptSource.SystemScriptGraph.NiagaraNodeAssignment_1.SetVariables_#2
  &self:SystemSpawnScript.SystemScriptSource.SystemScriptGraph.NiagaraNodeAssignment_2.SetVariables_#3
  &self:SystemSpawnScript.SystemScriptSource.SystemScriptGraph.NiagaraNodeAssignment_3.SetVariables_#4
  &self:SystemSpawnScript.SystemScriptSource.SystemScriptGraph.NiagaraNodeAssignment_4.SetVariables_#5
 ]
}
emitters {
 EnergyWave {
  bEnabled=false
  emitterUpdateScript [
   &EmitterState()
   &SpawnBurst_Instantaneous(Spawn Time=$System.SpawnTime, Use Spawn Probability=false, UseLoopCountLimit=false)
   &SpawnRate(SpawnRate=2, Use Spawn Probability=false)
  ]
  particleSpawnScript [
   &InitializeParticle(
    Color Maximum=<rgba 1 0 0.1263 1>
    Color Minimum=<rgba 0.7188 0 0.0998 1>
    Color Mode=::NewEnumerator2
    Lifetime=4.5
    Mesh Scale Max=<6 6 0.5>
    Mesh Scale Min=<2.25 3 0.5>
    Mesh Scale Mode=::NewEnumerator2
    Position Offset=<0 50 500>
    UsePositionOffset=true
    Write Material Random=true
   )
   &InitialMeshOrientation(Mesh Orientation Mode=::NewEnumerator2, Orientation Vector=<0 0 1>, Rotation=<0.25 0 0>, Use Rotation Vector=true)
  ]
  particleUpdateScript [
   &ParticleState()
   &UpdateMeshOrientation(Rotation Vector=<0 0 0>)
   &DynamicMaterialParameters(
    Index 0 Param 1=&FloatFromCurve(
     FloatCurve=DataInterface() {propertyValues {Curve {keys [{arriveTangent=-2.1015, interpMode=RCIM_Cubic, leaveTangent=-2.1015, tangentMode=RCTM_User, time=0, value=0.5}, <at 1 -0.15>]}}}
     Scale Curve=&RandomRangeFloat(Minimum=0.5, Maximum=1.5)
    )
    Index 0 Param 2=&FloatFromCurve(
     FloatCurve=DataInterface() {propertyValues {Curve {keys [{arriveTangent=-0.3464, interpMode=RCIM_Cubic, leaveTangent=-0.3464, tangentMode=RCTM_User, time=0, value=1}, <at 1 0.9>]}}}
     Scale Curve=&RandomRangeFloat(Minimum=0.5, Maximum=1)
    )
    Index 0 Param 3=&FloatFromCurve(
     FloatCurve=DataInterface() {propertyValues {Curve {keys [
         {interpMode=RCIM_Cubic, tangentMode=RCTM_Break, time=0, value=0}
         {arriveTangent=0.4478, interpMode=RCIM_Cubic, leaveTangent=0.4478, tangentMode=RCTM_User, time=0.15, value=1}
         {interpMode=RCIM_Cubic, tangentMode=RCTM_Break, time=1, value=1.2}
        ]}}}
     Scale Curve=&RandomRangeFloat(Minimum=0.5, Maximum=1)
    )
    Module.Param0WriteEnabled=true
    Module.Param1WriteEnabled=true
    Module.Param2WriteEnabled=true
    Module.Param3WriteEnabled=false
   )
   &ScaleColor(
    Scale Alpha=&FloatFromCurve(FloatCurve=DataInterface() {propertyValues {Curve {keys [{interpMode=RCIM_Cubic, tangentMode=RCTM_User, time=0.5, value=1}, {interpMode=RCIM_Cubic, tangentMode=RCTM_User, time=1, value=0}]}}})
    Scale RGB=&Multiply_VectorByFloat(
     Vector=&MakeVectorFromLinearColorRGB(Color=&ColorFromCurve(ColorCurve=DataInterface() {propertyValues {
         AlphaCurve {keys [<at 0 1 RCIM_Cubic>]}
         BlueCurve {keys [<at 0.1059 1 RCIM_Cubic>,<at 0.85 1 RCIM_Cubic>,<at 0.9998 1>]}
         GreenCurve {keys [<at 0.1059 0.833 RCIM_Cubic>, {arriveTangent=0.1515, interpMode=RCIM_Cubic, leaveTangent=0.1515, time=0.85, value=1}, <at 0.9998 0.9684>]}
         RedCurve {keys [<at 0.1059 0.3906 RCIM_Cubic>, {arriveTangent=0.6531, interpMode=RCIM_Cubic, leaveTangent=0.6531, time=0.85, value=1}, <at 0.9998 0.9745>]}
        }}))
     Float=1
    )
    ScaleA=true
    ScaleRGB=true
   )
   &ScaleMeshSize(Scale Factor=&VectorFromFloat(Value=&Multiply_Float(
      A=&FloatFromCurve(
       FloatCurve=DataInterface() {propertyValues {Curve {keys [
           {interpMode=RCIM_Cubic, leaveTangent=1.4504, tangentMode=RCTM_Break, time=0, value=0.2}
           {arriveTangent=2.3, interpMode=RCIM_Cubic, leaveTangent=2.3, time=0.1, value=1}
           {interpMode=RCIM_Cubic, tangentMode=RCTM_Break, time=1, value=2.5}
          ]}}}
       Scale Curve=&Multiply_Float(A=&RandomRangeFloat(Minimum=2, Maximum=3.5), B=1.5)
      )
      B=$System.PortalScale
     )))
  ]
  rendererClasses [&NiagaraMeshRendererProperties]
  renderers [{rendererClass=&NiagaraMeshRendererProperties, rendererIndex=0}]
 }
 Glow {
  bEnabled=false
  emitterUpdateScript [
   &EmitterState()
   &SpawnRate(SpawnRate=3, Use Spawn Probability=false)
  ]
  particleSpawnScript [
   &InitializeParticle(
    Color=<rgba 0.163 0.0234 0.3791 0.7507>
    Lifetime=2.7
    Sprite Size Mode=::NewEnumerator3
    Uniform Sprite Size=4000
    UsePositionOffset=false
    Write Material Random=true
   )
   &CameraOffset(Camera Offset Amount=30)
   &SpriteFacingAndAlignment(Sprite Facing=<0 0 1>, WriteAlignment=false, WriteFacing=true)
  ]
  particleUpdateScript [
   &ParticleState()
   &FN_Light_Attributes(
    Color=<rgba 1.4181 0 5 1>
    Create Color Attribute=true
    Create Exponent Attribute=true
    Create Position Attribute=true
    Create Radius Attribute=true
    Enable Particle Light Enable=true
    Exponent=2
    Light Enabled=true
    Radius=250
    Scale Color=&FloatFromCurve(
     FloatCurve=DataInterface() {propertyValues {Curve {keys [
         <at 0 0 RCIM_Cubic>
         {arriveTangentWeight=0.0581, interpMode=RCIM_Cubic, leaveTangentWeight=0.349, tangentMode=RCTM_User, time=0.0705, value=1}
         {arriveTangent=-1.0758, interpMode=RCIM_Cubic, leaveTangent=-1.0758, time=0.9036, value=0.8288}
         {arriveTangent=-3.8525, interpMode=RCIM_Cubic, leaveTangent=-3.8525, tangentMode=RCTM_User, time=1, value=0}
        ]}}}
     Scale Curve=&FN_FrontEndFloatModifier(Gameplay=4, Front End=1)
    )
   )
   &ScaleColor(
    Scale Alpha=&FloatFromCurve(FloatCurve=DataInterface() {propertyValues {Curve {keys [
         {arriveTangent=5.0727, interpMode=RCIM_Cubic, leaveTangent=5.0727, tangentMode=RCTM_User, time=0, value=0.1641}
         {
          arriveTangent=0.1472
          arriveTangentWeight=0.3592
          interpMode=RCIM_Cubic
          leaveTangent=0.1472
          leaveTangentWeight=0.3592
          tangentMode=RCTM_User
          tangentWeightMode=RCTWM_WeightedBoth
          time=0.5103
          value=1
         }
         {tangentMode=RCTM_User, time=1, value=0}
        ]}}})
    ScaleA=true
    ScaleRGB=true
   )
  ]
  rendererClasses [&NiagaraSpriteRendererProperties, &NiagaraLightRendererProperties]
  renderers [
   {rendererClass=&NiagaraSpriteRendererProperties}
   {rendererClass=&NiagaraLightRendererProperties, rendererIndex=1}
  ]
 }
 Lightning_Primary001 {
  emitterUpdateScript [
   &EmitterState(Life Cycle Mode=::NewEnumerator1, Loop Duration=1, UseLoopDelay=false)
   &SpawnBurst_Instantaneous(Spawn Count=10, Use Spawn Probability=false, UseLoopCountLimit=false) {enabled=false}
   &SpawnRate(SpawnRate=&Multiply_Float(A=$User.Lightning Spawn Rate, B=2), Use Spawn Probability=false) {module=SpawnRate001}
  ]
  particleSpawnScript [
   /SetVariables(Particles.isShort:NiagaraFloat=&FloatFromCurve(
     FloatCurve=DataInterface() {propertyValues {Curve {keys [<at 0 0 RCIM_Constant>,<at 0.8 1 RCIM_Constant>]}}}
     CurveIndex=&RandomRangeFloat(Minimum=0, Maximum=1)
    )) {isSetParameters=true}
   &InitializeParticle(
    Color=&SelectLinearColorFromArray(
     Array Sampling Mode=::NewEnumerator1
     Direct Array Index=&RandomRangeInt(Minimum=0, Maximum=1)
     Color Selection Array=DataInterface() {propertyValues {ColorData [<rgba 0.5 0.0145 0.2765 1>, <rgba 0.5 0 0.2161 1>]}}
     Random Seed=?
    )
    Lifetime Max=0.4
    Lifetime Min=0.2
    Lifetime Mode=::NewEnumerator1
    Mesh Scale Mode=::NewEnumerator3
    Mesh Uniform Scale=&MakeFloatFromInt(Int=&Multiply_Int(A=&RandomRangeInt(Minimum=4, Maximum=12), B=4))
    Position Mode=::NewEnumerator1
    UsePositionOffset=false
    Write Material Random=true
   ) {module=InitializeParticle001}
   &DynamicMaterialParameters(
    Index 0 Param 1=1
    Index 0 Param 2=0.15
    Index 0 Param 3=1
    Index 0 Param 4=1
    Module.Param0WriteEnabled=true
    Module.Param1WriteEnabled=true
    Module.Param2WriteEnabled=true
    Module.Param3WriteEnabled=true
   )
   &ShapeLocation(
    MaskSpawnGroup=false
    Non Uniform Scale=<1 1 0>
    Offset Mode=::NewEnumerator0
    Override Seed=false
    Sphere Radius=3000
    Sphere Surface Distribution=1
    UseOwnerScale=true
   ) {module=ShapeLocation002}
   &AddVelocity(Module.Use Falloff Distance=false, Override Seed=false, Velocity Mode=::NewEnumerator1, Velocity Speed=0.25)
   &InitialMeshOrientation(
    Mesh Orientation Mode=::NewEnumerator7
    Rotation=&RandomRangeVector(Minimum=<0 0 0>, Maximum=<0 0 0>)
    Use Rotation Vector=true
   ) {module=InitialMeshOrientation002}
  ]
  particleUpdateScript [
   &SolveForcesAndVelocity(Clamp Velocity=false, Limit Acceleration=false)
   &ParticleState()
   &ScaleMeshSize(Scale Factor=&Multiply_VectorByFloat(
     Vector=&MakeVector(
      X=&Multiply_Float(
       A=&FloatFromCurve(FloatCurve=DataInterface() {propertyValues {Curve {keys [{interpMode=RCIM_Cubic, tangentMode=RCTM_User, time=0, value=1}, {interpMode=RCIM_Cubic, tangentMode=RCTM_User, time=1, value=1.5}]}}})
       B=&RandomRangeFloat(Minimum=0.5, Maximum=0.8)
      )
      Y=1
      Z=1
     )
     Float=&MakeFloatFromVector(
      Vector=&VectorFromFloat(Value=&Multiply_Float(
        A=&RandomRangeFloat(Minimum=0.8, Maximum=1.2)
        B=&Multiply_Float(
         A=0.1
         B=&FloatFromCurve(FloatCurve=DataInterface() {propertyValues {Curve {keys [
              {arriveTangent=0.2102, interpMode=RCIM_Cubic, leaveTangent=0.2102, tangentMode=RCTM_User, time=0, value=0.9}
              {interpMode=RCIM_Cubic, tangentMode=RCTM_User, time=1, value=1}
             ]}}})
        )
       ))
      Channel=::NewEnumerator0
     )
    ))
   &DynamicMaterialParameters(
    Index 0 Param 1=&Lerp_Float(
     A=&FloatFromCurve(FloatCurve=DataInterface() {propertyValues {Curve {keys [{tangentMode=RCTM_User, time=0, value=1}, <at 0.2 0.5>, <at 0.8 0.5>, {tangentMode=RCTM_User, time=1, value=0}]}}})
     B=0.4
     Alpha=$Particles.isShort
    )
    Index 0 Param 2=&FloatFromCurve(
     FloatCurve=DataInterface() {propertyValues {Curve {keys [
         {interpMode=RCIM_Cubic, leaveTangent=-1.1405, tangentMode=RCTM_Break, time=0, value=0.5}
         {arriveTangent=4.4936, interpMode=RCIM_Cubic, leaveTangent=-5.3647, tangentMode=RCTM_Break, time=0.4833, value=1}
         <at 1 0>
        ]}}}
     Scale Curve=&Multiply_Float(A=10, B=$User.WidthMult)
    )
    Index 0 Param 3=120
    Index 1 Param 1=&RandomRangeFloat(Minimum=0, Maximum=1)
    Index 1 Param 2=1
    Index 1 Param 3=1
    Index 1 Param 4=1
    Module.Param0WriteEnabled=true
    Module.Param0WriteEnabled1=true
    Module.Param1WriteEnabled=true
    Module.Param1WriteEnabled1=true
    Module.Param2WriteEnabled=true
    Module.Param2WriteEnabled1=true
    Module.Param3WriteEnabled=false
    Module.Param3WriteEnabled1=true
    Write Parameter Index 1=true
   ) {module=DynamicMaterialParameters001}
   &ScaleColor(
    Scale Alpha=&FloatFromCurve(FloatCurve=DataInterface() {propertyValues {Curve {keys [
         {interpMode=RCIM_Cubic, tangentMode=RCTM_User, time=0, value=0}
         {interpMode=RCIM_Cubic, tangentMode=RCTM_User, time=0.1258, value=1}
         <at 0.341 0>
         {interpMode=RCIM_Cubic, tangentMode=RCTM_User, time=0.496, value=0.5}
         <at 1 0>
        ]}}})
    Scale RGB=&VectorFromFloat(Value=&FloatFromCurve(
      FloatCurve=DataInterface() {propertyValues {Curve {keys [
          {arriveTangent=0.1143, interpMode=RCIM_Cubic, leaveTangent=0.1143, tangentMode=RCTM_User, time=0.018, value=4.9326}
          {arriveTangent=-6.0507, interpMode=RCIM_Cubic, leaveTangent=-6.0507, time=0.1778, value=0.8567}
          {arriveTangent=-0.5645, interpMode=RCIM_Cubic, leaveTangent=-0.5645, tangentMode=RCTM_User, time=0.2493, value=3.533}
          {arriveTangent=-22.9399, interpMode=RCIM_Cubic, leaveTangent=-22.9399, time=0.3224, value=1.7136}
          {arriveTangent=2.8763, interpMode=RCIM_Cubic, leaveTangent=2.8763, time=0.3773, value=0.5981}
          {arriveTangent=-1.8312, interpMode=RCIM_Cubic, leaveTangent=-1.8312, time=0.4693, value=2.1361}
          {arriveTangent=-2.2307, interpMode=RCIM_Cubic, leaveTangent=-2.2307, time=0.5983, value=0.1933}
          {arriveTangent=1.5587, interpMode=RCIM_Cubic, leaveTangent=1.5587, time=0.7391, value=1.5342}
          {arriveTangent=-2.0478, interpMode=RCIM_Cubic, leaveTangent=-2.0478, time=0.8629, value=0.6057}
          {arriveTangent=-0.0659, interpMode=RCIM_Cubic, leaveTangent=-0.0659, tangentMode=RCTM_User, time=1, value=1}
         ]}}}
      Scale Curve=$User.ColorMult
     ))
    ScaleA=true
    ScaleRGB=true
   ) {module=ScaleColor001}
  ]
  rendererClasses [&NiagaraMeshRendererProperties]
  renderers [
   {rendererClass=&NiagaraMeshRendererProperties, rendererIndex=0}
   {rendererClass=&NiagaraMeshRendererProperties, rendererIndex=1}
  ]
 }
 MiddleGlow_Dark {
  bEnabled=false
  emitterUpdateScript [
   &EmitterState()
   &SpawnBurst_Instantaneous(Spawn Time=$System.SpawnTime, Use Spawn Probability=false, UseLoopCountLimit=true)
  ]
  particleSpawnScript [
   &InitializeParticle(
    Color=<rgba 0.0091 0.0167 0.062 1>
    Lifetime=1.6
    Sprite Size=<26250 52500>
    Sprite Size Mode=::NewEnumerator4
    UsePositionOffset=true
    Write Material Random=true
   )
   &CameraOffset(Camera Offset Amount=2000)
  ]
  particleUpdateScript [
   &ParticleState(Kill Particles When Lifetime Has Elapsed=false, Let Infinitely Lived Particles Die When Emitter Deactivates=true)
   /SetVariables(Particles.Size002:Vector2f=&Multiply_Vector2D_ByFloat(
     Float=$System.PortalScale
     Vector2D=&Multiply_Vector2D_ByFloat(
      Float=$User.Fade
      Vector2D=&Multiply_Vector2D(
       A=<9000 9000>
       B=&Multiply_Vector2D_ByFloat(
        Float=&Waveform("Waveform [1]"=::NewEnumerator12, "[1] Frequency"=9, "[1] Amplitude Min/Max"=<0.9 1.1>)
        Vector2D=<4 0.5>
       )
      )
     )
    )) {isSetParameters=true}
   /SetVariables(Particles.color002:LinearColor=&MakeLinearColorFromVectorAndFloat(
     "Vector (RGB)"=&Multiply_VectorByFloat(
      Vector=&MakeVectorFromLinearColorRGB(Color=&LerpLinearColors(
        StartColor=<rgba 1 0.0639 0.3455 1>
        EndColor=<rgba 1 0.0734 0.3551 1>
        LerpFactor=&Waveform("Waveform [1]"=::NewEnumerator12, "[1] Frequency"=3, "[1] Amplitude Min/Max"=<0 1>)
       ))
      Float=3
     )
     "Float (Alpha)"=&Multiply_Float(A=0.25, B=&Waveform("Waveform [1]"=::NewEnumerator12, "[1] Frequency"=3, "[1] Amplitude Min/Max"=<0.2 1>))
    )) {isSetParameters=true}
  ]
  rendererClasses [&NiagaraSpriteRendererProperties]
  renderers [{rendererClass=&NiagaraSpriteRendererProperties}]
 }
 Ring_Energy {
  bEnabled=false
  emitterUpdateScript [
   &EmitterState()
   &SpawnBurst_Instantaneous(Spawn Count=2, Use Spawn Probability=false, UseLoopCountLimit=false)
   &SpawnRate(
    SpawnRate=&FloatFromCurve(
     FloatCurve=DataInterface() {propertyValues {Curve {keys [
         {tangentMode=RCTM_User, time=0, value=1}
         {arriveTangent=-0.0405, interpMode=RCIM_Cubic, leaveTangent=-0.0405, tangentMode=RCTM_User, time=0.7009, value=1}
         {arriveTangent=-8.0089, interpMode=RCIM_Cubic, leaveTangent=-8.0089, tangentMode=RCTM_User, time=1, value=0}
        ]}}}
     CurveIndex=$Emitter.NormalizedLoopAge
     Scale Curve=5
    )
    Use Spawn Probability=false
   )
  ]
  particleSpawnScript [
   &InitializeParticle(
    Lifetime Min=1.4
    Lifetime Mode=::NewEnumerator1
    Mesh Scale=&Multiply_VectorByFloat(
     Vector=&MakeVector(X=1, Y=1, Z=&RandomRangeFloat(Minimum=3, Maximum=4))
     Float=&RandomRangeFloat(Minimum=10, Maximum=4)
    )
    Mesh Scale Mode=::NewEnumerator4
    UsePositionOffset=true
    Write Material Random=true
   ) {module=InitializeParticle001}
   &SystemLocation(MaskSpawnGroup=false, Offset=&RandomRangeVector(Minimum=<0 0 0>, Maximum=<0 0 3>))
   &InitialMeshOrientation(
    Mesh Orientation Mode=::NewEnumerator2
    Orientation Coordinate Space=::Local
    Rotation=&RandomRangeVector(Minimum=<0 0 0>, Maximum=<0 0 1>)
    Use Rotation Vector=true
   ) {module=InitialMeshOrientation001}
   &AddVelocity(Velocity=&RandomRangeVector(Minimum=<0 0 0>, Maximum=<0 0 50>))
  ]
  particleUpdateScript [
   &SolveForcesAndVelocity(Clamp Velocity=false, Limit Acceleration=false)
   &ParticleState()
   &UpdateMeshOrientation(
    Rotation Rate=&SelectFloatFromArray(
     Array Sampling Mode=::NewEnumerator0
     Direct Array Index=?
     Float Selection Array=DataInterface() {propertyValues {FloatData [-0.15,0.15,0.1,-0.1]}}
     Random Seed=0
    )
    Rotation Vector=<0 0 1>
   )
   &DynamicMaterialParameters(
    Index 0 Param 1=&FloatFromCurve(
     FloatCurve=DataInterface() {propertyValues {Curve {keys [{interpMode=RCIM_Cubic, leaveTangent=-0.0324, tangentMode=RCTM_Break, time=0, value=1.5}, <at 1 -0.5>]}}}
     Scale Curve=&RandomRangeFloat(Minimum=0.8, Maximum=1)
    )
    Index 0 Param 2=&FloatFromCurve(FloatCurve=DataInterface() {propertyValues {Curve {keys [
         {arriveTangent=0.3262, interpMode=RCIM_Cubic, leaveTangent=0.3262, tangentMode=RCTM_User, time=0.2, value=0}
         {arriveTangent=1.6237, interpMode=RCIM_Cubic, leaveTangent=1.6237, tangentMode=RCTM_User, time=1, value=1}
        ]}}})
    Module.Index0Float0Write=true
    Module.Index0Float1Write=true
    Module.Index0Float2Write=true
    Module.Index0Float3Write=true
   )
   &ScaleMeshSize(Scale Factor=&Multiply_VectorByFloat(
     Vector=&Multiply_VectorByFloat(
      Vector=&MakeVector(
       X=1
       Y=1
       Z=&Multiply_Float(
        A=&FloatFromCurve(
         FloatCurve=DataInterface() {propertyValues {Curve {keys [{arriveTangent=-1.572, time=0, value=1}, {interpMode=RCIM_Cubic, tangentMode=RCTM_User, time=1, value=0.5}]}}}
         Scale Curve=0.5
        )
        B=&RandomRangeFloat(Minimum=1, Maximum=1)
       )
      )
      Float=0.8
     )
     Float=&FloatFromCurve(
      FloatCurve=DataInterface() {propertyValues {Curve {keys [
          {arriveTangent=4.7736, interpMode=RCIM_Cubic, leaveTangent=4.7736, tangentMode=RCTM_User, time=0, value=2.0364}
          {arriveTangent=0.9183, interpMode=RCIM_Cubic, leaveTangent=0.9183, tangentMode=RCTM_User, time=1, value=3.5}
         ]}}}
      Scale Curve=&Multiply_Float(
       A=1.7
       B=&FloatFromCurve(
        FloatCurve=DataInterface() {propertyValues {Curve {keys [{interpMode=RCIM_Cubic, tangentMode=RCTM_User, time=0, value=0.4}, {interpMode=RCIM_Cubic, tangentMode=RCTM_User, time=0.2095, value=1}]}}}
        CurveIndex=$Emitter.Age
       )
      )
     )
    ))
   &ScaleColor(
    Scale Alpha=&FloatFromCurve(
     FloatCurve=DataInterface() {propertyValues {Curve {keys [
         {arriveTangent=-0.0329, interpMode=RCIM_Cubic, leaveTangent=-0.0329, tangentMode=RCTM_User, time=0, value=0}
         {arriveTangent=3.3381, interpMode=RCIM_Cubic, leaveTangent=3.3381, tangentMode=RCTM_User, time=0.1, value=0.6}
         {arriveTangent=0.0085, interpMode=RCIM_Cubic, leaveTangent=0.0085, tangentMode=RCTM_User, time=0.2, value=0.75}
         {arriveTangent=-1.463, interpMode=RCIM_Cubic, leaveTangent=-1.463, tangentMode=RCTM_User, time=0.6, value=0.45}
         {arriveTangent=0.0414, interpMode=RCIM_Cubic, leaveTangent=0.0414, tangentMode=RCTM_User, time=1, value=0}
        ]}}}
     Scale Curve=&RandomRangeFloat(Minimum=1, Maximum=1.4)
    )
    Scale RGB=&MakeVectorFromLinearColorRGB(Color=&LerpLinearColors(
      StartColor=&SelectLinearColorFromArray(
       Array Sampling Mode=::NewEnumerator0
       Direct Array Index=?
       Color Selection Array=DataInterface() {propertyValues {ColorData [<rgba 0.0457 0.005 0.0343 1>, <rgba 1 0.1161 0.439 1>]}}
       Random Seed=0
      )
      EndColor=<rgba 11.5381 1.6095 20 1>
      LerpFactor=&FloatFromCurve(
       FloatCurve=DataInterface() {propertyValues {Curve {keys [
           {arriveTangent=-7.9858, interpMode=RCIM_Cubic, leaveTangent=-7.9858, tangentMode=RCTM_User, time=0, value=1}
           {interpMode=RCIM_Cubic, tangentMode=RCTM_User, time=0.4067, value=0}
          ]}}}
       CurveIndex=$Emitter.Age
      )
     ))
    ScaleA=true
    ScaleRGB=true
   )
  ]
  rendererClasses [&NiagaraMeshRendererProperties]
  renderers [{rendererClass=&NiagaraMeshRendererProperties, rendererIndex=0}]
 }
 Ring_Line {
  bEnabled=false
  emitterUpdateScript [
   &EmitterState()
   &SpawnRate(SpawnRate=6, Use Spawn Probability=false)
  ]
  particleSpawnScript [
   &InitializeParticle(
    Color=<rgba 1 1 1 0.6734>
    Lifetime Max=1.6
    Lifetime Mode=::NewEnumerator1
    Mesh Scale=&Multiply_VectorByFloat(
     Vector=&MakeVector(X=1, Y=1, Z=&RandomRangeFloat(Minimum=10, Maximum=12))
     Float=&RandomRangeFloat(Minimum=14.4, Maximum=16.8)
    )
    Mesh Scale Mode=::NewEnumerator4
    Position Offset=<0 0 30>
    UsePositionOffset=true
    Write Material Random=true
   ) {module=InitializeParticle001}
   &SystemLocation(MaskSpawnGroup=false, Offset=&RandomRangeVector(Minimum=<0 0 0>, Maximum=<0 0 10>))
   &InitialMeshOrientation(
    Mesh Orientation Mode=::NewEnumerator2
    Orientation Coordinate Space=::Local
    Rotation=&RandomRangeVector(Minimum=<0 -0.05 0>, Maximum=<0 0.05 1>)
    Use Rotation Vector=true
   ) {module=InitialMeshOrientation001}
  ]
  particleUpdateScript [
   &ParticleState()
   &UpdateMeshOrientation(
    Rotation Rate=&SelectFloatFromArray(
     Array Sampling Mode=::NewEnumerator0
     Direct Array Index=?
     Float Selection Array=DataInterface() {propertyValues {FloatData [-0.15,0.15,0.1,-0.1]}}
     Random Seed=0
    )
    Rotation Vector=<0 0 1>
   )
   &ScaleMeshSize(Scale Factor=&VectorFromFloat(Value=3.5)) {module=ScaleMeshSize001}
   &ScaleMeshSize(Scale Factor=&Multiply_VectorByFloat(
     Vector=&Multiply_VectorByFloat(
      Vector=&MakeVector(
       X=1
       Y=1
       Z=&Multiply_Float(
        A=&FloatFromCurve(
         FloatCurve=DataInterface() {propertyValues {Curve {keys [{arriveTangent=-1.572, time=0, value=1}, {interpMode=RCIM_Cubic, tangentMode=RCTM_User, time=1, value=0.5}]}}}
         Scale Curve=0.7
        )
        B=&RandomRangeFloat(Minimum=1, Maximum=1)
       )
      )
      Float=0.8
     )
     Float=&FloatFromCurve(
      FloatCurve=DataInterface() {propertyValues {Curve {keys [
          {arriveTangent=2.4838, interpMode=RCIM_Cubic, leaveTangent=2.4838, tangentMode=RCTM_User, time=0, value=1.295}
          {arriveTangent=0.3815, interpMode=RCIM_Cubic, leaveTangent=0.3815, tangentMode=RCTM_User, time=1, value=2}
         ]}}}
      Scale Curve=&FN_FrontEndFloatModifier(Gameplay=1.7, Front End=1.5)
     )
    ))
   &ScaleColor(
    Scale Alpha=&FloatFromCurve(
     FloatCurve=DataInterface() {propertyValues {Curve {keys [
         {arriveTangent=6.8924, interpMode=RCIM_Cubic, leaveTangent=6.8924, tangentMode=RCTM_User, time=0, value=0}
         {arriveTangent=3.3638, interpMode=RCIM_Cubic, leaveTangent=3.3638, tangentMode=RCTM_User, time=0.1, value=0.6}
         {arriveTangent=0.0044, interpMode=RCIM_Cubic, leaveTangent=0.0044, tangentMode=RCTM_User, time=0.2, value=0.75}
         {arriveTangent=-1.0842, interpMode=RCIM_Cubic, leaveTangent=-1.0842, tangentMode=RCTM_User, time=0.6953, value=0.5076}
         {arriveTangent=-1.8777, interpMode=RCIM_Cubic, leaveTangent=-1.8777, tangentMode=RCTM_User, time=1, value=0}
        ]}}}
     Scale Curve=2
    )
    Scale RGB=&MakeVectorFromLinearColorRGB(Color=&ColorFromCurve(ColorCurve=DataInterface() {propertyValues {
        AlphaCurve {keys []}
        BlueCurve {keys [
          {arriveTangent=0.3322, interpMode=RCIM_Cubic, leaveTangent=0.3322, tangentMode=RCTM_User, time=0, value=0.285}
          <at 0.4237 0.053>
          <at 0.7343 0.0723>
         ]}
        GreenCurve {keys [
          {arriveTangent=-1.1199, interpMode=RCIM_Cubic, leaveTangent=-1.1199, tangentMode=RCTM_User, time=0, value=0.1386}
          <at 0.4237 0.0686>
          <at 0.7343 0.025>
         ]}
        RedCurve {keys [
          {arriveTangent=-3.8908, interpMode=RCIM_Cubic, leaveTangent=-3.8908, tangentMode=RCTM_User, time=0, value=0.555}
          <at 0.4237 0.328>
          <at 0.7343 0.1615>
         ]}
       }}))
    ScaleA=true
    ScaleRGB=true
   )
  ]
  rendererClasses [&NiagaraMeshRendererProperties]
  renderers [{rendererClass=&NiagaraMeshRendererProperties, rendererIndex=0}]
 }
 SmokeRing {
  bEnabled=false
  emitterUpdateScript [
   &EmitterState()
   &SpawnBurst_Instantaneous(Spawn Count=4, Use Spawn Probability=false, UseLoopCountLimit=false)
   &SpawnRate(
    SpawnRate=&FloatFromCurve(
     FloatCurve=DataInterface() {propertyValues {Curve {keys [
         {arriveTangent=0.2741, interpMode=RCIM_Cubic, leaveTangent=0.2741, tangentMode=RCTM_User, time=0, value=1}
         {arriveTangent=0.0207, interpMode=RCIM_Cubic, leaveTangent=0.0207, tangentMode=RCTM_User, time=0.822, value=0.9854}
        ]}}}
     CurveIndex=$Emitter.NormalizedLoopAge
     Scale Curve=3.5
    )
    Use Spawn Probability=false
   )
  ]
  particleSpawnScript [
   &InitializeParticle(
    Lifetime Min=1.6
    Lifetime Mode=::NewEnumerator1
    Mesh Scale=&Multiply_VectorByFloat(
     Vector=&RandomRangeVector(Minimum=<1 1 1>, Maximum=<1 1 1.5>)
     Float=&RandomRangeFloat(Minimum=6.4, Maximum=9.6)
    )
    Mesh Scale Mode=::NewEnumerator4
    UsePositionOffset=false
    Write Material Random=true
   )
   &InitialMeshOrientation(
    Mesh Orientation Mode=::NewEnumerator2
    Rotation=&RandomRangeVector(Minimum=<0.005 0.005 -1>, Maximum=<-0.005 -0.005 1>)
    Use Rotation Vector=true
   )
   &ShapeLocation(
    Cylinder Height=3
    Cylinder Radius=1
    MaskSpawnGroup=false
    Offset=<0 0 5>
    Offset Mode=::NewEnumerator0
    Override Seed=false
    Shape Primitive=::NewEnumerator1
    Surface Only=false
    UseOwnerScale=true
   )
   &AddVelocity(Velocity=&RandomRangeVector(Minimum=<0 0 0>, Maximum=<0 0 50>))
  ]
  particleUpdateScript [
   &ParticleState()
   &DynamicMaterialParameters(
    Index 0 Param 1=&FloatFromCurve(FloatCurve=DataInterface() {propertyValues {Curve {keys [{interpMode=RCIM_Cubic, tangentMode=RCTM_User, time=0, value=0.8}, {interpMode=RCIM_Cubic, tangentMode=RCTM_User, time=1, value=0}]}}})
    Module.Index0Float0Write=true
    Module.Index0Float1Write=true
    Module.Index0Float2Write=true
    Module.Index0Float3Write=true
   )
   &ScaleMeshSize(Scale Factor=&Multiply_VectorByFloat(
     Vector=&Multiply_VectorByFloat(Vector=$Engine.Owner.Scale, Float=&FN_FrontEndFloatModifier(Gameplay=0.9, Front End=0.8))
     Float=&FloatFromCurve(
      FloatCurve=DataInterface() {propertyValues {Curve {keys [
          {arriveTangent=0.0893, interpMode=RCIM_Cubic, leaveTangent=0.0893, tangentMode=RCTM_User, time=0, value=0.55}
          {arriveTangent=0.0158, interpMode=RCIM_Cubic, leaveTangent=0.0158, tangentMode=RCTM_User, time=1, value=1}
         ]}}}
      Scale Curve=&Multiply_Float(
       A=4
       B=&FloatFromCurve(
        FloatCurve=DataInterface() {propertyValues {Curve {keys [{interpMode=RCIM_Cubic, tangentMode=RCTM_User, time=0, value=0.4}, {interpMode=RCIM_Cubic, tangentMode=RCTM_User, time=0.2095, value=1}]}}}
        CurveIndex=$Emitter.Age
       )
      )
     )
    ))
   &UpdateMeshOrientation(
    Rotation Rate=&SelectFloatFromArray(
     Array Sampling Mode=::NewEnumerator0
     Direct Array Index=?
     Float Selection Array=DataInterface() {propertyValues {FloatData [-0.15,0.15,0.1,-0.1]}}
     Random Seed=0
    )
    Rotation Vector=<0 0 1>
   )
   &ScaleColor(
    Scale Alpha=&FloatFromCurve(
     FloatCurve=DataInterface() {propertyValues {Curve {keys [
         {arriveTangent=1.0939, interpMode=RCIM_Cubic, leaveTangent=1.0939, tangentMode=RCTM_User, time=0, value=0}
         {arriveTangent=-0.0731, interpMode=RCIM_Cubic, leaveTangent=-0.0731, tangentMode=RCTM_User, time=0.326, value=1}
         {arriveTangent=-0.05, interpMode=RCIM_Cubic, leaveTangent=-0.8744, tangentMode=RCTM_Break, time=0.4, value=1}
         {arriveTangent=-0.5359, tangentMode=RCTM_User, time=1, value=0}
        ]}}}
     Scale Curve=0.8
    )
    Scale RGB=&MakeVectorFromLinearColorRGB(Color=&LerpLinearColors(
      StartColor=&ColorFromCurve(ColorCurve=DataInterface() {propertyValues {
         AlphaCurve {keys []}
         BlueCurve {keys [
           {arriveTangent=0.3322, interpMode=RCIM_Cubic, leaveTangent=0.3322, tangentMode=RCTM_User, time=0.1294, value=0.2234}
           <at 0.4879 0.0897>
           <at 0.7343 0.0918>
          ]}
         GreenCurve {keys [
           {arriveTangent=-1.1199, interpMode=RCIM_Cubic, leaveTangent=-1.1199, tangentMode=RCTM_User, time=0.1294, value=0.0949}
           <at 0.4879 0.0488>
           <at 0.7343 0.0664>
          ]}
         RedCurve {keys [
           {arriveTangent=-3.8908, interpMode=RCIM_Cubic, leaveTangent=-3.8908, tangentMode=RCTM_User, time=0.1294, value=0.9578}
           <at 0.4879 0.4271>
           <at 0.7343 0.5312>
          ]}
        }})
      EndColor=<rgba 20 0 12.5636 1>
      LerpFactor=&FloatFromCurve(
       FloatCurve=DataInterface() {propertyValues {Curve {keys [
           {arriveTangent=-5.4468, interpMode=RCIM_Cubic, leaveTangent=-5.4468, tangentMode=RCTM_User, time=0, value=1}
           {interpMode=RCIM_Cubic, tangentMode=RCTM_User, time=0.4046, value=0}
          ]}}}
       CurveIndex=$Emitter.Age
      )
     ))
    ScaleA=true
    ScaleRGB=true
   )
   &SolveForcesAndVelocity(Clamp Velocity=false, Limit Acceleration=false)
  ]
  rendererClasses [&NiagaraMeshRendererProperties]
  renderers [{rendererClass=&NiagaraMeshRendererProperties, rendererIndex=0}]
 }
 Trail_ring001 {
  bEnabled=false
  emitterUpdateScript [
   &EmitterState()
   &SpawnBurst_Instantaneous(Use Spawn Probability=false, UseLoopCountLimit=false)
   &SpawnRate(SpawnRate=2, Use Spawn Probability=false)
  ]
  particleSpawnScript [
   &InitializeParticle(
    Color=<rgba 0.2292 0.0346 0.1384 1>
    Lifetime=1.3
    Mesh Scale=&Multiply_VectorByFloat(Vector=<1 1 1>, Float=&RandomRangeFloat(Minimum=20, Maximum=30))
    Mesh Scale Mode=::NewEnumerator4
    Position Offset=&RandomRangeVector(Minimum=<0 0 5>, Maximum=<0 0 50>)
    UsePositionOffset=true
    Write Material Random=true
   )
   &InitialMeshOrientation(
    Mesh Axis To Orient=<0 1 0>
    Mesh Orientation Mode=::NewEnumerator2
    Orientation Vector=<0 0 1>
    Rotation=&RandomRangeVector(Minimum=<0 0 0>, Maximum=<0 1 0>)
    Use Rotation Vector=true
   )
   &DynamicMaterialParameters(
    Index 0 Param 1=&Multiply_Float(
     A=&RandomRangeFloat(Minimum=30, Maximum=30)
     B=&FloatFromCurve(
      FloatCurve=DataInterface() {propertyValues {Curve {keys [
          <at 1 2.5 RCIM_Cubic>
          {arriveTangent=-1.44, interpMode=RCIM_Cubic, leaveTangent=-1.44, time=1.6444, value=1.8}
          {arriveTangent=-0.0085, tangentMode=RCTM_User, time=2.25, value=0.7}
         ]}}}
      CurveIndex=$Emitter.Age
      Scale Curve=0
     )
    )
    Index 0 Param 2=&Multiply_Float(
     A=&RandomRangeFloat(Minimum=20, Maximum=30)
     B=&FloatFromCurve(
      FloatCurve=DataInterface() {propertyValues {Curve {keys [
          <at 1 2.5 RCIM_Cubic>
          {arriveTangent=-1.44, interpMode=RCIM_Cubic, leaveTangent=-1.44, time=1.6444, value=1.8}
          {arriveTangent=-0.0085, tangentMode=RCTM_User, time=2.25, value=0.7}
         ]}}}
      CurveIndex=$Emitter.Age
      Scale Curve=0
     )
    )
    Index 0 Param 3=1.5
    Index 0 Param 4=&RandomRangeFloat(Minimum=0, Maximum=1)
    Index 1 Param 1=&RandomRangeFloat(Minimum=0.2, Maximum=1)
    Index 1 Param 2=&FloatFromCurve(FloatCurve=DataInterface() {propertyValues {Curve {keys [{tangentMode=RCTM_User, time=0, value=1}, {tangentMode=RCTM_User, time=0.5, value=1}]}}})
    Module.Index0Float0Write=true
    Module.Index0Float1Write=true
    Module.Index0Float2Write=true
    Module.Index0Float3Write=true
    Module.Index1Float0Write=true
    Module.Index1Float1Write=true
    Module.Index1Float2Write=false
    Module.Index1Float3Write=false
    Write Parameter Index 1=::NewEnumerator1
   )
   &CameraOffset(Camera Offset Amount=5)
  ]
  particleUpdateScript [
   &ParticleState()
   &ScaleMeshSize(Scale Factor=&Multiply_VectorByFloat(
     Vector=<1 1 1>
     Float=&Multiply_Float(
      A=0.8
      B=&FloatFromCurve(
       FloatCurve=DataInterface() {propertyValues {Curve {keys [<at 0 0.8>,<at 1 1.2>]}}}
       Scale Curve=&FloatFromCurve(FloatCurve=DataInterface() {propertyValues {Curve {keys [{interpMode=RCIM_Cubic, tangentMode=RCTM_User, time=0, value=0.4}, {interpMode=RCIM_Cubic, tangentMode=RCTM_User, time=0.2095, value=1}]}}})
      )
     )
    ))
   &ScaleColor(
    Scale Alpha=&FloatFromCurve(
     FloatCurve=DataInterface() {propertyValues {Curve {keys [
         {arriveTangent=4.6142, interpMode=RCIM_Cubic, leaveTangent=4.6142, tangentMode=RCTM_User, time=0, value=0}
         {interpMode=RCIM_Cubic, tangentMode=RCTM_User, time=0.5, value=1}
         {arriveTangent=-2, interpMode=RCIM_Cubic, leaveTangent=-2, time=0.8764, value=0.8105}
         {arriveTangent=-5.0551, interpMode=RCIM_Cubic, leaveTangent=-5.0551, tangentMode=RCTM_User, time=1, value=0}
        ]}}}
     Scale Curve=&FloatFromCurve(FloatCurve=DataInterface() {propertyValues {Curve {keys [<at 0.7 1>,<at 2.5753 1>,<at 2.7698 0>]}}}, Scale Curve=0.8)
    )
    Scale RGB=&Multiply_VectorByFloat(Vector=<1 1 1>, Float=1)
    ScaleA=true
    ScaleRGB=true
   )
  ]
  rendererClasses [&NiagaraMeshRendererProperties]
  renderers [{rendererClass=&NiagaraMeshRendererProperties, rendererIndex=0}]
 }
 WispySmoke {
  bEnabled=false
  emitterUpdateScript [
   &EmitterState()
   &SpawnRate(SpawnRate=15, Use Spawn Probability=false)
   &SpawnBurst_Instantaneous(Spawn Count=4, Use Spawn Probability=false, UseLoopCountLimit=false)
  ]
  particleSpawnScript [
   &InitializeParticle(
    Color Maximum=<rgba 0.7216 0.1093 0.5686 1>
    Color Minimum=<rgba 0.1568 0.0288 0.1548 1>
    Color Mode=::NewEnumerator2
    Lifetime Max=1.2
    Lifetime Min=0.8
    Lifetime Mode=::NewEnumerator1
    Sprite Rotation Mode=::NewEnumerator2
    Sprite Size Mode=::NewEnumerator1
    Uniform Sprite Size Max=1000
    Uniform Sprite Size Min=700
    UsePositionOffset=false
    Write Material Random=true
   )
   &ShapeLocation(
    Handle Radius=0
    Large Radius=1500
    MaskSpawnGroup=false
    Offset=<0 0 100>
    Offset Mode=::NewEnumerator0
    Override Seed=false
    Shape Primitive=::NewEnumerator3
    UseOwnerScale=true
   )
   &CameraOffset(Camera Offset Amount=30)
   &AddVelocity(Velocity=&RandomRangeVector(Minimum=<0 0 100>, Maximum=<0 125 50>))
  ]
  particleUpdateScript [
   &ParticleState()
   &SpriteRotationRate(Rotation Rate=&RandomRangeFloat(Minimum=-50, Maximum=50))
   &ScaleSpriteSize(
    Scale Sprite Size Mode=::NewEnumerator0
    Uniform Scale Factor=&FloatFromCurve(FloatCurve=DataInterface() {propertyValues {Curve {keys [<at 0 0.8>,<at 1 1.2>]}}})
   )
   &SubUVAnimation(
    End Frame Range Override=30
    "Sprite Renderer "=DataInterface() {propertyValues {SpriteRenderer=&self:WispySmoke_0.NiagaraSpriteRendererProperties_0}}
    Start Frame Range Override=&RandomRangeInt(Minimum=0, Maximum=5)
    UseEndFrame=true
    UseStartFrame=true
   )
   &ScaleColor(
    Scale Alpha=&Multiply_Float(
     A=&RandomRangeFloat(Minimum=0.5, Maximum=1)
     B=&FloatFromCurve(
      FloatCurve=DataInterface() {propertyValues {Curve {keys [
          {arriveTangent=6.7463, interpMode=RCIM_Cubic, leaveTangent=6.7463, tangentMode=RCTM_User, time=0, value=0}
          {arriveTangent=2, interpMode=RCIM_Cubic, leaveTangent=2, time=0.0827, value=0.7619}
          {interpMode=RCIM_Cubic, tangentMode=RCTM_User, time=0.5, value=1}
          {arriveTangent=-5.5093, interpMode=RCIM_Cubic, leaveTangent=-5.5093, tangentMode=RCTM_User, time=1, value=0}
         ]}}}
      Scale Curve=&FloatFromCurve(
       FloatCurve=DataInterface() {propertyValues {Curve {keys [
           {tangentMode=RCTM_User, time=0, value=1}
           {arriveTangent=-0.0405, interpMode=RCIM_Cubic, leaveTangent=-0.0405, tangentMode=RCTM_User, time=1.5916, value=1}
           {arriveTangent=-2.5083, interpMode=RCIM_Cubic, leaveTangent=-2.5083, tangentMode=RCTM_User, time=2.4445, value=0}
          ]}}}
       CurveIndex=$Emitter.Age
      )
     )
    )
    ScaleA=true
    ScaleRGB=true
   )
   &SolveForcesAndVelocity(Clamp Velocity=false, Limit Acceleration=false)
  ]
  rendererClasses [&NiagaraSpriteRendererProperties]
  renderers [{rendererClass=&NiagaraSpriteRendererProperties}]
 }
}
summary {
 systemName=NS_PistolCanary_PortalCharge
 emitters [
  {bEnabled=false, emitterName=EnergyWave, rendererClasses [&NiagaraMeshRendererProperties]}
  {bEnabled=false, emitterName=MiddleGlow_Dark, rendererClasses [&NiagaraSpriteRendererProperties]}
  {bEnabled=false, emitterName=WispySmoke, rendererClasses [&NiagaraSpriteRendererProperties]}
  {bEnabled=false, emitterName=Ring_Energy, rendererClasses [&NiagaraMeshRendererProperties]}
  {bEnabled=false, emitterName=SmokeRing, rendererClasses [&NiagaraMeshRendererProperties]}
  {bEnabled=false, emitterName=Ring_Line, rendererClasses [&NiagaraMeshRendererProperties]}
  {bEnabled=false, emitterName=Trail_ring001, rendererClasses [&NiagaraMeshRendererProperties]}
  {
   bEnabled=false
   emitterName=Glow
   rendererClasses [&NiagaraSpriteRendererProperties, &NiagaraLightRendererProperties]
  }
  {emitterName=Lightning_Primary001, rendererClasses [&NiagaraMeshRendererProperties]}
 ]
 userVariables [
  User.bIsFrontEnd {type=NiagaraBool}
  User.bIsFrontEndPreview {type=NiagaraBool}
  User.BlackHoleEnd {type=Vector3f, description="", defaultValue {struct=Vector3f, value=<0 50 -150>}}
  User.BlackHoleStart {type=NiagaraPosition, description="", defaultValue {}}
  User.ColorMult {type=NiagaraFloat, defaultValue=1, description=""}
  User.CoolColor_01 {type=LinearColor, description="", defaultValue {struct=LinearColor, value=<rgba 0.2245 0 0.4844 1>}}
  User.CoolColor_02 {type=LinearColor, description="", defaultValue {struct=LinearColor, value=<rgba 0.1234 0.0016 0.151 1>}}
  User.ExplosionScale {type=NiagaraFloat, defaultValue=1.2, description=""}
  User.Fade {type=NiagaraFloat, defaultValue=1, description=""}
  User.GlobalScale {type=NiagaraFloat, defaultValue=33, description=""}
  User.Intensity {type=NiagaraFloat, defaultValue=50, description=""}
  User.Lifetime {type=NiagaraFloat, defaultValue=3, description=""}
  User.Lightning Spawn Rate {type=NiagaraFloat, defaultValue=0, description=""}
  User.Open {type=NiagaraFloat, defaultValue=2, description=""}
  User.OverallWidth {type=NiagaraFloat, defaultValue=3720.3586, description=""}
  User.ParaGlideLeanAlpha {type=NiagaraFloat, defaultValue=0, description=""}
  User.PlayerSpeed {type=NiagaraFloat, defaultValue=0, description=""}
  User.PortalSize {type=NiagaraFloat, defaultValue=1, description=""}
  User.RemoveBeam {type=NiagaraFloat, defaultValue=0, description=""}
  User.Scale {type=NiagaraFloat, defaultValue=1, description=""}
  User.SFXVolume {type=NiagaraFloat, defaultValue=0, description=""}
  User.SkeletalMesh {
   type=NiagaraDataInterfaceSkeletalMesh
   defaultValue {
    struct=NiagaraExt_VariableValue_DataInterface
    value {dataInterface=&self:NiagaraDataInterfaceSkeletalMesh_0}
   }
  }
  User.SplineUpdatedEndPos {type=Vector3f, description="", defaultValue {struct=Vector3f, value=<0 0 -25920.5664>}}
  User.SplineUpdatedStartPos {type=NiagaraPosition, description="", defaultValue {}}
  User.Volume {type=NiagaraFloat, defaultValue=0, description=""}
  User.WarmColor_01 {type=LinearColor, description="", defaultValue {struct=LinearColor, value=<rgba 1 0.0284 0.1722 1>}}
  User.WarmColor_02 {type=LinearColor, description="", defaultValue {struct=LinearColor, value=<rgba 1 0.0036 0.093 1>}}
  User.WidthMult {type=NiagaraFloat, defaultValue=1, description=""}
 ]
}
