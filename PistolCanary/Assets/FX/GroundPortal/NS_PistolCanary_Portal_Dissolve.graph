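# Niagara system for the dissolve effect of the PistolCanary ground portal, built from timed, mostly magenta-purple emitters.
# Core glow, flash, shock cone and accelerating energy-trail meshes start immediately, while a wave sprite, bolt meshes and flares follow after loop delays.
# Curves over particle life drive scale, color and dynamic material parameters, with light renderers on the cone and wave.
# Some variants, such as EnergyWave and Flash_Flare, are disabled.
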NiagaraSystem /PistolCanary/Assets/FX/GroundPortal/NS_PistolCanary_Portal_Dissolve
dependencies {
 usedDataInterfaces [
  &NiagaraDataInterfaceVectorCurve
  &NiagaraDataInterfaceCurve
  &NiagaraDataInterfaceColorCurve
  &NiagaraDataInterfaceVector2DCurve
  &NiagaraDataInterfaceCamera
  &NiagaraDataInterfaceArrayColor
 ]
 usedDynamicInputs [&MakeVector4_FromLinearColor, &MaskFloatBySpawnGroup]
 usedModules [&SystemState]
}
emitters {
 Bolt {
  emitterUpdateScript [
   &EmitterState(Life Cycle Mode=::NewEnumerator1, Loop Behavior=::NewEnumerator1, Loop Delay=3, Loop Duration=2, UseLoopDelay=true)
   &SpawnRate(SpawnRate=8, Use Spawn Probability=false)
   &SpawnBurst_Instantaneous(Spawn Count=10, Use Spawn Probability=false, UseLoopCountLimit=false)
  ]
  particleSpawnScript [
   /SetVariables(Particles.MeshIndex=&RandomRangeInt(Minimum=0, Maximum=3)) {isSetParameters=true}
   &InitializeParticle(
    Color=&SelectLinearColorFromArray(
     Array Sampling Mode=::NewEnumerator1
     Direct Array Index=&RandomRangeInt(Minimum=0, Maximum=1)
     Color Selection Array=DataInterface() {propertyValues {ColorData [<rgba 1.7708 0 1.7708 1>]}}
     Random Seed=?
    )
    Lifetime=&RandomRangeFloat(Minimum=0.6, Maximum=0.5)
    Mesh Scale=<0 0 0>
    Mesh Scale Mode=::NewEnumerator4
    UsePositionOffset=false
    Write Material Random=true
   )
   &SystemLocation(MaskSpawnGroup=false, Offset=&RandomRangeVector(Minimum=<0 0 25>, Maximum=<0 0 25>))
   &InitialMeshOrientation(
    Mesh Orientation Mode=::NewEnumerator2
    Orientation Vector=<0 0 1>
    Rotation=&MakeVector(X=&RandomRangeFloat(Minimum=0, Maximum=1), Y=0, Z=0)
    Use Rotation Vector=true
   )
   &AddVelocityInCone(
    Cone Angle=90
    Cone Axis=<0 0 0.25>
    Module.Use Velocity Falloff On Cone Axis=false
    Velocity Distribution Along Cone Axis=0.25
    Velocity Strength=-25
   ) {enabled=false}
   &AddVelocity(Velocity=<0 0 1>) {enabled=false}
   /SetVariables(Particles.Scale=&Multiply_VectorByFloat(
     Vector=<2 2 2>
     Float=&RemapRange(
      Input Value To Remap=&FN_LOD_Distance_Remap(Start Distance=5000, End Distance=15000, Minimum Output=0.65, Maximum Output=1)
      Input Min=0
      Input Max=1
      Output Min=0
      Output Max=1
     )
    )) {isSetParameters=true}
   &CameraOffset(Camera Offset Amount=&Lerp_Float(
     A=-200
     B=0
     Alpha=&FN_LOD_Distance_Remap(Start Distance=0, End Distance=1500, Minimum Output=0, Maximum Output=1)
    )) {enabled=false}
  ]
  particleUpdateScript [
   &SolveForcesAndVelocity(Clamp Velocity=false, Limit Acceleration=false)
   &ParticleState()
   &ScaleColor(
    Scale Alpha=&FloatFromCurve(FloatCurve=DataInterface() {propertyValues {Curve {keys [
         {interpMode=RCIM_Cubic, tangentMode=RCTM_User, time=0, value=0}
         {interpMode=RCIM_Cubic, tangentMode=RCTM_User, time=0.5, value=1}
         {tangentMode=RCTM_User, time=0.6, value=0}
        ]}}})
    Scale RGB=&VectorFromFloat(Value=&FloatFromCurve(
      FloatCurve=DataInterface() {propertyValues {Curve {keys [
          {interpMode=RCIM_Cubic, tangentMode=RCTM_User, time=0, value=0}
          {arriveTangent=-3.0443, interpMode=RCIM_Cubic, leaveTangent=-3.0443, tangentMode=RCTM_User, time=0.1, value=0.95}
          {interpMode=RCIM_Cubic, tangentMode=RCTM_User, time=0.4, value=0.55}
          {tangentMode=RCTM_User, time=1, value=0}
         ]}}}
      Scale Curve=15
     ))
    ScaleA=true
    ScaleRGB=true
   )
   &DynamicMaterialParameters(
    Index 0 Param 1=70
    Index 0 Param 2=&FloatFromCurve(FloatCurve=DataInterface() {propertyValues {Curve {keys [
         {interpMode=RCIM_Cubic, tangentMode=RCTM_User, time=0, value=0}
         {arriveTangent=-2.9636, interpMode=RCIM_Cubic, leaveTangent=-2.9636, tangentMode=RCTM_User, time=0.4, value=0.95}
         {tangentMode=RCTM_User, time=0.9, value=0}
        ]}}})
    Index 0 Param 3=&FloatFromCurve(
     FloatCurve=DataInterface() {propertyValues {Curve {keys [
         {interpMode=RCIM_Cubic, tangentMode=RCTM_User, time=0, value=0}
         {tangentMode=RCTM_User, time=1.5, value=0.85}
         {interpMode=RCIM_Cubic, tangentMode=RCTM_User, time=2.5, value=1}
        ]}}}
     Scale Curve=2
    )
    Index 0 Param 4=1
    Index 1 Param 1=&FloatFromCurve(FloatCurve=DataInterface() {propertyValues {Curve {keys [{interpMode=RCIM_Cubic, tangentMode=RCTM_User, time=0, value=1}, {interpMode=RCIM_Cubic, tangentMode=RCTM_User, time=1, value=0}]}}})
    Index 1 Param 2=50
    Index 1 Param 3=1
    Index 1 Param 4=1
    Index 2 Param 1=&FloatFromCurve(FloatCurve=DataInterface() {propertyValues {Curve {keys [
         {arriveTangentWeight=0.3333, interpMode=RCIM_Cubic, leaveTangentWeight=0.3333, tangentMode=RCTM_User, time=0, value=1}
         {tangentMode=RCTM_User, time=1, value=0.5}
        ]}}})
    Module.Param0WriteEnabled=true
    Module.Param0WriteEnabled1=true
    Module.Param0WriteEnabled2=true
    Module.Param1WriteEnabled=true
    Module.Param1WriteEnabled1=true
    Module.Param1WriteEnabled2=true
    Module.Param2WriteEnabled=true
    Module.Param2WriteEnabled1=true
    Module.Param2WriteEnabled2=true
    Module.Param3WriteEnabled=true
    Module.Param3WriteEnabled1=true
    Module.Param3WriteEnabled2=true
    Write Parameter Index 1=true
    Write Parameter Index 2=true
   )
   &ScaleMeshSize(Scale Factor=&MakeVector(
     X=&FloatFromCurve(
      FloatCurve=DataInterface() {propertyValues {Curve {keys [
          {arriveTangent=-5.4159, interpMode=RCIM_Cubic, leaveTangent=-5.4159, tangentMode=RCTM_User, time=0, value=1.5}
          {interpMode=RCIM_Cubic, tangentMode=RCTM_User, time=1, value=1.5}
         ]}}}
      Scale Curve=2
     )
     Y=&FloatFromCurve(FloatCurve=DataInterface() {propertyValues {Curve {keys [
          {arriveTangent=73.8235, interpMode=RCIM_Cubic, leaveTangent=73.8235, tangentMode=RCTM_User, time=0, value=0}
          {tangentMode=RCTM_User, time=0.05, value=1}
         ]}}})
     Z=&FloatFromCurve(FloatCurve=DataInterface() {propertyValues {Curve {keys [
          {arriveTangent=9.6743, interpMode=RCIM_Cubic, leaveTangent=9.6743, tangentMode=RCTM_User, time=0, value=0}
          {interpMode=RCIM_Cubic, tangentMode=RCTM_User, time=0.1588, value=1.0402}
          {interpMode=RCIM_Cubic, tangentMode=RCTM_User, time=0.9122, value=1.3662}
          <at 1 1>
         ]}}})
    ))
  ]
  rendererClasses [&NiagaraMeshRendererProperties]
  renderers [{rendererClass=&NiagaraMeshRendererProperties, rendererIndex=0}]
 }
 Cone_Shock_Light {
  emitterUpdateScript [
   &EmitterState(Life Cycle Mode=::NewEnumerator1, Loop Behavior=::NewEnumerator1, Loop Delay=3, Loop Duration=1, UseLoopDelay=false)
   &SpawnBurst_Instantaneous(Use Spawn Probability=false, UseLoopCountLimit=false)
  ]
  particleSpawnScript [
   &InitializeParticle(
    Color=<rgba 1.7708 0 1.7708 2>
    Lifetime=4
    Mesh Scale=&Multiply_VectorByFloat(
     Vector=<5 1 1>
     Float=&FN_LOD_Distance_Remap(Start Distance=4500, End Distance=15000, Minimum Output=12, Maximum Output=12, Clamp Output=true)
    )
    Mesh Scale Mode=::NewEnumerator4
    Sprite Size Mode=::NewEnumerator3
    Uniform Sprite Size=&Multiply_Float(
     A=800
     B=&FN_LOD_Distance_Remap(Start Distance=4500, End Distance=22500, Minimum Output=20, Maximum Output=20, Clamp Output=true)
    )
    UsePositionOffset=true
    Write Material Random=true
   )
   &InitialMeshOrientation(Mesh Axis To Orient=<0 0 1>, Mesh Orientation Mode=::NewEnumerator2, Use Rotation Vector=true)
   &DynamicMaterialParameters(
    Index 0 Param 1=0.7
    Index 0 Param 2=0.3
    Index 0 Param 3=2
    Index 2 Param 1=300
    Index 3 Param 1=0.5
    Index 3 Param 2=3
    Index 3 Param 3=1
    Index 3 Param 4=1
    Module.Param0WriteEnabled=true
    Module.Param0WriteEnabled2=true
    Module.Param0WriteEnabled3=true
    Module.Param1WriteEnabled=true
    Module.Param1WriteEnabled2=false
    Module.Param1WriteEnabled3=true
    Module.Param2WriteEnabled=true
    Module.Param2WriteEnabled2=false
    Module.Param2WriteEnabled3=true
    Module.Param3WriteEnabled=false
    Module.Param3WriteEnabled2=false
    Module.Param3WriteEnabled3=true
    Write Parameter Index 2=true
    Write Parameter Index 3=true
   )
   &CameraOffset(Camera Offset Amount=2000)
  ]
  particleUpdateScript [
   &ParticleState()
   &ScaleSpriteSize(Uniform Curve Sprite Scale=DataInterface() {propertyValues {Curve {defaultValue=1, keys [{tangentMode=RCTM_User, time=0, value=1}, {tangentMode=RCTM_User, time=1, value=0}]}}})
   &ScaleMeshSize(Scale Factor=&VectorFromCurve(
     VectorCurve=DataInterface() {propertyValues {
       XCurve {keys [
         {
          arriveTangent=1.5385
          arriveTangentWeight=0.3333
          interpMode=RCIM_Cubic
          leaveTangent=1.5385
          leaveTangentWeight=0.6116
          tangentMode=RCTM_User
          time=0
          value=0
         }
         {
          arriveTangent=0.1488
          arriveTangentWeight=0.337
          interpMode=RCIM_Cubic
          leaveTangent=0.1488
          leaveTangentWeight=0.337
          tangentMode=RCTM_User
          time=1
          value=1
         }
        ]}
       YCurve {keys [
         {
          arriveTangent=2.3109
          arriveTangentWeight=0.8393
          interpMode=RCIM_Cubic
          leaveTangent=2.3109
          leaveTangentWeight=0.8393
          tangentMode=RCTM_User
          time=0
          value=0
         }
         {
          arriveTangent=0.6124
          arriveTangentWeight=0.3909
          interpMode=RCIM_Cubic
          leaveTangent=0.6124
          leaveTangentWeight=0.3909
          tangentMode=RCTM_User
          time=1
          value=1
         }
        ]}
       ZCurve {keys [
         {
          arriveTangent=2.3109
          arriveTangentWeight=0.8393
          interpMode=RCIM_Cubic
          leaveTangent=2.3109
          leaveTangentWeight=0.8393
          tangentMode=RCTM_User
          time=0
          value=0
         }
         {
          arriveTangent=0.6124
          arriveTangentWeight=0.3909
          interpMode=RCIM_Cubic
          leaveTangent=0.6124
          leaveTangentWeight=0.3909
          tangentMode=RCTM_User
          time=1
          value=1
         }
        ]}
      }}
     Scale Curve=<0.3 7 7>
    ))
   &ScaleColor(
    Scale Alpha=&FloatFromCurve(FloatCurve=DataInterface() {propertyValues {Curve {keys [
         {arriveTangent=48.0273, interpMode=RCIM_Cubic, time=0, value=0}
         <at 0.2 1 RCIM_Cubic>
         {interpMode=RCIM_Cubic, leaveTangent=-3.1778, time=1, value=0}
        ]}}})
    Scale RGB=&MakeVectorFromLinearColorRGB(Color=&ColorFromCurve(ColorCurve=DataInterface() {propertyValues {
        AlphaCurve {keys []}
        BlueCurve {keys [<at 0.1026 1 RCIM_Cubic>,<at 0.959 0.1 RCIM_Cubic>]}
        GreenCurve {keys [<at 0.1026 1 RCIM_Cubic>,<at 0.959 0.0492 RCIM_Cubic>]}
        RedCurve {keys [<at 0.1026 1 RCIM_Cubic>,<at 0.959 0.053 RCIM_Cubic>]}
       }}))
    ScaleA=true
    ScaleRGB=true
   )
   /SetVariables(Particles.FlareColor=&MultiplyLinearColor(
     Linear Color=$Particles.Color
     Scale Factor=&MakeVector4(
      X=1
      Y=1
      Z=1
      W=&FloatFromCurve(FloatCurve=DataInterface() {propertyValues {Curve {keys [<at 0 1>,<at 1 0>]}}})
     )
    )) {isSetParameters=true}
   /SetVariables(Particles.MeshColor:LinearColor=&MultiplyLinearColor(
     Linear Color=$Particles.Color
     Scale Factor=&MakeVector4(
      X=1
      Y=1
      Z=1
      W=&FloatFromCurve(FloatCurve=DataInterface() {propertyValues {Curve {keys [<at 0 0.3>,<at 1 0>]}}})
     )
    )) {isSetParameters=true}
   &DynamicMaterialParameters(
    Index 1 Param 1=&FloatFromCurve(FloatCurve=DataInterface() {propertyValues {Curve {keys [<at 0 -1 RCIM_Cubic>,<at 1 0>]}}})
    Module.Param0WriteEnabled1=true
    Module.Param1WriteEnabled1=true
    Module.Param2WriteEnabled1=false
    Module.Param3WriteEnabled1=false
    Write Parameter Index 0=false
    Write Parameter Index 1=true
   ) {module=DynamicMaterialParameters001}
   &FN_Light_Attributes(
    Color=<rgba 0.5899 2.008 4 1>
    Create Color Attribute=true
    Create Exponent Attribute=true
    Create Position Attribute=true
    Create Radius Attribute=true
    Exponent=2
    Position=&AddVectorToPosition(
     Position=$Particles.SkeletalMeshLocation.SampledPosition
     Vector=&RotateVectorByQuaternion(Vector To Rotate=<0 50 0>, Quaternion=$Particles.SkeletalMeshLocation.SampledRotation)
    )
    Radius=2000
    Scale Color=&FloatFromCurve(
     FloatCurve=DataInterface() {propertyValues {Curve {keys [
         {arriveTangent=1.5581, interpMode=RCIM_Cubic, time=0, value=0}
         {arriveTangent=0.1339, interpMode=RCIM_Cubic, leaveTangent=0.1339, tangentMode=RCTM_User, time=0.05, value=1}
         {arriveTangent=-2.9133, interpMode=RCIM_Cubic, leaveTangent=-2.9133, tangentMode=RCTM_User, time=1, value=0}
        ]}}}
     Scale Curve=6
    )
   )
  ]
  rendererClasses [
   &NiagaraMeshRendererProperties
   &NiagaraSpriteRendererProperties
   &NiagaraLightRendererProperties
  ]
  renderers [
   {rendererClass=&NiagaraMeshRendererProperties, rendererIndex=0}
   {rendererClass=&NiagaraSpriteRendererProperties, rendererIndex=1}
   {rendererClass=&NiagaraLightRendererProperties, rendererIndex=2}
  ]
 }
 CoreGlow {
  emitterUpdateScript [
   &EmitterState(Life Cycle Mode=::NewEnumerator1, Loop Behavior=::NewEnumerator1, Loop Duration=1, UseLoopDelay=false)
   &SpawnBurst_Instantaneous(Use Spawn Probability=false, UseLoopCountLimit=false)
  ]
  particleSpawnScript [&InitializeParticle(
    Color=<rgba 5 0 4 1>
    Lifetime=2.5
    Mesh Scale Mode=::NewEnumerator3
    Mesh Uniform Scale=20
    Position Offset=<0 0 -1000>
    Sprite Size Mode=::NewEnumerator3
    Uniform Sprite Size=1000
    UsePositionOffset=false
    Write Material Random=true
   )]
  particleUpdateScript [
   &ParticleState()
   &ScaleSpriteSize(Uniform Curve Sprite Scale=DataInterface() {propertyValues {Curve {keys [<at 0 3 RCIM_Cubic>, {arriveTangent=-2, interpMode=RCIM_Cubic, leaveTangent=-2, time=0.85, value=5}, <at 1 1>]}}})
   &ScaleColor(
    Scale Alpha=&FloatFromCurve(FloatCurve=DataInterface() {propertyValues {Curve {
        defaultValue=0
        keys [
         {arriveTangent=9.0725, interpMode=RCIM_Cubic, time=0, value=0.5}
         {arriveTangent=-0.4822, interpMode=RCIM_Cubic, leaveTangent=-0.4822, tangentMode=RCTM_User, time=0.95, value=0.9847}
         <at 1 0 RCIM_Cubic>
        ]
       }}})
    Scale RGB=&Multiply_VectorByFloat(
     Vector=<1 1 1>
     Float=&FloatFromCurve(FloatCurve=DataInterface() {propertyValues {Curve {keys [<at 0 10 RCIM_Cubic>,<at 1 20>]}}})
    )
    ScaleA=true
    ScaleRGB=true
   )
   &DynamicMaterialParameters(
    Module.Index0Float0Write=true
    Module.Index0Float1Write=true
    Module.Index0Float2Write=true
    Module.Index0Float3Write=true
    Module.Index1Float0Write=true
    Module.Index1Float1Write=true
    Module.Index1Float2Write=true
    Module.Index1Float3Write=true
    Write Parameter Index 1=::NewEnumerator1
   )
  ]
  rendererClasses [&NiagaraSpriteRendererProperties]
  renderers [{rendererClass=&NiagaraSpriteRendererProperties}]
 }
 CoreGlow001 {
  bEnabled=false
  emitterUpdateScript [
   &EmitterState(Life Cycle Mode=::NewEnumerator1, Loop Behavior=::NewEnumerator1, Loop Duration=1, UseLoopDelay=false)
   &SpawnBurst_Instantaneous(Use Spawn Probability=false, UseLoopCountLimit=false)
  ]
  particleSpawnScript [&InitializeParticle(
    Color=<rgba 5 0 4 10>
    Lifetime=3.18
    Mesh Scale Mode=::NewEnumerator3
    Mesh Uniform Scale=4
    Position Offset=<0 0 -1000>
    Sprite Size Mode=::NewEnumerator3
    Uniform Sprite Size=1000
    UsePositionOffset=false
    Write Material Random=true
   )]
  particleUpdateScript [
   &ParticleState()
   &ScaleMeshSize(Scale Factor=&VectorFromFloat(Value=&FloatFromCurve(FloatCurve=DataInterface() {propertyValues {Curve {keys [<at 0 3 RCIM_Cubic>, {arriveTangent=-2, interpMode=RCIM_Cubic, leaveTangent=-2, time=0.87, value=5}, <at 1 1>]}}})))
   &ScaleSpriteSize(Uniform Curve Sprite Scale=DataInterface() {propertyValues {Curve {keys [<at 0 1 RCIM_Cubic>, {arriveTangent=2, interpMode=RCIM_Cubic, leaveTangent=2, time=0.8, value=5}, <at 1 3>]}}}) {enabled=false}
   &UpdateMeshOrientation(
    Rotation Coordinate Space=::NewEnumerator0
    Rotation Vector=&RandomRangeVector(Minimum=<0 0 0>, Maximum=<0 0 1>)
   ) {enabled=false}
   &ScaleMeshSize(Scale Factor=&Multiply_VectorByFloat(
     Vector=&VectorFromCurve(VectorCurve=DataInterface() {propertyValues {
        XCurve {keys [<at 0 0.2>,<at 0.75 1>,<at 0.8 0.5>,<at 0.9 10>,<at 1 13>]}
        YCurve {keys [<at 0 0.2>,<at 0.75 1>,<at 0.8 0.5>,<at 0.9 10>,<at 1 13>]}
        ZCurve {keys [<at 0 0.2>,<at 0.75 1>,<at 0.8 0.5>,<at 0.9 0.3>,<at 1 0.2>]}
       }})
     Float=1
    )) {enabled=false, module=ScaleMeshSize001}
   &ScaleColor(
    Scale Alpha=&FloatFromCurve(FloatCurve=DataInterface() {propertyValues {Curve {
        defaultValue=0
        keys [
         {arriveTangent=9.0725, interpMode=RCIM_Cubic, time=0, value=0.5}
         {arriveTangent=-0.4822, interpMode=RCIM_Cubic, leaveTangent=-0.4822, tangentMode=RCTM_User, time=0.95, value=0.9847}
         <at 1 0 RCIM_Cubic>
        ]
       }}})
    Scale RGB=&Multiply_VectorByFloat(
     Vector=<1 1 1>
     Float=&FloatFromCurve(FloatCurve=DataInterface() {propertyValues {Curve {keys [<at 0 1 RCIM_Cubic>,<at 1 2>]}}})
    )
    ScaleA=true
    ScaleRGB=true
   )
   &DynamicMaterialParameters(
    Index 0 Param 4=&FloatFromCurve(FloatCurve=DataInterface() {propertyValues {Curve {keys [<at 0.35 1 RCIM_Cubic>,<at 1 0>]}}})
    Module.Index0Float0Write=true
    Module.Index0Float1Write=true
    Module.Index0Float2Write=true
    Module.Index0Float3Write=true
    Module.Index1Float0Write=true
    Module.Index1Float1Write=true
    Module.Index1Float2Write=true
    Module.Index1Float3Write=true
    Write Parameter Index 1=::NewEnumerator1
   )
  ]
  rendererClasses [&NiagaraMeshRendererProperties]
  renderers [{rendererClass=&NiagaraMeshRendererProperties, rendererIndex=0}]
 }
 EnergyTrail {
  emitterUpdateScript [
   &EmitterState(Life Cycle Mode=::NewEnumerator1, Loop Behavior=::NewEnumerator1, Loop Duration=1.8, UseLoopDelay=false)
   &SpawnRate(
    SpawnRate=&FloatFromCurve(FloatCurve=DataInterface() {propertyValues {Curve {keys [<at 0 2 RCIM_Cubic>,<at 1 20>]}}}, CurveIndex=$Emitter.NormalizedLoopAge)
    Use Spawn Probability=false
   )
   &SpawnBurst_Instantaneous(Spawn Count=5, Use Spawn Probability=false, UseLoopCountLimit=false)
  ]
  particleSpawnScript [
   &InitializeParticle(
    Color=&RandomRangeLinearColor(Minimum=<rgba 0.0831 0.0156 0.1 1>, Maximum=<rgba 0.5498 0.1034 0.6615 3>)
    Lifetime Max=2.5
    Lifetime Min=2
    Lifetime Mode=::NewEnumerator1
    Mesh Scale Mode=::NewEnumerator1
    Mesh Uniform Scale Max=200
    Mesh Uniform Scale Min=100
    Position Offset=<0 0 500>
    UsePositionOffset=true
    Write Material Random=true
   )
   &InitialMeshOrientation(
    Mesh Orientation Mode=::NewEnumerator1
    Rotation=&RandomRangeVector(Minimum=<-0.02 -0.02 0>, Maximum=<0.02 0.02 1>)
    Rotation Coordinate Space=::NewEnumerator1
    Use Rotation Vector=true
   ) {module=InitialMeshOrientation001}
   &DynamicMaterialParameters(
    Index 0 Param 4=&RandomRangeFloat(Minimum=0.25, Maximum=1)
    Module.Index0Float0Write=true
    Module.Index0Float1Write=true
    Module.Index0Float2Write=true
    Module.Index0Float3Write=true
   ) {enabled=false, module=DynamicMaterialParameters001}
  ]
  particleUpdateScript [
   &ParticleState(Lifetime:NiagaraFloat=&FloatFromCurve(FloatCurve=DataInterface() {propertyValues {Curve {keys [<at 0 3 RCIM_Cubic>,<at 0.9 1>]}}}))
   &UpdateMeshOrientation(
    Rotation Rate=&FloatFromCurve(FloatCurve=DataInterface() {propertyValues {Curve {keys [<at 0 1 RCIM_Cubic>,<at 0.5 0>]}}})
    Rotation Vector=<0 0 1>
   ) {enabled=false}
   &ScaleMeshSize(Scale Factor=&VectorFromFloat(Value=&FloatFromCurve(FloatCurve=DataInterface() {propertyValues {Curve {
         defaultValue=0
         keys [{interpMode=RCIM_Cubic, tangentMode=RCTM_User, time=0, value=1}, {interpMode=RCIM_Cubic, tangentMode=RCTM_User, time=1, value=0.8}]
        }}}))) {enabled=false}
   &ScaleColor(
    Scale Alpha=&FloatFromCurve(FloatCurve=DataInterface() {propertyValues {Curve {
        defaultValue=0
        keys [
         {arriveTangent=9.0725, interpMode=RCIM_Cubic, leaveTangent=9.0725, tangentMode=RCTM_User, time=0, value=0}
         {arriveTangent=-0.0405, interpMode=RCIM_Cubic, leaveTangent=-0.0405, tangentMode=RCTM_User, time=0.85, value=3}
         {interpMode=RCIM_Cubic, leaveTangent=-2.5249, time=1, value=0}
        ]
       }}})
    Scale RGB=&VectorFromCurve(VectorCurve=DataInterface() {propertyValues {
       XCurve {keys [<at 0.75 1 RCIM_Cubic>,<at 0.8 100>]}
       YCurve {keys [<at 0.75 1 RCIM_Cubic>,<at 0.8 100>]}
       ZCurve {keys [<at 0.75 1 RCIM_Cubic>,<at 0.8 100>]}
      }})
    ScaleA=true
    ScaleRGB=true
   )
   &MeshRotationForce(Rotation=&RandomRangeVector(Minimum=<0 0 0>, Maximum=<0.25 0.25 2>)) {enabled=false}
   &SolveForcesAndVelocity(Clamp Velocity=false, Limit Acceleration=false)
   &DynamicMaterialParameters(
    Index 0 Param 1=&FloatFromCurve(FloatCurve=DataInterface() {propertyValues {Curve {
        defaultValue=1
        keys [
         {arriveTangent=3.4793, interpMode=RCIM_Cubic, leaveTangent=3.4793, tangentMode=RCTM_User, time=0, value=-1}
         {tangentMode=RCTM_User, time=1, value=1}
        ]
       }}})
    Index 0 Param 2=&RandomRangeFloat(Minimum=0, Maximum=1)
    Index 0 Param 3=&RandomRangeFloat(Minimum=0, Maximum=1)
    Index 0 Param 4=&RandomRangeFloat(Minimum=0.25, Maximum=1)
    Module.Index0Float0Write=true
    Module.Index0Float1Write=true
    Module.Index0Float2Write=true
    Module.Index0Float3Write=true
   )
  ]
  rendererClasses [&NiagaraMeshRendererProperties]
  renderers [{rendererClass=&NiagaraMeshRendererProperties, rendererIndex=0}]
 }
 EnergyWave {
  bEnabled=false
  emitterUpdateScript [
   &EmitterState(
    Life Cycle Mode=::NewEnumerator1
    Loop Behavior=::NewEnumerator1
    Loop Delay=3
    Loop Duration=0.1
    Scalability Mode=::NewEnumerator1
    UseLoopDelay=true
   )
   &SpawnBurst_Instantaneous(Use Spawn Probability=false, UseLoopCountLimit=false)
   &SpawnBurst_Instantaneous(Spawn Time=0.7333, Use Spawn Probability=false, UseLoopCountLimit=false) {enabled=false, module=SpawnBurst_Instantaneous001}
   &SpawnBurst_Instantaneous(Spawn Time=0.8667, Use Spawn Probability=false, UseLoopCountLimit=false) {module=SpawnBurst_Instantaneous002}
  ]
  particleSpawnScript [
   &InitializeParticle(
    Color=&LerpLinearColors(
     StartColor=<rgba 1.7708 0 1.7708 1>
     EndColor=&HueShiftLinearColor(Hue=0.06, Saturation=1, Value=1, Linear Color=$User.Color)
     LerpFactor=$User.UseCustomColor
    )
    Mesh Scale=<25 35 10>
    Mesh Scale Mode=::NewEnumerator4
    UsePositionOffset=false
    Write Material Random=true
   )
   &InitialMeshOrientation(Orientation Axis=<0 0 1>, Orientation Vector=<0 0 1>, Rotation=<0 0.25 0>, Use Orientation Vector=true, Use Rotation Vector=true)
   &InitialMeshOrientation(
    Orientation Axis=<0 0 0>
    Orientation Vector=<0 0 0>
    Rotation=&RandomRangeVector(Minimum=<0 0 0>, Maximum=<1 0 0>)
    Use Orientation Vector=false
    Use Rotation Vector=true
   ) {enabled=false, module=InitialMeshOrientation001}
   &CameraOffset(Camera Offset Amount=30)
   &ShapeLocation(MaskSpawnGroup=false, Offset=<0 0 25>, Offset Mode=::NewEnumerator0, Override Seed=false, Sphere Radius=1, UseOwnerScale=true)
   /SetVariables(Particles.LightPosOffset:NiagaraPosition=<0 0 50>) {isSetParameters=true}
  ]
  particleUpdateScript [
   &ParticleState()
   &ScaleColor(
    Scale Alpha=&FloatFromCurve(FloatCurve=DataInterface() {propertyValues {Curve {keys [<at 0.25 1>,<at 1 0>]}}})
    Scale RGB=&VectorFromFloat(Value=&FloatFromCurve(
      FloatCurve=DataInterface() {propertyValues {Curve {keys [{tangentMode=RCTM_User, time=0, value=0.5}, {tangentMode=RCTM_User, time=1, value=1}]}}}
      Scale Curve=2
     ))
    ScaleA=true
    ScaleRGB=true
   ) {enabled=false}
   &DynamicMaterialParameters(
    Index 0 Param 1=&FloatFromCurve(FloatCurve=DataInterface() {propertyValues {Curve {keys [{arriveTangent=-2.1015, interpMode=RCIM_Cubic, leaveTangent=-2.1015, tangentMode=RCTM_User, time=0, value=0.5}, <at 1 -0.15>]}}})
    Index 0 Param 2=&FloatFromCurve(FloatCurve=DataInterface() {propertyValues {Curve {keys [{arriveTangent=-0.3464, interpMode=RCIM_Cubic, leaveTangent=-0.3464, tangentMode=RCTM_User, time=0, value=1}, <at 1 0.9>]}}})
    Index 0 Param 3=&FloatFromCurve(FloatCurve=DataInterface() {propertyValues {Curve {keys [
         {interpMode=RCIM_Cubic, tangentMode=RCTM_Break, time=0, value=0}
         {arriveTangent=0.4478, interpMode=RCIM_Cubic, leaveTangent=0.4478, tangentMode=RCTM_User, time=0.15, value=1}
         {interpMode=RCIM_Cubic, tangentMode=RCTM_Break, time=1, value=1.2}
        ]}}})
    Module.Index0Float0Write=true
    Module.Index0Float1Write=true
    Module.Index0Float2Write=true
    Module.Index0Float3Write=true
   )
   &ScaleColor(
    Scale Alpha=&FloatFromCurve(FloatCurve=DataInterface() {propertyValues {Curve {keys [{interpMode=RCIM_Cubic, tangentMode=RCTM_User, time=0.5, value=1}, {interpMode=RCIM_Cubic, tangentMode=RCTM_User, time=1, value=0}]}}})
    Scale RGB=&VectorFromFloat(Value=&FloatFromCurve(FloatCurve=DataInterface() {propertyValues {Curve {keys [
          {interpMode=RCIM_Cubic, tangentMode=RCTM_Break, time=0, value=10}
          {arriveTangent=-9, leaveTangent=-9, time=0.15, value=2}
          {interpMode=RCIM_Cubic, tangentMode=RCTM_Break, time=0.5, value=1}
         ]}}}))
    ScaleA=true
    ScaleRGB=true
   ) {module=ScaleColor001}
   &ScaleMeshSize(Scale Factor=&VectorFromFloat(Value=&FloatFromCurve(FloatCurve=DataInterface() {propertyValues {Curve {keys [
          {interpMode=RCIM_Cubic, tangentMode=RCTM_Break, time=0, value=0.2}
          {arriveTangent=2.9365, interpMode=RCIM_Cubic, leaveTangent=2.9365, tangentMode=RCTM_User, time=0.1, value=1.5}
          {interpMode=RCIM_Cubic, tangentMode=RCTM_Break, time=1, value=2.5}
         ]}}})))
   &FN_Light_Attributes(
    Color=<rgba 1.7708 0 1.7708 1>
    Create Color Attribute=true
    Create Exponent Attribute=true
    Create Position Attribute=true
    Create Radius Attribute=true
    Exponent=2
    Position=&Add_Vector(A=Linked() {linkedVariable {name=Particles.Initial.Position, type=NiagaraPosition}}, B=<0 0 25>)
    Radius=100
    Scale Color=&FloatFromCurve(
     FloatCurve=DataInterface() {propertyValues {Curve {keys [
         <at 0 1 RCIM_Cubic>
         {arriveTangent=-1, interpMode=RCIM_Cubic, leaveTangent=-1, leaveTangentWeight=0.9776, time=0.7, value=1}
         <at 1 0 RCIM_Cubic>
        ]}}}
     Scale Curve=10
    )
   )
   /SetVariables(Particles.LightPos:NiagaraPosition=&Add_Vector(
     A=Linked() {linkedVariable {name=Particles.Position, type=NiagaraPosition}}
     B=Linked() {linkedVariable {name=Particles.LightPosOffset, type=NiagaraPosition}}
    )) {isSetParameters=true}
  ]
  rendererClasses [&NiagaraMeshRendererProperties, &NiagaraLightRendererProperties]
  renderers [
   {rendererClass=&NiagaraMeshRendererProperties, rendererIndex=0}
   {rendererClass=&NiagaraLightRendererProperties, rendererIndex=1}
  ]
 }
 EnergyWave002 {
  emitterUpdateScript [
   &EmitterState(Life Cycle Mode=::NewEnumerator1, Loop Behavior=::NewEnumerator1, Loop Delay=1, Loop Duration=1, UseLoopDelay=true)
   &SpawnBurst_Instantaneous(Use Spawn Probability=false, UseLoopCountLimit=false)
  ]
  particleSpawnScript [&InitializeParticle(
    Color=<rgba 4 4 4 2>
    Mesh Scale Mode=::NewEnumerator3
    Mesh Uniform Scale=-3.5
    Sprite Rotation Angle=90
    Sprite Rotation Mode=::NewEnumerator0
    Sprite Size Mode=::NewEnumerator3
    Uniform Sprite Size=&Multiply_Float(A=1000, B=10)
    UsePositionOffset=true
    Write Material Random=true
   )]
  particleUpdateScript [
   &ParticleState()
   &CameraOffset(Camera Offset Amount=&Multiply_Float(
     A=500
     B=&FloatFromCurve(
      FloatCurve=DataInterface() {propertyValues {Curve {keys [<at 0 0>,<at 1 1>]}}}
      CurveIndex=/Game/Effects/Niagara/Modules/Math/FN_CameraDistance(
       0Dist:NiagaraFloat=500
       1Dist:NiagaraFloat=2000
       Camera Query:NiagaraDataInterfaceCamera=DataInterface()
       Position:NiagaraPosition=$Particles.Position
      )
     )
    ))
   &ScaleSpriteSize(
    Uniform Curve Scale=2
    Uniform Curve Sprite Scale=DataInterface() {propertyValues {Curve {defaultValue=0, keys [{tangentMode=RCTM_User, time=0, value=1}, {tangentMode=RCTM_User, time=1, value=0}]}}}
   )
   &DynamicMaterialParameters(
    Index 0 Param 1=1
    Index 0 Param 2=&FloatFromCurve(FloatCurve=DataInterface() {propertyValues {Curve {keys [
         {interpMode=RCIM_Cubic, tangentMode=RCTM_User, time=0, value=1}
         {arriveTangent=-1, interpMode=RCIM_Cubic, leaveTangent=-1, time=0.6, value=0.7}
         {interpMode=RCIM_Cubic, leaveTangent=-3.3753, time=1, value=0}
        ]}}})
    Index 0 Param 3=1
    Index 0 Param 4=1
    Module.Param0WriteEnabled=true
    Module.Param1WriteEnabled=true
    Module.Param2WriteEnabled=true
    Module.Param3WriteEnabled=true
   )
   &ScaleColor(
    Scale Alpha=&FloatFromCurve(
     FloatCurve=DataInterface() {propertyValues {Curve {keys [
         <at 0 0 RCIM_Cubic>
         {arriveTangent=-2.878, interpMode=RCIM_Cubic, leaveTangent=-2.878, tangentMode=RCTM_User, time=0.2, value=1}
         {arriveTangent=-1.25, interpMode=RCIM_Cubic, leaveTangent=-1.25, time=0.35, value=0.2}
         {interpMode=RCIM_Cubic, tangentMode=RCTM_User, time=1, value=0}
        ]}}}
     Scale Curve=&Multiply_Float(A=50, B=1)
    )
    Scale RGB=&MakeVectorFromLinearColorRGB(Color=&MultiplyLinearColor(
      Linear Color=&ColorFromCurve(ColorCurve=DataInterface() {propertyValues {
         AlphaCurve {keys []}
         BlueCurve {keys [<at 0.0364 1.7708 RCIM_Cubic>, {interpMode=RCIM_Cubic, leaveTangent=-0.7175, time=0.3716, value=0.1}]}
         GreenCurve {keys [<at 0.0364 0 RCIM_Cubic>, {interpMode=RCIM_Cubic, leaveTangent=-4.4353, time=0.3716, value=0.0199}]}
         RedCurve {keys [<at 0.0364 1.7708 RCIM_Cubic>, {interpMode=RCIM_Cubic, leaveTangent=2.2176, time=0.3716, value=0.0946}]}
        }})
      Scale Factor=&MakeFloat4FromVectorAndFloat("Vector (RGB)"=&VectorFromFloat(Value=10), "Float (Alpha)"=1)
     ))
    ScaleA=true
    ScaleRGB=true
   )
   /SetVariables(Particles.Color02:LinearColor=&MakeLinearColorFromVectorAndFloat(
     "Vector (RGB)"=&MakeVectorFromLinearColorRGB(Color=&MultiplyLinearColor(
       Linear Color=<rgba 0.0831 0.0156 0.1 1>
       Scale Factor=&MakeFloat4FromVectorAndFloat("Vector (RGB)"=&VectorFromFloat(Value=3), "Float (Alpha)"=1)
      ))
     "Float (Alpha)"=&FloatFromCurve(
      FloatCurve=DataInterface() {propertyValues {Curve {keys [
          <at 0 0 RCIM_Cubic>
          {arriveTangent=-1.6399, interpMode=RCIM_Cubic, leaveTangent=-1.6399, tangentMode=RCTM_User, time=0.15, value=1}
          {interpMode=RCIM_Cubic, tangentMode=RCTM_User, time=1, value=0}
         ]}}}
      Scale Curve=2
     )
    )) {isSetParameters=true}
   &FN_Light_Attributes(
    Color=<rgba 1 0.156 0.1861 1>
    Create Color Attribute=true
    Create Position Attribute=true
    Create Radius Attribute=true
    Position=/Niagara/DynamicInputs/Helpers/SimulationPosition()
    Radius=1491.5984
    Scale Color=&FloatFromCurve(
     FloatCurve=DataInterface() {propertyValues {Curve {keys [<at 0 0>, {interpMode=RCIM_Cubic, leaveTangent=-2.9073, tangentMode=RCTM_Break, time=0.1, value=1}, <at 0.8 0>]}}}
     Scale Curve=4449615.5
    )
   )
   &SpriteFacingAndAlignment(Sprite Facing=<0 0 1>, WriteAlignment=true, WriteFacing=true)
  ]
  rendererClasses [&NiagaraSpriteRendererProperties, &NiagaraLightRendererProperties]
  renderers [
   {rendererClass=&NiagaraSpriteRendererProperties}
   {rendererClass=&NiagaraSpriteRendererProperties, rendererIndex=1}
   {rendererClass=&NiagaraLightRendererProperties, rendererIndex=2}
  ]
 }
 Flash {
  emitterUpdateScript [
   &EmitterState(Life Cycle Mode=::NewEnumerator1, Loop Behavior=::NewEnumerator1, Loop Delay=1, Loop Duration=1, UseLoopDelay=false)
   &SpawnBurst_Instantaneous(Use Spawn Probability=false, UseLoopCountLimit=false)
  ]
  particleSpawnScript [
   &InitializeParticle(
    Color=<rgba 7 1.7017 5.5336 1>
    Lifetime=3
    Mesh Scale Mode=::NewEnumerator3
    Mesh Uniform Scale=70
    Position Offset=<0 0 500>
    Sprite Size Mode=::NewEnumerator3
    Uniform Sprite Size=800
    UsePositionOffset=true
    Write Material Random=true
   )
   &CameraOffset(Camera Offset Amount=40)
   &InitialMeshOrientation(Mesh Orientation Mode=::NewEnumerator2, Orientation Vector=<0 0 1>, Use Rotation Vector=false)
  ]
  particleUpdateScript [
   &ParticleState()
   &ScaleMeshSize(Scale Factor=&VectorFromFloat(Value=&FloatFromCurve(FloatCurve=DataInterface() {propertyValues {Curve {defaultValue=0.6, keys [{tangentMode=RCTM_User, time=0, value=1}, {tangentMode=RCTM_User, time=1, value=0}]}}})))
   &ScaleColor(
    Scale Alpha=&FloatFromCurve(FloatCurve=DataInterface() {propertyValues {Curve {keys [
         {arriveTangent=-1.1111, arriveTangentWeight=0.162, interpMode=RCIM_Cubic, leaveTangentWeight=0.2268, time=0, value=0}
         <at 0.2 1 RCIM_Cubic>
         <at 1 0>
        ]}}})
    ScaleA=true
    ScaleRGB=false
   )
   &DynamicMaterialParameters(
    Index 0 Param 1=&FloatFromCurve(FloatCurve=DataInterface() {propertyValues {Curve {keys [
         {interpMode=RCIM_Cubic, tangentMode=RCTM_User, time=0.5, value=0}
         {arriveTangent=3.8908, interpMode=RCIM_Cubic, leaveTangent=3.8908, tangentMode=RCTM_User, time=1, value=1}
        ]}}})
    Index 0 Param 2=&FloatFromCurve(FloatCurve=DataInterface() {propertyValues {Curve {keys [
         {arriveTangent=-0.6134, interpMode=RCIM_Cubic, leaveTangent=-0.6134, tangentMode=RCTM_User, time=0, value=3}
         {arriveTangent=-3.248, interpMode=RCIM_Cubic, leaveTangent=-3.248, tangentMode=RCTM_User, time=1, value=0.5}
        ]}}})
    Index 0 Param 3=&FloatFromCurve(FloatCurve=DataInterface() {propertyValues {Curve {keys [
         {arriveTangent=-1.2425, interpMode=RCIM_Cubic, time=0, value=1}
         {arriveTangent=-2.1091, interpMode=RCIM_Cubic, leaveTangent=-2.1091, tangentMode=RCTM_User, time=1, value=0}
        ]}}})
    Index 0 Param 4=&FloatFromCurve(
     FloatCurve=DataInterface() {propertyValues {Curve {keys [
         {arriveTangent=2.3563, interpMode=RCIM_Cubic, leaveTangent=2.3563, tangentMode=RCTM_User, time=0, value=0}
         {tangentMode=RCTM_User, time=1, value=1}
        ]}}}
     Scale Curve=2
    )
    Module.Index0Param0Write=true
    Module.Index0Param1Write=true
    Module.Index0Param2Write=true
    Module.Index0Param3Write=true
   ) {module=DynamicMaterialParameters001}
   &ScaleSpriteSize(Uniform Curve Sprite Scale=DataInterface() {propertyValues {Curve {keys [
        {arriveTangent=13.251, interpMode=RCIM_Cubic, leaveTangent=13.251, tangentMode=RCTM_User, time=0, value=0.5}
        {arriveTangent=-1.9983, interpMode=RCIM_Cubic, leaveTangent=-1.9983, tangentMode=RCTM_User, time=0.11, value=1}
        <at 0.9986 0>
       ]}}})
  ]
  rendererClasses [&NiagaraMeshRendererProperties]
  renderers [{rendererClass=&NiagaraMeshRendererProperties, rendererIndex=0}]
 }
 Flash_Flare {
  bEnabled=false
  emitterUpdateScript [
   &EmitterState(
    Inactive Response=::NewEnumerator1
    Life Cycle Mode=::NewEnumerator1
    Loop Behavior=::NewEnumerator1
    Loop Delay=3
    Loop Duration=1
    UseLoopDelay=true
   )
   &SpawnBurst_Instantaneous(Use Spawn Probability=false, UseLoopCountLimit=false)
  ]
  particleSpawnScript [
   &InitializeParticle(
    Color=<rgba 2 0 1.3333 1>
    Lifetime=0.5
    Position Offset=<0 0 5>
    Sprite Size=<10000 10000>
    Sprite Size Mode=::NewEnumerator4
    UsePositionOffset=true
    Write Material Random=true
   )
   &CameraOffset(Camera Offset Amount=10)
   &DynamicMaterialParameters(
    Index 0 Param 1=0.2
    Index 0 Param 2=1
    Index 3 Param 1=1.5
    Index 3 Param 2=1
    Module.Index0Float0Write=true
    Module.Index0Float1Write=true
    Module.Index0Float2Write=true
    Module.Index0Float3Write=true
    Module.Index3Float0Write=true
    Module.Index3Float1Write=true
    Module.Index3Float2Write=true
    Module.Index3Float3Write=true
    Write Parameter Index 3=::NewEnumerator1
   )
   /SetVariables(Particles.GlowCameraOffset=30, Particles.GlowSize=<1500 1500>, Particles.GlowSpriteFacing:Vector3f=<0 0 1>) {isSetParameters=true, module=SetVariables_#1}
  ]
  particleUpdateScript [
   &ParticleState()
   &ScaleSpriteSize(
    Non-Uniform Curve Sprite Scale=DataInterface() {propertyValues {
      XCurve {keys [
        <at 0 3>
        {tangentMode=RCTM_User, time=0.8, value=2}
        {arriveTangentWeight=0.0667, interpMode=RCIM_Cubic, tangentMode=RCTM_User, time=1, value=0.2}
       ]}
      YCurve {keys [
        <at 0 1>
        {tangentMode=RCTM_User, time=0.8, value=2}
        {arriveTangentWeight=0.0667, interpMode=RCIM_Cubic, tangentMode=RCTM_User, time=1, value=0.2}
       ]}
     }}
    Scale Sprite Size Mode=::NewEnumerator3
   )
   &ScaleColor(
    Scale Alpha=&FloatFromCurve(FloatCurve=DataInterface() {propertyValues {Curve {keys [<at 0 1 RCIM_Cubic>,<at 1 0>]}}})
    Scale RGB=&VectorFromFloat(Value=&FloatFromCurve(FloatCurve=DataInterface() {propertyValues {Curve {keys [<at 0 5 RCIM_Cubic>,<at 0.75 0>]}}}))
    ScaleA=true
    ScaleRGB=true
   )
   /SetVariables(Particles.FlarePivotOffset:Vector2f=&Vector2DFromCurve(
     Vector2Curve=DataInterface() {propertyValues {XCurve {keys [<at 0 1>,<at 1 1>]}, YCurve {keys [<at 0 0.5>,<at 1 0.5458>]}}}
     ScaleCurve=<0.5 1>
    )) {enabled=false, isSetParameters=true, module=SetVariables_#2}
   /SetVariables(Particles.GlowColor=&MakeLinearColor(
     R:NiagaraFloat=0.223
     G:NiagaraFloat=0
     B:NiagaraFloat=1
     A:NiagaraFloat=&FloatFromCurve(
      FloatCurve=DataInterface() {propertyValues {Curve {keys [
          {arriveTangent=5.0727, interpMode=RCIM_Cubic, leaveTangent=5.0727, tangentMode=RCTM_User, time=0, value=0.1641}
          {
           arriveTangent=0.1472
           arriveTangentWeight=0.3592
           interpMode=RCIM_Cubic
           leaveTangent=0.1472
           leaveTangentWeight=0.3592
           tangentMode=RCTM_User
           tangentWeightMode=RCTWM_WeightedBoth
           time=0.5103
           value=1
          }
          {tangentMode=RCTM_User, time=1, value=0}
         ]}}}
      Scale Curve=0.751
     )
    )) {isSetParameters=true, module=SetVariables_#3}
  ]
  rendererClasses [&NiagaraSpriteRendererProperties]
  renderers [
   {rendererClass=&NiagaraSpriteRendererProperties}
   {rendererClass=&NiagaraSpriteRendererProperties, rendererIndex=1}
  ]
 }
 Flash_Flare001 {
  emitterUpdateScript [
   &EmitterState(
    Inactive Response=::NewEnumerator1
    Life Cycle Mode=::NewEnumerator1
    Loop Behavior=::NewEnumerator1
    Loop Delay=3
    Loop Duration=1
    UseLoopDelay=true
   )
   &SpawnBurst_Instantaneous(Use Spawn Probability=false, UseLoopCountLimit=false)
  ]
  particleSpawnScript [
   &InitializeParticle(
    Color=<rgba 0.9763 0 3 1>
    Lifetime=0.2
    Position Offset=<0 0 5>
    Sprite Size=<5000 5000>
    Sprite Size Mode=::NewEnumerator4
    UsePositionOffset=true
    Write Material Random=true
   )
   &CameraOffset(Camera Offset Amount=10)
   &DynamicMaterialParameters(
    Index 0 Param 1=0.2
    Index 0 Param 2=1
    Index 3 Param 1=1.5
    Index 3 Param 2=1
    Module.Index0Float0Write=true
    Module.Index0Float1Write=true
    Module.Index0Float2Write=true
    Module.Index0Float3Write=true
    Module.Index3Float0Write=true
    Module.Index3Float1Write=true
    Module.Index3Float2Write=true
    Module.Index3Float3Write=true
    Write Parameter Index 3=::NewEnumerator1
   )
   /SetVariables(Particles.GlowCameraOffset=30, Particles.GlowSize=<100 100>, Particles.GlowSpriteFacing:Vector3f=<0 0 1>) {isSetParameters=true, module=SetVariables_#1}
  ]
  particleUpdateScript [
   &ParticleState()
   &ScaleSpriteSize(
    Non-Uniform Curve Sprite Scale=DataInterface() {propertyValues {
      XCurve {keys [{tangentMode=RCTM_User, time=0, value=0.5}, <at 0.2 1.075>, <at 0.4 0.1>, <at 0.8 0.1>, {tangentMode=RCTM_User, time=1, value=0}]}
      YCurve {keys [{tangentMode=RCTM_User, time=0, value=0.5}, <at 0.1 1.025>, {tangentMode=RCTM_User, time=0.9998, value=1.258}]}
     }}
    Scale Sprite Size Mode=::NewEnumerator3
   )
   &SolveForcesAndVelocity(Clamp Velocity=false, Limit Acceleration=false)
   &ScaleColor(
    Scale Alpha=&FNTimeOfDayScalar(Night Scale=0.5, Morning Scale=1, Evening Scale=0.5)
    Scale RGB=&FN_Select_Vector_from_Time_Of_Day(Morning=<1 1 1>, Day=<1 1 1>, Evening=<0.8 0.8 0.8>, Night=<0.8 0.8 0.8>)
    ScaleA=true
    ScaleRGB=true
   )
   /SetVariables(Particles.FlarePivotOffset:Vector2f=&Vector2DFromCurve(
     Vector2Curve=DataInterface() {propertyValues {XCurve {keys [<at 0 1>,<at 1 1>]}, YCurve {keys [<at 0 0.5>,<at 1 0.5458>]}}}
     ScaleCurve=<0.5 1>
    )) {isSetParameters=true, module=SetVariables_#2}
   /SetVariables(Particles.GlowColor=&MakeLinearColor(
     R:NiagaraFloat=0.223
     G:NiagaraFloat=0
     B:NiagaraFloat=1
     A:NiagaraFloat=&FloatFromCurve(
      FloatCurve=DataInterface() {propertyValues {Curve {keys [
          {arriveTangent=5.0727, interpMode=RCIM_Cubic, leaveTangent=5.0727, tangentMode=RCTM_User, time=0, value=0.1641}
          {
           arriveTangent=0.1472
           arriveTangentWeight=0.3592
           interpMode=RCIM_Cubic
           leaveTangent=0.1472
           leaveTangentWeight=0.3592
           tangentMode=RCTM_User
           tangentWeightMode=RCTWM_WeightedBoth
           time=0.5103
           value=1
          }
          {tangentMode=RCTM_User, time=1, value=0}
         ]}}}
      Scale Curve=0.751
     )
    )) {isSetParameters=true, module=SetVariables_#3}
  ]
  rendererClasses [&NiagaraSpriteRendererProperties]
  renderers [
   {rendererClass=&NiagaraSpriteRendererProperties}
   {rendererClass=&NiagaraSpriteRendererProperties, rendererIndex=1}
  ]
 }
 GroundWaveSlow {
  emitterUpdateScript [
   &EmitterState(Life Cycle Mode=::NewEnumerator1, Loop Behavior=::NewEnumerator1, Loop Delay=1, Loop Duration=3, UseLoopDelay=true)
   &SpawnBurst_Instantaneous(Use Spawn Probability=false, UseLoopCountLimit=false)
   /SetVariables(Emitter.Linear Color:LinearColor=&ColorFromCurve(
     ColorCurve=DataInterface() {propertyValues {
       AlphaCurve {keys [<at 0 1 RCIM_Cubic>,<at 1 0 RCIM_Cubic>]}
       BlueCurve {keys [<at 0 1.7708 RCIM_Cubic>,<at 1 0 RCIM_Cubic>]}
       GreenCurve {keys [<at 0 0 RCIM_Cubic>,<at 1 0 RCIM_Cubic>]}
       RedCurve {keys [<at 0 1.7708 RCIM_Cubic>,<at 1 0 RCIM_Cubic>]}
      }}
     CurveIndex=$Emitter.NormalizedLoopAge
    )) {isSetParameters=true}
  ]
  particleSpawnScript [
   &InitializeParticle(
    Color=<rgba 1.7708 0 1.7708 1>
    Lifetime=4
    Mesh Scale=&Multiply_VectorByFloat(
     Vector=<1 1 0.25>
     Float=&FN_LOD_Distance_Remap(Start Distance=1000, End Distance=22500, Minimum Output=0.5, Maximum Output=1, Clamp Output=true)
    )
    Mesh Scale Mode=::NewEnumerator4
    Position Offset=<0 0 500>
    UsePositionOffset=true
    Write Material Random=true
   )
   &InitialMeshOrientation(
    Mesh Orientation Mode=::NewEnumerator2
    Rotation=&RandomRangeVector(Minimum=<0 0 0>, Maximum=<0 0 1>)
    Use Rotation Vector=true
   )
  ]
  particleUpdateScript [
   &ParticleState()
   &ScaleMeshSize(Scale Factor=&Multiply_VectorByFloat(
     Vector=<1 1 1>
     Float=&FloatFromCurve(
      FloatCurve=DataInterface() {propertyValues {Curve {keys [
          {leaveTangentWeight=0.3333, time=0, value=2}
          {arriveTangent=-5.6899, arriveTangentWeight=1.9257, interpMode=RCIM_Cubic, leaveTangent=-5.6899, tangentMode=RCTM_User, time=1, value=0.4}
         ]}}}
      Scale Curve=60
     )
    ))
   &ScaleColor(
    Scale Alpha=&FloatFromCurve(
     FloatCurve=DataInterface() {propertyValues {Curve {keys [
         {arriveTangent=26.1682, interpMode=RCIM_Cubic, leaveTangent=26.1682, tangentMode=RCTM_User, time=0, value=1}
         {arriveTangent=-3.7386, interpMode=RCIM_Cubic, leaveTangent=-3.7386, tangentMode=RCTM_User, time=0.1, value=2}
         {tangentMode=RCTM_User, time=1, value=0}
        ]}}}
     Scale Curve=2
    )
    ScaleA=true
    ScaleRGB=true
   )
  ]
  rendererClasses [&NiagaraMeshRendererProperties]
  renderers [{rendererClass=&NiagaraMeshRendererProperties, rendererIndex=0}]
 }
 SmokeMain {
  bEnabled=false
  emitterUpdateScript [
   &EmitterState(Life Cycle Mode=::NewEnumerator1, Loop Behavior=::NewEnumerator1, Loop Delay=3.1, Loop Duration=1, UseLoopDelay=true)
   &SpawnBurst_Instantaneous(Use Spawn Probability=false, UseLoopCountLimit=false)
  ]
  particleSpawnScript [&InitializeParticle(
    Lifetime=2
    Mesh Scale Mode=::NewEnumerator3
    Mesh Uniform Scale=50
    Position Offset=<0 0 -1000>
    UsePositionOffset=true
    Write Material Random=true
   )]
  particleUpdateScript [
   &ParticleState()
   &ScaleMeshSize(Scale Factor=&VectorFromCurve(VectorCurve=DataInterface() {propertyValues {
       XCurve {keys [<at 0 1 RCIM_Cubic>,<at 1 1 RCIM_Cubic>]}
       YCurve {keys [<at 0 1 RCIM_Cubic>,<at 1 1 RCIM_Cubic>]}
       ZCurve {keys [<at 0 0.5 RCIM_Cubic>,<at 1 1 RCIM_Cubic>]}
      }})) {enabled=false}
   &ScaleMeshSize(Scale Factor=&Multiply_VectorByFloat(
     Vector=<1.5 1.5 1>
     Float=&FloatFromCurve(FloatCurve=DataInterface() {propertyValues {Curve {
         defaultValue=1
         keys [
          {arriveTangent=4.1785, interpMode=RCIM_Cubic, leaveTangent=4.1785, tangentMode=RCTM_User, time=0, value=0}
          {arriveTangent=1.1179, leaveTangent=1.1179, time=0.1, value=1}
          {arriveTangent=0.0224, interpMode=RCIM_Cubic, leaveTangent=0.0224, tangentMode=RCTM_User, time=1, value=1.5}
         ]
        }}})
    )) {module=ScaleMeshSize001}
   &ScaleColor(
    Scale Alpha=&FloatFromCurve(FloatCurve=DataInterface() {propertyValues {Curve {
        defaultValue=0
        keys [
         {arriveTangent=-2.5249, interpMode=RCIM_Cubic, leaveTangent=-2.5249, tangentMode=RCTM_User, time=0, value=1}
         {arriveTangent=-1.3129, time=0.7, value=0.1}
         {arriveTangent=-0.0405, interpMode=RCIM_Cubic, leaveTangent=-0.0405, tangentMode=RCTM_User, time=1, value=0}
        ]
       }}})
    Scale RGB=&VectorFromCurve(VectorCurve=DataInterface() {propertyValues {XCurve {keys [<at 0 1 RCIM_Cubic>,<at 1 0>]}, YCurve {keys [<at 0 1 RCIM_Cubic>,<at 1 0>]}, ZCurve {keys [<at 0 1 RCIM_Cubic>,<at 1 0>]}}})
    ScaleA=true
    ScaleRGB=true
   )
  ]
  rendererClasses [&NiagaraMeshRendererProperties]
  renderers [{rendererClass=&NiagaraMeshRendererProperties, rendererIndex=0}]
 }
 Sparks_01 {
  bEnabled=false
  simTarget=GPUComputeSim
  emitterUpdateScript [
   &EmitterState(Life Cycle Mode=::NewEnumerator1, Loop Behavior=::NewEnumerator1, Loop Delay=3, Loop Duration=0.1, UseLoopDelay=true)
   &SpawnBurst_Instantaneous(Spawn Count=500, Use Spawn Probability=false, UseLoopCountLimit=false)
   &SpawnRate(SpawnRate=&Multiply_Float(A=500, B=1), Use Spawn Probability=true) {enabled=false}
  ]
  particleSpawnScript [
   &InitializeParticle(
    Color=<rgba 0.3542 0.024 0.3365 1>
    Lifetime Min=0.75
    Lifetime Mode=::NewEnumerator1
    Sprite Rotation Mode=::NewEnumerator2
    Sprite Size Mode=::NewEnumerator1
    Uniform Sprite Size Max=35
    Uniform Sprite Size Min=10
    UsePositionOffset=false
    Write Material Random=true
   ) {module=InitializeParticle001}
   &CameraOffset(Camera Offset Amount=25)
   &ShapeLocation(MaskSpawnGroup=false, Non Uniform Scale=<1 1 0>, Offset Mode=::NewEnumerator0, Override Seed=false, Sphere Radius=3500, UseOwnerScale=true) {module=ShapeLocation002}
   &PointForce(
    Force Strength=&RandomRangeFloat(Minimum=1000, Maximum=10000)
    Module.Use Falloff=false
    Module.UseLinearForce=false
   )
   &AddVelocity(Velocity=&RandomRangeVector(Minimum=<0 0 0>, Maximum=<1 1 3000>))
   &ApplyInitialForces()
  ]
  particleUpdateScript [
   &ParticleState()
   &CurlNoiseForce(Bias Noise=true, Noise Strength=2500, Pan Noise=true, Pan Noise Field=<1 0 1>)
   &GravityForce(Gravity=&RandomRangeVector(Minimum=<0 0 0>, Maximum=<0 0 490>))
   &Drag(Drag=2.5, Use Linear Drag=true, Use Rotational Drag=true)
   &ScaleColor(
    Scale Alpha=&FloatFromCurve(FloatCurve=DataInterface() {propertyValues {Curve {defaultValue=1, keys [{interpMode=RCIM_Cubic, tangentMode=RCTM_User, time=0.25, value=1}, {tangentMode=RCTM_User, time=1, value=0}]}}})
    Scale RGB=&VectorFromFloat(Value=&RandomRangeFloat(Minimum=5, Maximum=15))
    ScaleA=true
    ScaleRGB=true
   )
   &SolveForcesAndVelocity(Clamp Velocity=false, Limit Acceleration=false) {module=SolveForcesAndVelocity001}
   &ScaleSpriteSizeBySpeed(Max Scale Factor=<1 4>, Velocity Threshold=1500)
  ]
  rendererClasses [&NiagaraSpriteRendererProperties]
  renderers [{rendererClass=&NiagaraSpriteRendererProperties}]
 }
}
summary {
 systemName=NS_PistolCanary_Portal_Dissolve
 emitters [
  {bEnabled=false, emitterName=SmokeMain, rendererClasses [&NiagaraMeshRendererProperties]}
  {bEnabled=false, emitterName=Sparks_01, simTarget=GPUComputeSim, rendererClasses [&NiagaraSpriteRendererProperties]}
  {emitterName=EnergyTrail, rendererClasses [&NiagaraMeshRendererProperties]}
  {emitterName=CoreGlow, rendererClasses [&NiagaraSpriteRendererProperties]}
  {
   bEnabled=false
   emitterName=EnergyWave
   rendererClasses [&NiagaraMeshRendererProperties, &NiagaraLightRendererProperties]
  }
  {bEnabled=false, emitterName=Flash_Flare, rendererClasses [&NiagaraSpriteRendererProperties]}
  {bEnabled=false, emitterName=CoreGlow001, rendererClasses [&NiagaraMeshRendererProperties]}
  {emitterName=EnergyWave002, rendererClasses [&NiagaraSpriteRendererProperties, &NiagaraLightRendererProperties]}
  {emitterName=Flash, rendererClasses [&NiagaraMeshRendererProperties]}
  {emitterName=GroundWaveSlow, rendererClasses [&NiagaraMeshRendererProperties]}
  {
   emitterName=Cone_Shock_Light
   rendererClasses [
    &NiagaraMeshRendererProperties
    &NiagaraSpriteRendererProperties
    &NiagaraLightRendererProperties
   ]
  }
  {emitterName=Bolt, rendererClasses [&NiagaraMeshRendererProperties]}
  {emitterName=Flash_Flare001, rendererClasses [&NiagaraSpriteRendererProperties]}
 ]
 userVariables [
  User.Color {type=LinearColor, description="", defaultValue {struct=LinearColor, value=<rgba 1 0 0.1618 1>}}
  User.UseCustomColor {type=NiagaraFloat, defaultValue=0, description=""}
 ]
}
