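# Niagara effect for the tentacle emerge moment of the PistolCanary ground portal, showing a short magenta burst raised above the system origin.
# ChromaSphere_Mesh fires two instantaneous bursts, using spawn-group masks to vary lifetime, size, color and fade per layer.
# Its meshes expand with randomized orientation and animated material parameters, alongside a purple light modulated by a waveform.
# A green teleport-rays emitter with sprite, mesh and light renderers is present but disabled.
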NiagaraSystem /PistolCanary/Assets/FX/GroundPortal/NS_PistolCanary_Tentacle_Emerge
dependencies {
 usedDataInterfaces [
  &NiagaraDataInterfaceCurve
  &NiagaraDataInterfaceVector2DCurve
  &NiagaraDataInterfaceArrayFloat
  &NiagaraDataInterfaceArrayColor
 ]
 usedDynamicInputs [&Add_Float, &Lerp_Float]
 usedModules [
  &self:SystemSpawnScript.SystemScriptSource.SystemScriptGraph.NiagaraNodeAssignment_0.SetVariables_#1
  &SystemState
  &self:SystemSpawnScript.SystemScriptSource.SystemScriptGraph.NiagaraNodeAssignment_1.SetVariables_#2
  &self:SystemSpawnScript.SystemScriptSource.SystemScriptGraph.NiagaraNodeAssignment_2.SetVariables_#3
  &self:SystemSpawnScript.SystemScriptSource.SystemScriptGraph.NiagaraNodeAssignment_3.SetVariables_#4
 ]
}
emitters {
 ChromaSphere_Mesh {
  emitterUpdateScript [
   &EmitterState()
   &SpawnBurst_Instantaneous(Use Spawn Probability=false, UseLoopCountLimit=true)
   &SpawnBurst_Instantaneous(Spawn Group=1, Use Spawn Probability=false, UseLoopCountLimit=true) {module=SpawnBurst_Instantaneous001}
  ]
  particleSpawnScript [
   /SetVariables(Particles.VisibilityTag=&SpawnGroup()) {isSetParameters=true}
   &InitializeParticle(
    Color=&MaskLinearColorBySpawnGroup(Masked Linear Color=<rgba 25 4.3845 21.8754 2>, Passthrough Linear Color=<rgba 0.9962 0.1132 1 1>)
    Lifetime Max=&MaskFloatBySpawnGroup(Masked Float=0.8333, Passthrough Float=1)
    Lifetime Min:NiagaraFloat=&MaskFloatBySpawnGroup(Masked Float=0.3333, Passthrough Float=0.6667)
    Lifetime Mode=::NewEnumerator1
    Mesh Scale Mode=::NewEnumerator3
    Mesh Uniform Scale=&MaskFloatBySpawnGroup(Masked Float=45, Passthrough Float=&RandomRangeFloat(Minimum=30, Maximum=40))
    Position Offset=<0 0 400>
    UsePositionOffset=true
    Write Material Random=true
   )
   &InitialMeshOrientation(Use Rotation Vector=false)
   /SetVariables(Particles.MeshBOrient:Quat4f=&MakeQuatFromAxisAngle(Axis=<1 0 0>, AngleInDegrees=&RandomRangeFloat(Minimum=0, Maximum=360), NormalizeAxis=true)) {isSetParameters=true}
  ]
  particleUpdateScript [
   &ParticleState()
   &ScaleMeshSize(Scale Factor=&VectorFromFloat(Value=&MaskFloatBySpawnGroup(
      Masked Float=&FloatFromCurve(FloatCurve=DataInterface() {propertyValues {Curve {keys [
           {arriveTangent=0.7941, interpMode=RCIM_Cubic, leaveTangent=0.7941, leaveTangentWeight=1.1033, tangentMode=RCTM_User, time=0, value=0.5}
           <at 1 2>
          ]}}})
      Passthrough Float=&FloatFromCurve(FloatCurve=DataInterface() {propertyValues {Curve {keys [
           {arriveTangent=0.7941, interpMode=RCIM_Cubic, leaveTangent=0.7941, leaveTangentWeight=1.1033, tangentMode=RCTM_User, time=0, value=0.5}
           <at 1 1.5>
          ]}}})
     )))
   &DynamicMaterialParameters(
    Index 0 Param 1=&FloatFromCurve(FloatCurve=DataInterface() {propertyValues {Curve {keys [<at 0 0.07>,<at 1 0.25>]}}})
    Index 0 Param 2=&RandomRangeFloat(Minimum=0, Maximum=1)
    Index 0 Param 3=&RandomRangeFloat(Minimum=0, Maximum=1)
    Index 0 Param 4=&FloatFromCurve(FloatCurve=DataInterface() {propertyValues {Curve {keys [<at 0 1>,<at 1 3>]}}})
    Index 3 Param 1=&RandomRangeFloat(Minimum=3, Maximum=5)
    Module.Param0WriteEnabled=true
    Module.Param0WriteEnabled3=true
    Module.Param1WriteEnabled=true
    Module.Param1WriteEnabled3=true
    Module.Param2WriteEnabled=true
    Module.Param2WriteEnabled3=true
    Module.Param3WriteEnabled=true
    Module.Param3WriteEnabled3=true
    Write Parameter Index 3=true
   )
   &ScaleColor(
    Scale Alpha=&MaskFloatBySpawnGroup(
     Masked Float=&FloatFromCurve(FloatCurve=DataInterface() {propertyValues {Curve {keys [{interpMode=RCIM_Cubic, tangentMode=RCTM_User, time=0, value=1}, {interpMode=RCIM_Cubic, tangentMode=RCTM_User, time=1, value=0}]}}})
     Passthrough Float=&FloatFromCurve(FloatCurve=DataInterface() {propertyValues {Curve {keys [
          {interpMode=RCIM_Cubic, tangentMode=RCTM_User, time=0, value=0}
          {interpMode=RCIM_Cubic, tangentMode=RCTM_User, time=0.5, value=1}
          {tangentMode=RCTM_User, time=1, value=0}
         ]}}})
    )
    Scale RGB=&MaskVectorBySpawnGroup(
     Masked Vector=<1 1 1>
     Passthrough Vector=&VectorFromFloat(Value=&FloatFromCurve(
       FloatCurve=DataInterface() {propertyValues {Curve {keys [
           {arriveTangent=9.0725, interpMode=RCIM_Cubic, leaveTangent=9.0725, tangentMode=RCTM_User, time=0, value=0}
           {arriveTangent=-2.5249, interpMode=RCIM_Cubic, leaveTangent=-2.5249, tangentMode=RCTM_User, time=0.1408, value=0.9847}
           {arriveTangent=-0.0405, interpMode=RCIM_Cubic, leaveTangent=-0.0405, tangentMode=RCTM_User, time=1, value=0}
          ]}}}
       Scale Curve=5
      ))
    )
    ScaleA=true
    ScaleRGB=true
   )
   &ScaleSpriteSize(Uniform Curve Sprite Scale=DataInterface() {propertyValues {Curve {keys [
        <at 0 0 RCIM_Cubic>
        {arriveTangent=3.3072, interpMode=RCIM_Cubic, leaveTangent=3.3072, tangentMode=RCTM_User, time=0.011, value=0.85}
        {arriveTangent=0.6813, interpMode=RCIM_Cubic, leaveTangent=0.6813, tangentMode=RCTM_User, time=0.05, value=0.96}
        {arriveTangent=0.0129, interpMode=RCIM_Cubic, leaveTangent=0.0129, tangentMode=RCTM_User, time=0.171, value=0.98}
        <at 1 1 RCIM_Cubic>
       ]}}})
   &FN_Light_Attributes(
    Color=<rgba 2.4945 0.1507 5 1>
    Create Color Attribute=true
    Create Exponent Attribute=true
    Create Position Attribute=true
    Create Radius Attribute=true
    Exponent=2
    Position=&Add_Vector(A=Linked() {linkedVariable {name=Particles.Position, type=NiagaraPosition}}, B=<25 0 0>)
    Radius=100
    Scale Color=&FloatFromCurve(
     FloatCurve=DataInterface() {propertyValues {Curve {keys [
         {interpMode=RCIM_Cubic, tangentMode=RCTM_User, time=0, value=0}
         {interpMode=RCIM_Cubic, tangentMode=RCTM_User, time=0.5, value=1}
         {tangentMode=RCTM_User, time=1, value=0}
        ]}}}
     Scale Curve=&Waveform(
      Global Amplitude Scale=150
      "Waveform [1]"=::NewEnumerator2
      "[1] Frequency"=1
      "[1] Amplitude Min/Max"=<0.5 1.5>
      "[2] Phase"=$Emitter.Age
      "[3] Phase"=$Emitter.Age
     )
    )
   )
  ]
  rendererClasses [&NiagaraMeshRendererProperties, &NiagaraLightRendererProperties]
  renderers [
   {rendererClass=&NiagaraMeshRendererProperties, rendererIndex=0}
   {rendererClass=&NiagaraMeshRendererProperties, rendererIndex=1}
   {rendererClass=&NiagaraLightRendererProperties, rendererIndex=2}
  ]
 }
 Packaged_Teleport_Out_Rays {
  bEnabled=false
  emitterUpdateScript [
   &EmitterState()
   &SpawnRate(SpawnRate=20, Use Spawn Probability=false)
   &SpawnRate(
    SpawnGroup=1
    SpawnRate=&FloatFromCurve(FloatCurve=DataInterface() {propertyValues {Curve {keys [<at 0 10 RCIM_Constant>,<at 2 0 RCIM_Constant>]}}}, CurveIndex=$System.Age)
    Use Spawn Probability=false
   ) {module=SpawnRate001}
   &SpawnRate(
    SpawnGroup=2
    SpawnRate=&FloatFromCurve(FloatCurve=DataInterface() {propertyValues {Curve {keys [<at 0 6 RCIM_Constant>,<at 1.4 0 RCIM_Constant>]}}}, CurveIndex=$System.Age)
    Use Spawn Probability=false
   ) {module=SpawnRate002}
  ]
  particleSpawnScript [
   /SetVariables(Particles.VisibilityTag=&SpawnGroup()) {isSetParameters=true}
   &InitializeParticle(
    Color=&MakeLinearColorFromVectorAndFloat(
     "Vector (RGB)"=&MakeVectorFromLinearColorRGB(Color=&SelectLinearColorFromArray(
       Array Sampling Mode=::NewEnumerator1
       Direct Array Index=&SpawnGroup()
       Color Selection Array=DataInterface() {propertyValues {ColorData [<rgba 0 15 10.1968 1>, <rgba 0 1 0.2242 1>, <rgba 0 1 0.315 1>]}}
       Random Seed=?
      ))
     "Float (Alpha)"=&MaskFloatBySpawnGroup(Masked Float=&RandomRangeFloat(Minimum=0.15, Maximum=0.25), Passthrough Float=1, Spawn Group Mask=1)
    )
    Lifetime Max=&SelectFloatFromArray(
     Direct Array Index=&SpawnGroup()
     Float Selection Array=DataInterface() {propertyValues {FloatData [0.36,1.65,0.6]}}
    )
    Lifetime Min:NiagaraFloat=&SelectFloatFromArray(
     Direct Array Index=&SpawnGroup()
     Float Selection Array=DataInterface() {propertyValues {FloatData [0.24,1.2,0.3]}}
    )
    Lifetime Mode=::NewEnumerator1
    Mesh Scale Max=<3.5 1000 1.25>
    Mesh Scale Min=<2 1000 20>
    Mesh Scale Mode=::NewEnumerator2
    Position Offset=&MaskVectorBySpawnGroup(Masked Vector=<50 0 0>, Passthrough Vector=<0 0 0>, Spawn Group Mask=2)
    Sprite Rotation Mode=::NewEnumerator2
    Sprite Size Mode=::NewEnumerator1
    Uniform Sprite Size Max=20
    Uniform Sprite Size Min=13
    UsePositionOffset=true
    Write Material Random=true
   )
   &CameraOffset(Camera Offset Amount=&MaskFloatBySpawnGroup(Masked Float=10, Passthrough Float=0))
   &InitialMeshOrientation(
    Mesh Orientation Mode=::NewEnumerator2
    Orientation Vector=<0 0 1>
    Rotation=&RandomRangeVector(Minimum=<-0.025 -0.025 0>, Maximum=<0.025 0.025 0>)
    Rotation Coordinate Space=::NewEnumerator0
    Use Rotation Vector=true
   )
  ]
  particleUpdateScript [
   &ParticleState()
   &ScaleMeshSize(Scale Factor=&MakeVector(
     X=&FloatFromCurve(
      FloatCurve=DataInterface() {propertyValues {Curve {keys [
          {arriveTangent=1.6012, interpMode=RCIM_Cubic, leaveTangent=1.6012, tangentMode=RCTM_User, time=0, value=0.3}
          {arriveTangent=0.7, interpMode=RCIM_Cubic, leaveTangent=0.7, time=0.4, value=0.8667}
          {interpMode=RCIM_Cubic, leaveTangent=1.5011, time=1, value=1}
         ]}}}
      Scale Curve=30
     )
     Y=&FloatFromCurve(FloatCurve=DataInterface() {propertyValues {Curve {keys [<at 0 1>,<at 3 2>]}}}, CurveIndex=$Emitter.Age)
     Z=1
    ))
   &ScaleSpriteSize(Uniform Curve Sprite Scale=DataInterface() {propertyValues {Curve {keys [{tangentMode=RCTM_User, time=0, value=0.2}, {tangentMode=RCTM_User, time=1, value=1}]}}})
   &UpdateMeshOrientation(Rotation Rate=&RandomRangeFloat(Minimum=-0.1, Maximum=0.1), Rotation Vector=<0 0 1>)
   &DynamicMaterialParameters(
    Index 0 Param 1=&RandomRangeFloat(Minimum=-1, Maximum=1)
    Index 0 Param 2=1
    Module.Index0Param0Write=true
    Module.Index0Param1Write=true
    Module.Index0Param2Write=true
    Module.Index0Param3Write=true
   )
   &ScaleColor(
    Scale Alpha=&MaskFloatBySpawnGroup(
     Masked Float=&FloatFromCurve(FloatCurve=DataInterface() {propertyValues {Curve {keys [
          {arriveTangent=19.638, interpMode=RCIM_Cubic, leaveTangent=19.638, tangentMode=RCTM_User, time=0, value=0}
          {arriveTangentWeight=0.0612, interpMode=RCIM_Cubic, leaveTangentWeight=0.3364, tangentMode=RCTM_User, time=0.1, value=1}
          {arriveTangentWeight=0.2741, interpMode=RCIM_Cubic, leaveTangentWeight=0.1744, tangentMode=RCTM_User, time=0.65, value=1}
          {arriveTangent=-4.2611, interpMode=RCIM_Cubic, leaveTangent=-4.2611, tangentMode=RCTM_User, time=1, value=0}
         ]}}})
     Passthrough Float=&MaskFloatBySpawnGroup(
      Masked Float=&FloatFromCurve(
       FloatCurve=DataInterface() {propertyValues {Curve {keys [
           {arriveTangent=6.3142, interpMode=RCIM_Cubic, leaveTangent=6.3142, tangentMode=RCTM_User, time=0, value=0}
           <at 0.2 1 RCIM_Cubic>
           <at 1 0 RCIM_Cubic>
          ]}}}
       Scale Curve=4
      )
      Passthrough Float=1
      Spawn Group Mask=2
     )
     Spawn Group Mask=1
    )
    ScaleA=true
    ScaleRGB=true
   )
   /SetVariables(Particles.LightRadius=&RandomRangeFloat(Minimum=350, Maximum=500)) {isSetParameters=true}
  ]
  rendererClasses [
   &NiagaraSpriteRendererProperties
   &NiagaraMeshRendererProperties
   &NiagaraLightRendererProperties
  ]
  renderers [
   {rendererClass=&NiagaraSpriteRendererProperties}
   {rendererClass=&NiagaraMeshRendererProperties, rendererIndex=1}
   {rendererClass=&NiagaraLightRendererProperties, rendererIndex=2}
  ]
 }
}
summary {
 systemName=NS_PistolCanary_Tentacle_Emerge
 emitters [
  {
   emitterName=ChromaSphere_Mesh
   rendererClasses [&NiagaraMeshRendererProperties, &NiagaraLightRendererProperties]
  }
  {
   bEnabled=false
   emitterName=Packaged_Teleport_Out_Rays
   rendererClasses [
    &NiagaraSpriteRendererProperties
    &NiagaraMeshRendererProperties
    &NiagaraLightRendererProperties
   ]
  }
 ]
 userVariables [
  User.AltTimeline {type=NiagaraFloat, defaultValue=0, description=""}
  User.bIsFrontEnd {type=NiagaraBool}
  User.bIsFrontEndPreview {type=NiagaraBool}
  User.BlackHoleEnd {type=Vector3f, description="", defaultValue {struct=Vector3f, value=<0 50 -150>}}
  User.BlackHoleStart {type=NiagaraPosition, description="", defaultValue {}}
  User.ExplosionScale {type=NiagaraFloat, defaultValue=1.2, description=""}
  User.Fade {type=NiagaraFloat, defaultValue=1, description=""}
  User.GlobalScale {type=NiagaraFloat, defaultValue=33, description=""}
  User.HandGlow {type=NiagaraFloat, defaultValue=0, description=""}
  User.Intensity {type=NiagaraFloat, defaultValue=50, description=""}
  User.Lifetime {type=NiagaraFloat, defaultValue=3, description=""}
  User.Open {type=NiagaraFloat, defaultValue=2, description=""}
  User.OverallWidth {type=NiagaraFloat, defaultValue=3720.3586, description=""}
  User.ParaGlideLeanAlpha {type=NiagaraFloat, defaultValue=0, description=""}
  User.PlayerSpeed {type=NiagaraFloat, defaultValue=0, description=""}
  User.PortalSize {type=NiagaraFloat, defaultValue=1, description=""}
  User.PrimaryColor {type=LinearColor, description="", defaultValue {struct=LinearColor, value=<rgba 5 0.2882 0.6068 1>}}
  User.RemoveBeam {type=NiagaraFloat, defaultValue=0, description=""}
  User.Scale {type=NiagaraFloat, defaultValue=1, description=""}
  User.SkeletalMesh {
   type=NiagaraDataInterfaceSkeletalMesh
   defaultValue {
    struct=NiagaraExt_VariableValue_DataInterface
    value {dataInterface=&self:NiagaraDataInterfaceSkeletalMesh_0}
   }
  }
  User.SplineUpdatedEndPos {type=Vector3f, description="", defaultValue {struct=Vector3f, value=<0 0 -25920.5664>}}
  User.SplineUpdatedStartPos {type=NiagaraPosition, description="", defaultValue {}}
 ]
}
