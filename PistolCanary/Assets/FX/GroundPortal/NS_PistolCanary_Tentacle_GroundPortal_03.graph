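# Niagara ground portal effect for the PistolCanary tentacle, built entirely from mesh-rendered emitters.
# Two BlackHoleEnergy bursts place the portal meshes, while EnergyWave001 adds a material-animated wave and plays the swirl audio at User.SFXVolume.
# Lightning_Primary spawns flickering pink bolts around a flattened sphere, randomly long- or short-lived, scaled by User.WidthMult and User.ColorMult.
# Axis and both Trail_ring emitters are disabled, and most exposed User parameters are unreferenced in the emitter modules.
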NiagaraSystem /PistolCanary/Assets/FX/GroundPortal/NS_PistolCanary_Tentacle_GroundPortal_03
dependencies {
 usedDataInterfaces [
  &NiagaraDataInterfaceAudioPlayer
  &NiagaraDataInterfaceCurve
  &NiagaraDataInterfaceColorCurve
  &NiagaraDataInterfaceArrayColor
 ]
 usedDynamicInputs []
 usedModules [
  &self:SystemSpawnScript.SystemScriptSource.SystemScriptGraph.NiagaraNodeAssignment_0.SetVariables_#1
  &SystemState
  &self:SystemSpawnScript.SystemScriptSource.SystemScriptGraph.NiagaraNodeAssignment_1.SetVariables_#2
  &self:SystemSpawnScript.SystemScriptSource.SystemScriptGraph.NiagaraNodeAssignment_2.SetVariables_#3
  &self:SystemSpawnScript.SystemScriptSource.SystemScriptGraph.NiagaraNodeAssignment_3.SetVariables_#4
  &self:SystemSpawnScript.SystemScriptSource.SystemScriptGraph.NiagaraNodeAssignment_4.SetVariables_#5
 ]
}
emitters {
 Axis {
  bEnabled=false
  emitterUpdateScript [
   &EmitterState()
   &SpawnBurst_Instantaneous(Use Spawn Probability=false, UseLoopCountLimit=false)
  ]
  particleSpawnScript [&InitializeParticle(
    Color Mode=::NewEnumerator0
    Lifetime=5
    Mesh Scale Mode=::NewEnumerator3
    Mesh Uniform Scale=250
    UsePositionOffset=false
    Write Material Random=true
   )]
  particleUpdateScript [&ParticleState(Kill Particles When Lifetime Has Elapsed=false)]
  rendererClasses [&NiagaraMeshRendererProperties]
  renderers [{rendererClass=&NiagaraMeshRendererProperties, rendererIndex=0}]
 }
 BlackHoleEnergy {
  emitterUpdateScript [
   &EmitterState()
   &SpawnRate(Use Spawn Probability=false) {enabled=false}
   &SpawnBurst_Instantaneous(Use Spawn Probability=false, UseLoopCountLimit=false)
  ]
  particleSpawnScript [
   /Niagara/Modules/Spawn/Initialization/InitializeParticle(
    Color=<rgba 1.5 0.3114 1.0299 1.5>
    Lifetime=$User.Lifetime
    Mesh Scale=<0.1 1.5 1.5>
    Module.Write Color=true
    Module.Write Lifetime=true
    Module.Write Mass=false
    Module.Write Material Random=true
    Module.Write Position=false
    Module.Write Scale=true
    Module.Write SpriteRotation=false
    Module.Write SpriteSize=false
    Position=&SimulationPosition()
    Sprite Rotation=&UniformRangedFloat(Minimum=0, Maximum=1)
   ) {enabled=false}
   /Niagara/Modules/Spawn/Initialization/V2/InitializeParticle(
    Color=<rgba 2 2 2 1>
    Mesh Scale=<1 1 1.2>
    Mesh Scale Mode=::NewEnumerator4
    Position Offset=<0 -100 180>
    UsePositionOffset=true
    Write Material Random=true
   ) {module=InitializeParticle001}
   &InitialMeshOrientation(
    Orientation Axis=<0 0 1>
    Rotation=&MakeVector(X=0, Y=0, Z=0.06)
    Use Orientation Vector=true
    Use Rotation Vector=false
   )
  ]
  particleUpdateScript [
   &ParticleState(Kill Particles When Lifetime Has Elapsed=false)
   &ScaleMeshSize(Scale Factor=&Multiply_VectorByFloat(Vector=<1 1 1>, Float=0.73)) {module=ScaleMeshSize001}
   &ScaleColor(Scale Alpha=6.25, Scale RGB=<1.25 1 1>, ScaleA=true, ScaleRGB=true)
   &DynamicMaterialParameters(Write Parameter Index 0=::NewEnumerator0)
  ]
  rendererClasses [&NiagaraMeshRendererProperties]
  renderers [{rendererClass=&NiagaraMeshRendererProperties, rendererIndex=0}]
 }
 BlackHoleEnergy002 {
  emitterUpdateScript [
   &EmitterState()
   &SpawnBurst_Instantaneous(Use Spawn Probability=false, UseLoopCountLimit=false)
  ]
  particleSpawnScript [&InitializeParticle(
    Color=<rgba 2 2 2 1>
    Mesh Scale=<30 30 30>
    Mesh Scale Mode=::NewEnumerator4
    Position Offset=<0 0 200>
    UsePositionOffset=true
    Write Material Random=true
   ) {module=InitializeParticle001}]
  particleUpdateScript [&ParticleState(Kill Particles When Lifetime Has Elapsed=false)]
  rendererClasses [&NiagaraMeshRendererProperties]
  renderers [{rendererClass=&NiagaraMeshRendererProperties, rendererIndex=0}]
 }
 EnergyWave001 {
  emitterSpawnScript [&InitializeAudioPlayer(Audio Player=DataInterface() {propertyValues {
      Attenuation=null
      Concurrency=null
      MaxPlaysPerTick=10
      SoundToPlay=/PistolCanary_Event_Audio/Assets/Section01/MSS_PistolCanary_PortalSwirlLoop_Stereo
      bAllowLoopingOneShotSounds=false
      bLimitPlaysPerTick=true
      bOnlyActiveDuringGameplay=false
      bStopWhenComponentIsDestroyed=true
      ConfigurationUserParameter {parameter {name=None, typeDefHandle {registeredTypeIndex=1}, varData []}}
      ParameterNames []
     }})]
  emitterUpdateScript [
   &EmitterState(Life Cycle Mode=::NewEnumerator1, Loop Behavior=::NewEnumerator1, Loop Duration=1, UseLoopDelay=false)
   &SpawnBurst_Instantaneous(Use Spawn Probability=false, UseLoopCountLimit=false)
  ]
  particleSpawnScript [
   &InitializeParticle(
    Color=<rgba 0.7293 0.0378 1 1>
    Lifetime=4.5
    Mesh Scale Mode=::NewEnumerator3
    Mesh Uniform Scale=8
    Position Offset=<0 0 100>
    UsePositionOffset=true
    Write Material Random=true
   )
   &InitialMeshOrientation(Mesh Orientation Mode=::NewEnumerator2, Orientation Vector=<0 0 1>, Use Rotation Vector=true)
  ]
  particleUpdateScript [
   &ParticleState(Kill Particles When Lifetime Has Elapsed=false)
   &DynamicMaterialParameters(
    Index 0 Param 1=&FloatFromCurve(
     FloatCurve=DataInterface() {propertyValues {Curve {keys [{arriveTangent=-2.1015, interpMode=RCIM_Cubic, leaveTangent=-2.1015, tangentMode=RCTM_User, time=0, value=0.5}, <at 1 -0.15>]}}}
     Scale Curve=&RandomRangeFloat(Minimum=0.5, Maximum=1.5)
    )
    Index 0 Param 2=&FloatFromCurve(
     FloatCurve=DataInterface() {propertyValues {Curve {keys [{arriveTangent=-0.3464, interpMode=RCIM_Cubic, leaveTangent=-0.3464, tangentMode=RCTM_User, time=0, value=1}, <at 1 0.9>]}}}
     Scale Curve=&RandomRangeFloat(Minimum=0.5, Maximum=1)
    )
    Index 0 Param 3=&FloatFromCurve(
     FloatCurve=DataInterface() {propertyValues {Curve {keys [
         {interpMode=RCIM_Cubic, tangentMode=RCTM_Break, time=0, value=0}
         {arriveTangent=0.4478, interpMode=RCIM_Cubic, leaveTangent=0.4478, tangentMode=RCTM_User, time=0.15, value=1}
         {interpMode=RCIM_Cubic, tangentMode=RCTM_Break, time=1, value=1.2}
        ]}}}
     Scale Curve=&RandomRangeFloat(Minimum=0.5, Maximum=1)
    )
    Module.Param0WriteEnabled=true
    Module.Param1WriteEnabled=true
    Module.Param2WriteEnabled=true
    Module.Param3WriteEnabled=false
   )
   &ScaleColor(
    Scale Alpha=&FloatFromCurve(FloatCurve=DataInterface() {propertyValues {Curve {keys [{interpMode=RCIM_Cubic, tangentMode=RCTM_User, time=0.5, value=1}, {interpMode=RCIM_Cubic, tangentMode=RCTM_User, time=1, value=1}]}}})
    Scale RGB=&Multiply_VectorByFloat(
     Vector=&MakeVectorFromLinearColorRGB(Color=&ColorFromCurve(ColorCurve=DataInterface() {propertyValues {
         AlphaCurve {keys [<at 0 1 RCIM_Cubic>]}
         BlueCurve {keys [<at 0.1059 1 RCIM_Cubic>,<at 0.85 1 RCIM_Cubic>,<at 0.9998 1>]}
         GreenCurve {keys [
           <at 0.1059 0.833 RCIM_Cubic>
           {arriveTangent=-0.4934, interpMode=RCIM_Cubic, leaveTangent=-0.4934, time=0.85, value=0.5547}
           <at 0.9998 0.3919>
          ]}
         RedCurve {keys [
           <at 0.1059 0.3906 RCIM_Cubic>
           {arriveTangent=-0.0583, interpMode=RCIM_Cubic, leaveTangent=-0.0583, time=0.85, value=0.4062}
           <at 0.9998 0.3385>
          ]}
        }}))
     Float=1
    )
    ScaleA=true
    ScaleRGB=false
   )
   /Niagara/Modules/Audio/PlayLoopingAudio(Update volume every frame=true, Volume:NiagaraFloat=$User.SFXVolume)
  ]
  rendererClasses [&NiagaraMeshRendererProperties]
  renderers [{rendererClass=&NiagaraMeshRendererProperties, rendererIndex=0}]
 }
 Lightning_Primary {
  emitterUpdateScript [
   &EmitterState(Life Cycle Mode=::NewEnumerator1, Loop Duration=1, UseLoopDelay=false)
   &SpawnRate(SpawnRate=12.5, Use Spawn Probability=false) {module=SpawnRate001}
  ]
  particleSpawnScript [
   /SetVariables(Particles.isShort:NiagaraFloat=&FloatFromCurve(
     FloatCurve=DataInterface() {propertyValues {Curve {keys [<at 0 0 RCIM_Constant>,<at 0.8 1 RCIM_Constant>]}}}
     CurveIndex=&RandomRangeFloat(Minimum=0, Maximum=1)
    )) {isSetParameters=true}
   &InitializeParticle(
    Color=&SelectLinearColorFromArray(
     Array Sampling Mode=::NewEnumerator1
     Direct Array Index=&RandomRangeInt(Minimum=0, Maximum=1)
     Color Selection Array=DataInterface() {propertyValues {ColorData [<rgba 0.5 0.0145 0.2765 1>, <rgba 0.5 0 0.2161 1>]}}
     Random Seed=?
    )
    Lifetime=&Lerp_Float(
     A=&RandomRangeFloat(Minimum=0.5, Maximum=0.8)
     B=&RandomRangeFloat(Minimum=0.3, Maximum=0.4)
     Alpha=$Particles.isShort
    )
    Mesh Scale Mode=::NewEnumerator3
    Mesh Uniform Scale=&MakeFloatFromInt(Int=&Multiply_Int(A=&RandomRangeInt(Minimum=4, Maximum=12), B=4))
    Position Mode=::NewEnumerator1
    UsePositionOffset=false
    Write Material Random=true
   ) {module=InitializeParticle001}
   &DynamicMaterialParameters(
    Index 0 Param 1=1
    Index 0 Param 2=0.15
    Index 0 Param 3=1
    Index 0 Param 4=1
    Module.Param0WriteEnabled=true
    Module.Param1WriteEnabled=true
    Module.Param2WriteEnabled=true
    Module.Param3WriteEnabled=true
   )
   &ShapeLocation(
    MaskSpawnGroup=false
    Non Uniform Scale=<1 1 0>
    Offset Mode=::NewEnumerator0
    Override Seed=false
    Sphere Radius=3000
    Sphere Surface Distribution=1
    UseOwnerScale=true
   ) {module=ShapeLocation002}
   &AddVelocity(Module.Use Falloff Distance=false, Override Seed=false, Velocity Mode=::NewEnumerator1, Velocity Speed=0.25)
   &InitialMeshOrientation(
    Mesh Orientation Mode=::NewEnumerator7
    Rotation=&RandomRangeVector(Minimum=<0 0 0>, Maximum=<0 0 0>)
    Use Rotation Vector=true
   ) {module=InitialMeshOrientation002}
  ]
  particleUpdateScript [
   &SolveForcesAndVelocity(Clamp Velocity=false, Limit Acceleration=false)
   &ParticleState()
   &ScaleMeshSize(Scale Factor=&Multiply_VectorByFloat(
     Vector=&MakeVector(
      X=&Multiply_Float(
       A=&FloatFromCurve(FloatCurve=DataInterface() {propertyValues {Curve {keys [{interpMode=RCIM_Cubic, tangentMode=RCTM_User, time=0, value=1}, {interpMode=RCIM_Cubic, tangentMode=RCTM_User, time=1, value=1.5}]}}})
       B=&RandomRangeFloat(Minimum=0.5, Maximum=0.8)
      )
      Y=1
      Z=1
     )
     Float=&MakeFloatFromVector(
      Vector=&VectorFromFloat(Value=&Multiply_Float(
        A=&RandomRangeFloat(Minimum=0.8, Maximum=1.2)
        B=&Multiply_Float(
         A=0.1
         B=&FloatFromCurve(FloatCurve=DataInterface() {propertyValues {Curve {keys [
              {arriveTangent=0.2102, interpMode=RCIM_Cubic, leaveTangent=0.2102, tangentMode=RCTM_User, time=0, value=0.9}
              {interpMode=RCIM_Cubic, tangentMode=RCTM_User, time=1, value=1}
             ]}}})
        )
       ))
      Channel=::NewEnumerator0
     )
    )) {enabled=false}
   &ScaleMeshSize(Scale Factor=&Multiply_VectorByFloat(
     Vector=&MakeVector(X=0.75, Y=1, Z=1)
     Float=&MakeFloatFromVector(
      Vector=&VectorFromFloat(Value=&Multiply_Float(
        A=&RandomRangeFloat(Minimum=0.8, Maximum=1.2)
        B=&Multiply_Float(
         A=0.1
         B=&FloatFromCurve(FloatCurve=DataInterface() {propertyValues {Curve {keys [
              {arriveTangent=0.2102, interpMode=RCIM_Cubic, leaveTangent=0.2102, tangentMode=RCTM_User, time=0, value=0.9}
              {interpMode=RCIM_Cubic, tangentMode=RCTM_User, time=1, value=1}
             ]}}})
        )
       ))
      Channel=::NewEnumerator0
     )
    )) {module=ScaleMeshSize002}
   &DynamicMaterialParameters(
    Index 0 Param 1=&Lerp_Float(
     A=&FloatFromCurve(FloatCurve=DataInterface() {propertyValues {Curve {keys [{tangentMode=RCTM_User, time=0, value=1}, <at 0.2 0.5>, <at 0.8 0.5>, {tangentMode=RCTM_User, time=1, value=0}]}}})
     B=0.4
     Alpha=$Particles.isShort
    )
    Index 0 Param 2=&FloatFromCurve(
     FloatCurve=DataInterface() {propertyValues {Curve {keys [
         {interpMode=RCIM_Cubic, leaveTangent=-1.1405, tangentMode=RCTM_Break, time=0, value=0.5}
         {arriveTangent=4.4936, interpMode=RCIM_Cubic, leaveTangent=-5.3647, tangentMode=RCTM_Break, time=0.4833, value=1}
         <at 1 0>
        ]}}}
     Scale Curve=&Multiply_Float(A=10, B=$User.WidthMult)
    )
    Index 0 Param 3=120
    Index 1 Param 1=&RandomRangeFloat(Minimum=0, Maximum=1)
    Index 1 Param 2=1
    Index 1 Param 3=1
    Index 1 Param 4=1
    Module.Param0WriteEnabled=true
    Module.Param0WriteEnabled1=true
    Module.Param1WriteEnabled=true
    Module.Param1WriteEnabled1=true
    Module.Param2WriteEnabled=true
    Module.Param2WriteEnabled1=true
    Module.Param3WriteEnabled=false
    Module.Param3WriteEnabled1=true
    Write Parameter Index 1=true
   ) {module=DynamicMaterialParameters001}
   &ScaleColor(
    Scale Alpha=&FloatFromCurve(FloatCurve=DataInterface() {propertyValues {Curve {keys [
         {interpMode=RCIM_Cubic, tangentMode=RCTM_User, time=0, value=0}
         {interpMode=RCIM_Cubic, tangentMode=RCTM_User, time=0.1258, value=1}
         <at 0.341 0>
         {interpMode=RCIM_Cubic, tangentMode=RCTM_User, time=0.496, value=0.5}
         <at 1 0>
        ]}}})
    Scale RGB=&VectorFromFloat(Value=&FloatFromCurve(
      FloatCurve=DataInterface() {propertyValues {Curve {keys [
          {arriveTangent=0.1143, interpMode=RCIM_Cubic, leaveTangent=0.1143, tangentMode=RCTM_User, time=0.018, value=4.9326}
          {arriveTangent=-6.0507, interpMode=RCIM_Cubic, leaveTangent=-6.0507, time=0.1778, value=0.8567}
          {arriveTangent=-0.5645, interpMode=RCIM_Cubic, leaveTangent=-0.5645, tangentMode=RCTM_User, time=0.2493, value=3.533}
          {arriveTangent=-22.9399, interpMode=RCIM_Cubic, leaveTangent=-22.9399, time=0.3224, value=1.7136}
          {arriveTangent=2.8763, interpMode=RCIM_Cubic, leaveTangent=2.8763, time=0.3773, value=0.5981}
          {arriveTangent=-1.8312, interpMode=RCIM_Cubic, leaveTangent=-1.8312, time=0.4693, value=2.1361}
          {arriveTangent=-2.2307, interpMode=RCIM_Cubic, leaveTangent=-2.2307, time=0.5983, value=0.1933}
          {arriveTangent=1.5587, interpMode=RCIM_Cubic, leaveTangent=1.5587, time=0.7391, value=1.5342}
          {arriveTangent=-2.0478, interpMode=RCIM_Cubic, leaveTangent=-2.0478, time=0.8629, value=0.6057}
          {arriveTangent=-0.0659, interpMode=RCIM_Cubic, leaveTangent=-0.0659, tangentMode=RCTM_User, time=1, value=1}
         ]}}}
      Scale Curve=$User.ColorMult
     ))
    ScaleA=true
    ScaleRGB=true
   ) {module=ScaleColor001}
  ]
  rendererClasses [&NiagaraMeshRendererProperties]
  renderers [{rendererClass=&NiagaraMeshRendererProperties, rendererIndex=0}]
 }
 Trail_ring {
  bEnabled=false
  emitterUpdateScript [
   &EmitterState(Scalability Mode=::NewEnumerator1)
   &SpawnRate(Use Spawn Probability=false)
  ]
  particleSpawnScript [
   &InitializeParticle(
    Color Maximum=<rgba 1 0.0217 0.3635 1>
    Color Minimum=<rgba 0.2048 0 1 1>
    Color Mode=::NewEnumerator2
    Lifetime Max=3
    Lifetime Min=2
    Lifetime Mode=::NewEnumerator1
    Mesh Scale Mode=::NewEnumerator3
    Mesh Uniform Scale=40
    Position Offset=&MakeVector(
     X=0
     Y=0
     Z=&Add_Float(A=-600, B=&RandomRangeFloat(Minimum=-100, Maximum=200))
    )
    UsePositionOffset=true
    Write Material Random=true
   )
   &InitialMeshOrientation(Mesh Orientation Mode=::NewEnumerator2, Rotation=<0.25 0 0>, Use Rotation Vector=true)
   &InitialMeshOrientation(
    Mesh Orientation Mode=::NewEnumerator2
    Rotation=&RandomRangeVector(Minimum=<0 0 0>, Maximum=<0 0 1>)
    Use Rotation Vector=true
   ) {module=InitialMeshOrientation001}
   &InitialMeshOrientation(Mesh Orientation Mode=::NewEnumerator2, Rotation=<0 0.05 0>, Use Rotation Vector=true) {module=InitialMeshOrientation002}
   &DynamicMaterialParameters(
    Index 0 Param 1=&Multiply_Float(
     A=&RandomRangeFloat(Minimum=30, Maximum=30)
     B=&FloatFromCurve(
      FloatCurve=DataInterface() {propertyValues {Curve {keys [
          <at 1 2.5 RCIM_Cubic>
          {arriveTangent=-1.44, interpMode=RCIM_Cubic, leaveTangent=-1.44, time=1.6444, value=1.8}
          {arriveTangent=-0.0085, tangentMode=RCTM_User, time=2.25, value=0.7}
         ]}}}
      CurveIndex=$Emitter.Age
      Scale Curve=0
     )
    )
    Index 0 Param 2=&Multiply_Float(
     A=&RandomRangeFloat(Minimum=20, Maximum=30)
     B=&FloatFromCurve(
      FloatCurve=DataInterface() {propertyValues {Curve {keys [
          <at 1 2.5 RCIM_Cubic>
          {arriveTangent=-1.44, interpMode=RCIM_Cubic, leaveTangent=-1.44, time=1.6444, value=1.8}
          {arriveTangent=-0.0085, tangentMode=RCTM_User, time=2.25, value=0.7}
         ]}}}
      CurveIndex=$Emitter.Age
     )
    )
    Index 0 Param 3=1.5
    Index 0 Param 4=&RandomRangeFloat(Minimum=0, Maximum=1)
    Index 1 Param 1=&RandomRangeFloat(Minimum=0.2, Maximum=1)
    Index 1 Param 2=&FloatFromCurve(FloatCurve=DataInterface() {propertyValues {Curve {keys [{tangentMode=RCTM_User, time=0, value=1}, {tangentMode=RCTM_User, time=0.5, value=1}]}}})
    Module.Index0Param0Write=true
    Module.Index0Param1Write=true
    Module.Index0Param2Write=true
    Module.Index0Param3Write=true
    Module.Index1Param0Write=false
    Module.Index1Param1Write=false
    Module.Index1Param2Write=false
    Module.Index1Param3Write=false
    Write Parameter Index 1=true
   )
   &AddVelocity(Velocity=&RandomRangeVector(Minimum=<-100 250 0>, Maximum=<100 500 0>))
  ]
  particleUpdateScript [
   &ParticleState()
   &UpdateMeshOrientation(
    Rotation Coordinate Space=::NewEnumerator2
    Rotation Rate=&Multiply_Float(
     A=&RandomRangeFloat(Minimum=-0.12, Maximum=0.25)
     B=&FloatFromCurve(
      FloatCurve=DataInterface() {propertyValues {Curve {keys [
          {arriveTangent=-0.4674, interpMode=RCIM_Cubic, leaveTangent=-0.4674, tangentMode=RCTM_User, time=0.5, value=0}
          {arriveTangent=-0.9565, interpMode=RCIM_Cubic, leaveTangent=-0.9565, tangentMode=RCTM_User, time=1, value=1}
         ]}}}
      CurveIndex=$Particles.Age
     )
    )
    Rotation Vector=<0 0 1>
   )
   &ScaleColor(
    Scale Alpha=&FloatFromCurve(FloatCurve=DataInterface() {propertyValues {Curve {keys [
         {interpMode=RCIM_Cubic, tangentMode=RCTM_User, time=0, value=0}
         {interpMode=RCIM_Cubic, tangentMode=RCTM_User, time=0.5, value=1}
         {tangentMode=RCTM_User, time=1, value=0}
        ]}}})
    Scale RGB=&Multiply_VectorByFloat(Vector=<1 1 1>, Float=1)
    ScaleA=true
    ScaleRGB=true
   )
   &GravityForce(Gravity=<0 0 -122.5>)
   &SolveForcesAndVelocity(Clamp Velocity=false, Limit Acceleration=false)
  ]
  rendererClasses [&NiagaraMeshRendererProperties]
  renderers [{rendererClass=&NiagaraMeshRendererProperties, rendererIndex=0}]
 }
 Trail_ring001 {
  bEnabled=false
  emitterUpdateScript [
   &EmitterState(Scalability Mode=::NewEnumerator1)
   &SpawnRate(Use Spawn Probability=true)
  ]
  particleSpawnScript [
   &InitializeParticle(
    Color Maximum=<rgba 1 0.0217 0.3635 1>
    Color Minimum=<rgba 0.2048 0 1 1>
    Color Mode=::NewEnumerator2
    Lifetime Max=3
    Lifetime Min=2
    Lifetime Mode=::NewEnumerator1
    Mesh Scale Mode=::NewEnumerator1
    Mesh Uniform Scale Max=10
    Mesh Uniform Scale Min=8
    Position Offset=&MakeVector(
     X=0
     Y=4000
     Z=&Add_Float(A=-600, B=&RandomRangeFloat(Minimum=-100, Maximum=200))
    )
    UsePositionOffset=true
    Write Material Random=true
   )
   &InitialMeshOrientation(Mesh Orientation Mode=::NewEnumerator2, Rotation=<0.25 0 0>, Use Rotation Vector=true)
   &InitialMeshOrientation(
    Mesh Orientation Mode=::NewEnumerator2
    Rotation=&RandomRangeVector(Minimum=<0 0 0>, Maximum=<0 0 1>)
    Use Rotation Vector=true
   ) {module=InitialMeshOrientation001}
   &InitialMeshOrientation(Mesh Orientation Mode=::NewEnumerator2, Rotation=<0 0.05 0>, Use Rotation Vector=true) {module=InitialMeshOrientation002}
   &DynamicMaterialParameters(
    Index 0 Param 1=&Multiply_Float(
     A=&RandomRangeFloat(Minimum=30, Maximum=30)
     B=&FloatFromCurve(
      FloatCurve=DataInterface() {propertyValues {Curve {keys [
          <at 1 2.5 RCIM_Cubic>
          {arriveTangent=-1.44, interpMode=RCIM_Cubic, leaveTangent=-1.44, time=1.6444, value=1.8}
          {arriveTangent=-0.0085, tangentMode=RCTM_User, time=2.25, value=0.7}
         ]}}}
      CurveIndex=$Emitter.Age
      Scale Curve=0
     )
    )
    Index 0 Param 2=&Multiply_Float(
     A=&RandomRangeFloat(Minimum=20, Maximum=30)
     B=&FloatFromCurve(
      FloatCurve=DataInterface() {propertyValues {Curve {keys [
          <at 1 2.5 RCIM_Cubic>
          {arriveTangent=-1.44, interpMode=RCIM_Cubic, leaveTangent=-1.44, time=1.6444, value=1.8}
          {arriveTangent=-0.0085, tangentMode=RCTM_User, time=2.25, value=0.7}
         ]}}}
      CurveIndex=$Emitter.Age
      Scale Curve=0
     )
    )
    Index 0 Param 3=1.5
    Index 0 Param 4=&RandomRangeFloat(Minimum=0, Maximum=1)
    Index 1 Param 1=&RandomRangeFloat(Minimum=0.2, Maximum=1)
    Index 1 Param 2=&FloatFromCurve(FloatCurve=DataInterface() {propertyValues {Curve {keys [{tangentMode=RCTM_User, time=0, value=1}, {tangentMode=RCTM_User, time=0.5, value=1}]}}})
    Module.Index0Param0Write=true
    Module.Index0Param1Write=true
    Module.Index0Param2Write=true
    Module.Index0Param3Write=true
    Module.Index1Param0Write=true
    Module.Index1Param1Write=true
    Module.Index1Param2Write=false
    Module.Index1Param3Write=false
    Write Parameter Index 1=true
   )
   &AddVelocity(Velocity=&RandomRangeVector(Minimum=<-100 250 0>, Maximum=<100 500 0>))
  ]
  particleUpdateScript [
   &ParticleState()
   &UpdateMeshOrientation(
    Rotation Coordinate Space=::NewEnumerator2
    Rotation Rate=&Multiply_Float(
     A=&RandomRangeFloat(Minimum=-0.125, Maximum=0.25)
     B=&FloatFromCurve(
      FloatCurve=DataInterface() {propertyValues {Curve {keys [
          {arriveTangent=-0.4674, interpMode=RCIM_Cubic, leaveTangent=-0.4674, tangentMode=RCTM_User, time=0.5, value=0}
          {arriveTangent=-0.9565, interpMode=RCIM_Cubic, leaveTangent=-0.9565, tangentMode=RCTM_User, time=1, value=1}
         ]}}}
      CurveIndex=$Particles.Age
     )
    )
    Rotation Vector=<0 0 1>
   )
   &ScaleColor(
    Scale Alpha=&FloatFromCurve(FloatCurve=DataInterface() {propertyValues {Curve {keys [
         {interpMode=RCIM_Cubic, tangentMode=RCTM_User, time=0, value=0}
         {interpMode=RCIM_Cubic, tangentMode=RCTM_User, time=0.5, value=1}
         {tangentMode=RCTM_User, time=1, value=0}
        ]}}})
    Scale RGB=&Multiply_VectorByFloat(Vector=<1 1 1>, Float=1)
    ScaleA=true
    ScaleRGB=true
   )
   &GravityForce(Gravity=<0 0 -122.5>)
   &SolveForcesAndVelocity(Clamp Velocity=false, Limit Acceleration=false)
  ]
  rendererClasses [&NiagaraMeshRendererProperties]
  renderers [{rendererClass=&NiagaraMeshRendererProperties, rendererIndex=0}]
 }
}
summary {
 systemName=NS_PistolCanary_Tentacle_GroundPortal_03
 emitters [
  {emitterName=BlackHoleEnergy, rendererClasses [&NiagaraMeshRendererProperties]}
  {emitterName=EnergyWave001, rendererClasses [&NiagaraMeshRendererProperties]}
  {emitterName=Lightning_Primary, rendererClasses [&NiagaraMeshRendererProperties]}
  {emitterName=BlackHoleEnergy002, rendererClasses [&NiagaraMeshRendererProperties]}
  {bEnabled=false, emitterName=Axis, rendererClasses [&NiagaraMeshRendererProperties]}
  {bEnabled=false, emitterName=Trail_ring001, rendererClasses [&NiagaraMeshRendererProperties]}
  {bEnabled=false, emitterName=Trail_ring, rendererClasses [&NiagaraMeshRendererProperties]}
 ]
 userVariables [
  User.bIsFrontEnd {type=NiagaraBool}
  User.bIsFrontEndPreview {type=NiagaraBool}
  User.BlackHoleEnd {type=Vector3f, description="", defaultValue {struct=Vector3f, value=<0 50 -150>}}
  User.BlackHoleStart {type=NiagaraPosition, description="", defaultValue {}}
  User.ColorMult {type=NiagaraFloat, defaultValue=1, description=""}
  User.CoolColor_01 {type=LinearColor, description="", defaultValue {struct=LinearColor, value=<rgba 0.1779 0.2747 1 1>}}
  User.CoolColor_02 {type=LinearColor, description="", defaultValue {struct=LinearColor, value=<rgba 0.3712 0.0595 0.1714 1>}}
  User.ExplosionScale {type=NiagaraFloat, defaultValue=1.2, description=""}
  User.Fade {type=NiagaraFloat, defaultValue=1, description=""}
  User.GlobalScale {type=NiagaraFloat, defaultValue=33, description=""}
  User.Intensity {type=NiagaraFloat, defaultValue=50, description=""}
  User.Lifetime {type=NiagaraFloat, defaultValue=3, description=""}
  User.Open {type=NiagaraFloat, defaultValue=2, description=""}
  User.OverallWidth {type=NiagaraFloat, defaultValue=3720.3586, description=""}
  User.ParaGlideLeanAlpha {type=NiagaraFloat, defaultValue=0, description=""}
  User.PlayerSpeed {type=NiagaraFloat, defaultValue=0, description=""}
  User.PortalSize {type=NiagaraFloat, defaultValue=1, description=""}
  User.RemoveBeam {type=NiagaraFloat, defaultValue=0, description=""}
  User.Scale {type=NiagaraFloat, defaultValue=1, description=""}
  User.SFXVolume {type=NiagaraFloat, defaultValue=0, description=""}
  User.SplineUpdatedEndPos {type=Vector3f, description="", defaultValue {struct=Vector3f, value=<0 0 -25920.5664>}}
  User.SplineUpdatedStartPos {type=NiagaraPosition, description="", defaultValue {}}
  User.VariantSwitch {type=NiagaraFloat, defaultValue=0, description=""}
  User.Volume {type=NiagaraFloat, defaultValue=0, description=""}
  User.WarmColor_01 {type=LinearColor, description="", defaultValue {struct=LinearColor, value=<rgba 1 0 0.6896 1>}}
  User.WarmColor_02 {type=LinearColor, description="", defaultValue {struct=LinearColor, value=<rgba 0.782 0.1166 1 1>}}
  User.WidthMult {type=NiagaraFloat, defaultValue=1, description=""}
 ]
}
