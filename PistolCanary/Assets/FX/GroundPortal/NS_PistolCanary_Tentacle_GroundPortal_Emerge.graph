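# Niagara system for the emergence of a PistolCanary tentacle from a ground portal, mixing ground flow, pink lightning, smoke and debris.
# GroundFlow and lightning meshes animate through curve-driven dynamic material parameters, with User.WidthMult and ColorMult adjusting the lightning.
# GPU smoke sprites darken from pink to black, while GPU rocks launch upward under gravity and curl noise.
# The MeshRibbon emitter is disabled, and many exposed user parameters are unreferenced in the listed modules.
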NiagaraSystem /PistolCanary/Assets/FX/GroundPortal/NS_PistolCanary_Tentacle_GroundPortal_Emerge
dependencies {
 usedDataInterfaces [
  &NiagaraDataInterfaceCurve
  &NiagaraDataInterfaceArrayColor
  &NiagaraDataInterfaceColorCurve
  &NiagaraDataInterfaceGBuffer
 ]
 usedDynamicInputs []
 usedModules [
  &self:SystemSpawnScript.SystemScriptSource.SystemScriptGraph.NiagaraNodeAssignment_0.SetVariables_#1
  &SystemState
  &self:SystemSpawnScript.SystemScriptSource.SystemScriptGraph.NiagaraNodeAssignment_1.SetVariables_#2
  &self:SystemSpawnScript.SystemScriptSource.SystemScriptGraph.NiagaraNodeAssignment_2.SetVariables_#3
  &self:SystemSpawnScript.SystemScriptSource.SystemScriptGraph.NiagaraNodeAssignment_3.SetVariables_#4
  &self:SystemSpawnScript.SystemScriptSource.SystemScriptGraph.NiagaraNodeAssignment_4.SetVariables_#5
 ]
}
emitters {
 GroundFlow {
  emitterSpawnScript [/SetVariables(Emitter.MeshRibbonData=DataInterface() {propertyValues {
      bExposeCurve=true
      AlphaCurve {keys []}
      BlueCurve {keys [<at 0 1>,<at 1 1>]}
      GreenCurve {keys [<at 0 -0.5 RCIM_Cubic>, {arriveTangent=0.0331, interpMode=RCIM_Cubic, leaveTangent=0.0331, tangentMode=RCTM_User, time=1, value=-1}]}
      RedCurve {keys [<at 0 1 RCIM_Cubic>,<at 1 2 RCIM_Cubic>]}
     }}) {isSetParameters=true, module=SetVariables_#6}]
  emitterUpdateScript [
   &EmitterState()
   &SpawnBurst_Instantaneous(Spawn Count=25, Use Spawn Probability=false, UseLoopCountLimit=false) {enabled=false}
   &SpawnRate(SpawnRate=5, Use Spawn Probability=false)
  ]
  particleSpawnScript [
   &InitializeParticle(
    Color=&Multiply_LinearColorByFloat(Linear Color=<rgba 25 25 25 0.25>, Float=&RandomRangeFloat(Minimum=1, Maximum=2))
    Lifetime=&RandomRangeFloat(Minimum=3, Maximum=4)
    Mesh Scale=&Multiply_VectorByFloat(
     Vector=&RandomRangeVector(Minimum=<1 1 3>, Maximum=<1 1 5>)
     Float=&RandomRangeFloat(Minimum=10, Maximum=20)
    )
    Mesh Scale Mode=::NewEnumerator4
    UsePositionOffset=false
    Write Material Random=true
   )
   &InitialMeshOrientation(
    Mesh Orientation Mode=::NewEnumerator2
    Orientation Vector=<0 1 0>
    Rotation=&RandomRangeVector(Minimum=<0 0 0>, Maximum=<0 0 1>)
    Use Rotation Vector=true
   )
   &SystemLocation(MaskSpawnGroup=false)
   &MeshRotationForce(Rotation=<0 0 -0.1>) {enabled=false}
   &ApplyInitialForces() {enabled=false}
   &AddVelocity(
    Cone Angle=180
    Cone Axis=<0 0 1>
    Inner Cone Angle=179
    Module.UseConeFalloff=true
    Module.UseVelDistribution=false
    Override Seed=false
    Velocity Mode=::NewEnumerator2
    Velocity Speed=0
   ) {enabled=false}
   &ShapeLocation(MaskSpawnGroup=false, Override Seed=false, Ring Radius=5000, Shape Primitive=::NewEnumerator4, UseOwnerScale=true) {enabled=false}
  ]
  particleUpdateScript [
   &ParticleState()
   &ScaleMeshSize(Scale Factor=&VectorFromFloat(Value=&FloatFromCurve(FloatCurve=DataInterface() {propertyValues {Curve {keys [<at 0 0.75 RCIM_Cubic>,<at 1 1 RCIM_Cubic>]}}})))
   &DynamicMaterialParameters(
    Index 0 Param 1=&FloatFromCurve(
     FloatCurve=DataInterface() {propertyValues {Curve {keys [<at 0 0>, {arriveTangent=-1.8673, interpMode=RCIM_Cubic, leaveTangent=-1.8673, tangentMode=RCTM_User, time=1, value=1}]}}}
     Scale Curve=&RandomRangeFloat(Minimum=1, Maximum=1)
    )
    Index 0 Param 2=&FloatFromCurve(
     FloatCurve=DataInterface() {propertyValues {Curve {keys [{interpMode=RCIM_Cubic, tangentMode=RCTM_User, time=0, value=0}, {interpMode=RCIM_Cubic, tangentMode=RCTM_User, time=1, value=1}]}}}
     Scale Curve=&RandomRangeFloat(Minimum=0, Maximum=1)
    )
    Index 0 Param 3=&RandomRangeFloat(Minimum=0, Maximum=1)
    Index 0 Param 4=&RandomRangeFloat(Minimum=25, Maximum=100)
    Index 1 Param 1=1
    Index 1 Param 2=1000
    Module.Index0Param0Write=true
    Module.Index0Param1Write=true
    Module.Index0Param2Write=true
    Module.Index0Param3Write=true
    Module.Index1Param0Write=true
    Module.Index1Param1Write=true
    Module.Index1Param2Write=false
    Module.Index1Param3Write=false
    Write Parameter Index 1=true
   )
   &ScaleColor(
    Scale Alpha=&FloatFromCurve(
     FloatCurve=DataInterface() {propertyValues {Curve {
        defaultValue=1
        keys [
         {interpMode=RCIM_Cubic, tangentMode=RCTM_User, time=0, value=0}
         {interpMode=RCIM_Cubic, tangentMode=RCTM_User, time=0.5, value=1}
         {tangentMode=RCTM_User, time=1, value=0}
        ]
       }}}
     Scale Curve=0.01
    )
    Scale RGB=<5 1 5>
    ScaleA=true
    ScaleRGB=true
   )
   &CameraOffset() {enabled=false}
   &SolveForcesAndVelocity(Clamp Velocity=false, Limit Acceleration=false) {enabled=false}
  ]
  rendererClasses [&NiagaraMeshRendererProperties]
  renderers [{rendererClass=&NiagaraMeshRendererProperties, rendererIndex=0}]
 }
 Lightning_Primary001 {
  emitterUpdateScript [
   &EmitterState(Life Cycle Mode=::NewEnumerator1, Loop Duration=1, UseLoopDelay=false)
   &SpawnRate(SpawnRate=50, Use Spawn Probability=false) {module=SpawnRate001}
  ]
  particleSpawnScript [
   /SetVariables(Particles.isShort:NiagaraFloat=&FloatFromCurve(
     FloatCurve=DataInterface() {propertyValues {Curve {keys [<at 0 0 RCIM_Constant>,<at 0.8 1 RCIM_Constant>]}}}
     CurveIndex=&RandomRangeFloat(Minimum=0, Maximum=1)
    )) {enabled=false, isSetParameters=true}
   &InitializeParticle(
    Color=&SelectLinearColorFromArray(
     Array Sampling Mode=::NewEnumerator1
     Direct Array Index=&RandomRangeInt(Minimum=0, Maximum=1)
     Color Selection Array=DataInterface() {propertyValues {ColorData [<rgba 0.5 0.0145 0.2765 1>, <rgba 0.5 0 0.2161 1>]}}
     Random Seed=?
    )
    Lifetime=&Lerp_Float(
     A=&RandomRangeFloat(Minimum=0.5, Maximum=0.8)
     B=&RandomRangeFloat(Minimum=0.3, Maximum=0.4)
     Alpha=$Particles.isShort
    )
    Mesh Scale Mode=::NewEnumerator3
    Mesh Uniform Scale=&MakeFloatFromInt(Int=&Multiply_Int(A=&RandomRangeInt(Minimum=6, Maximum=12), B=4))
    Position Mode=::NewEnumerator1
    Sprite Size Mode=::NewEnumerator4
    UsePositionOffset=false
    Write Material Random=true
   ) {module=InitializeParticle001}
   &DynamicMaterialParameters(
    Index 0 Param 1=1
    Index 0 Param 2=0.15
    Index 0 Param 3=1
    Index 0 Param 4=1
    Module.Param0WriteEnabled=true
    Module.Param1WriteEnabled=true
    Module.Param2WriteEnabled=true
    Module.Param3WriteEnabled=true
   )
   &ShapeLocation(
    MaskSpawnGroup=false
    Non Uniform Scale=<1 1 0>
    Offset Mode=::NewEnumerator0
    Override Seed=false
    Sphere Radius=3000
    Sphere Surface Distribution=1
    UseOwnerScale=true
   ) {module=ShapeLocation002}
   &AddVelocity(Module.Use Falloff Distance=false, Override Seed=false, Velocity Mode=::NewEnumerator1, Velocity Speed=0.25)
   &InitialMeshOrientation(
    Mesh Orientation Mode=::NewEnumerator7
    Rotation=&RandomRangeVector(Minimum=<0 0 0>, Maximum=<0 0 0>)
    Use Rotation Vector=true
   ) {module=InitialMeshOrientation002}
  ]
  particleUpdateScript [
   &SolveForcesAndVelocity(Clamp Velocity=false, Limit Acceleration=false)
   &ParticleState()
   &ScaleMeshSize(Scale Factor=&Multiply_VectorByFloat(
     Vector=&MakeVector(
      X=&Multiply_Float(
       A=&FloatFromCurve(FloatCurve=DataInterface() {propertyValues {Curve {keys [{interpMode=RCIM_Cubic, tangentMode=RCTM_User, time=0, value=1}, {interpMode=RCIM_Cubic, tangentMode=RCTM_User, time=1, value=1.5}]}}})
       B=&RandomRangeFloat(Minimum=0.5, Maximum=0.8)
      )
      Y=1
      Z=1
     )
     Float=&MakeFloatFromVector(
      Vector=&VectorFromFloat(Value=&Multiply_Float(
        A=&RandomRangeFloat(Minimum=0.8, Maximum=1.2)
        B=&Multiply_Float(
         A=0.1
         B=&FloatFromCurve(FloatCurve=DataInterface() {propertyValues {Curve {keys [
              {arriveTangent=0.2102, interpMode=RCIM_Cubic, leaveTangent=0.2102, tangentMode=RCTM_User, time=0, value=0.9}
              {interpMode=RCIM_Cubic, tangentMode=RCTM_User, time=1, value=1}
             ]}}})
        )
       ))
      Channel=::NewEnumerator0
     )
    ))
   &DynamicMaterialParameters(
    Index 0 Param 1=&Lerp_Float(
     A=&FloatFromCurve(FloatCurve=DataInterface() {propertyValues {Curve {keys [{tangentMode=RCTM_User, time=0, value=1}, <at 0.2 0.5>, <at 0.8 0.5>, {tangentMode=RCTM_User, time=1, value=0}]}}})
     B=0.4
     Alpha=$Particles.isShort
    )
    Index 0 Param 2=&FloatFromCurve(
     FloatCurve=DataInterface() {propertyValues {Curve {keys [
         {interpMode=RCIM_Cubic, leaveTangent=-1.1405, tangentMode=RCTM_Break, time=0, value=0.5}
         {arriveTangent=4.4936, interpMode=RCIM_Cubic, leaveTangent=-5.3647, tangentMode=RCTM_Break, time=0.4833, value=1}
         <at 1 0>
        ]}}}
     Scale Curve=&Multiply_Float(A=10, B=$User.WidthMult)
    )
    Index 0 Param 3=120
    Index 1 Param 1=&RandomRangeFloat(Minimum=0, Maximum=1)
    Index 1 Param 2=1
    Index 1 Param 3=1
    Index 1 Param 4=1
    Module.Param0WriteEnabled=true
    Module.Param0WriteEnabled1=true
    Module.Param1WriteEnabled=true
    Module.Param1WriteEnabled1=true
    Module.Param2WriteEnabled=true
    Module.Param2WriteEnabled1=true
    Module.Param3WriteEnabled=false
    Module.Param3WriteEnabled1=true
    Write Parameter Index 1=true
   ) {module=DynamicMaterialParameters001}
   &ScaleColor(
    Scale Alpha=&FloatFromCurve(FloatCurve=DataInterface() {propertyValues {Curve {keys [
         {interpMode=RCIM_Cubic, tangentMode=RCTM_User, time=0, value=0}
         {interpMode=RCIM_Cubic, tangentMode=RCTM_User, time=0.1258, value=1}
         <at 0.341 0>
         {interpMode=RCIM_Cubic, tangentMode=RCTM_User, time=0.496, value=0.5}
         <at 1 0>
        ]}}})
    Scale RGB=&VectorFromFloat(Value=&FloatFromCurve(
      FloatCurve=DataInterface() {propertyValues {Curve {keys [
          {arriveTangent=0.1143, interpMode=RCIM_Cubic, leaveTangent=0.1143, tangentMode=RCTM_User, time=0.018, value=4.9326}
          {arriveTangent=-6.0507, interpMode=RCIM_Cubic, leaveTangent=-6.0507, time=0.1778, value=0.8567}
          {arriveTangent=-0.5645, interpMode=RCIM_Cubic, leaveTangent=-0.5645, tangentMode=RCTM_User, time=0.2493, value=3.533}
          {arriveTangent=-22.9399, interpMode=RCIM_Cubic, leaveTangent=-22.9399, time=0.3224, value=1.7136}
          {arriveTangent=2.8763, interpMode=RCIM_Cubic, leaveTangent=2.8763, time=0.3773, value=0.5981}
          {arriveTangent=-1.8312, interpMode=RCIM_Cubic, leaveTangent=-1.8312, time=0.4693, value=2.1361}
          {arriveTangent=-2.2307, interpMode=RCIM_Cubic, leaveTangent=-2.2307, time=0.5983, value=0.1933}
          {arriveTangent=1.5587, interpMode=RCIM_Cubic, leaveTangent=1.5587, time=0.7391, value=1.5342}
          {arriveTangent=-2.0478, interpMode=RCIM_Cubic, leaveTangent=-2.0478, time=0.8629, value=0.6057}
          {arriveTangent=-0.0659, interpMode=RCIM_Cubic, leaveTangent=-0.0659, tangentMode=RCTM_User, time=1, value=1}
         ]}}}
      Scale Curve=$User.ColorMult
     ))
    ScaleA=true
    ScaleRGB=true
   ) {module=ScaleColor001}
  ]
  rendererClasses [&NiagaraMeshRendererProperties]
  renderers [{rendererClass=&NiagaraMeshRendererProperties, rendererIndex=0}]
 }
 MeshRibbon {
  bEnabled=false
  emitterSpawnScript [/SetVariables(Emitter.MeshRibbonData=DataInterface() {propertyValues {
      bExposeCurve=true
      AlphaCurve {keys []}
      BlueCurve {keys [<at 0 1>,<at 1 1>]}
      GreenCurve {keys [<at 0 -0.5 RCIM_Cubic>, {arriveTangent=0.0331, interpMode=RCIM_Cubic, leaveTangent=0.0331, tangentMode=RCTM_User, time=1, value=-1}]}
      RedCurve {keys [<at 0 1 RCIM_Cubic>,<at 1 2 RCIM_Cubic>]}
     }}) {isSetParameters=true, module=SetVariables_#6}]
  emitterUpdateScript [
   &EmitterState()
   &SpawnRate(SpawnRate=100, Use Spawn Probability=false) {module=SpawnRate002}
   &SpawnBurst_Instantaneous(Spawn Count=12, Use Spawn Probability=false, UseLoopCountLimit=false) {enabled=false}
   &SpawnRate(
    SpawnRate=&Multiply_Float(A=&Lerp_Float(A=2, B=10, Alpha=$System.ActivationAlpha), B=1)
    Use Spawn Probability=false
   ) {enabled=false}
   &SpawnRate(
    SpawnGroup=1
    SpawnRate=&Multiply_Float(A=&Lerp_Float(A=2, B=10, Alpha=$System.ActivationAlpha), B=1)
    Use Spawn Probability=false
   ) {enabled=false, module=SpawnRate001}
  ]
  particleSpawnScript [
   /SetVariables(Particles.VisibilityTag=&SpawnGroup()) {isSetParameters=true}
   &InitializeParticle(
    Color=&MaskLinearColorBySpawnGroup(Masked Linear Color=<rgba 5 5 5 1>, Passthrough Linear Color=<rgba 1 1 1 11>)
    Lifetime=&MaskFloatBySpawnGroup(
     Masked Float=&Lerp_Float(
      A=&RandomRangeFloat(Minimum=1.8, Maximum=3.5)
      B=&RandomRangeFloat(Minimum=1, Maximum=1.5)
      Alpha=$System.ActivationAlpha
     )
     Passthrough Float=&Lerp_Float(
      A=&RandomRangeFloat(Minimum=5.5, Maximum=8.5)
      B=&RandomRangeFloat(Minimum=1.5, Maximum=2)
      Alpha=$System.ActivationAlpha
     )
    )
    Mesh Scale=&MaskVectorBySpawnGroup(
     Masked Vector=&Multiply_VectorByFloat(Vector=<1 1 2.5>, Float=&Lerp_Float(A=7, B=8, Alpha=$System.ActivationAlpha))
     Passthrough Vector=&Multiply_VectorByFloat(Vector=<1 -1 1>, Float=&RandomRangeFloat(Minimum=7.9909, Maximum=12.017))
    )
    Mesh Scale Mode=::NewEnumerator4
    UsePositionOffset=false
    Write Material Random=true
   )
   &InitialMeshOrientation(
    Mesh Axis To Orient=<0 0 -1>
    Mesh Orientation Mode=::NewEnumerator2
    Orientation Vector=<0 0 -1>
    Rotation=&RandomRangeVector(Minimum=<0 0 0>, Maximum=<0 0 350>)
    Use Rotation Vector=true
   )
  ]
  particleUpdateScript [
   &ParticleState()
   &ScaleMeshSize(Scale Factor=&VectorFromFloat(Value=&MaskFloatBySpawnGroup(
      Masked Float=&FloatFromCurve(FloatCurve=DataInterface() {propertyValues {Curve {keys [<at 0 1 RCIM_Cubic>,<at 1 0.9 RCIM_Cubic>]}}})
      Passthrough Float=&FloatFromCurve(FloatCurve=DataInterface() {propertyValues {Curve {keys [{arriveTangent=0.8366, interpMode=RCIM_Cubic, leaveTangent=0.8366, tangentMode=RCTM_User, time=0, value=0.5}, <at 1 0.75 RCIM_Cubic>]}}})
     )))
   &DynamicMaterialParameters(
    Index 0 Param 1=&FloatFromCurve(FloatCurve=DataInterface() {propertyValues {Curve {keys [
         {interpMode=RCIM_Cubic, tangentMode=RCTM_User, time=0, value=0.5}
         {arriveTangent=2.7589, interpMode=RCIM_Cubic, leaveTangent=2.7589, tangentMode=RCTM_User, time=1, value=2}
        ]}}})
    Index 0 Param 2=&FloatFromCurve(FloatCurve=DataInterface() {propertyValues {Curve {keys [{interpMode=RCIM_Cubic, tangentMode=RCTM_User, time=0, value=0}, {interpMode=RCIM_Cubic, tangentMode=RCTM_User, time=1, value=1}]}}})
    Index 0 Param 3=&RandomRangeFloat(Minimum=3.0512, Maximum=5.4)
    Index 0 Param 4=&Add_Float(A=&RandomRangeFloat(Minimum=16, Maximum=40), B=0)
    Module.Index0Param0Write=true
    Module.Index0Param1Write=true
    Module.Index0Param2Write=true
    Module.Index0Param3Write=true
   )
   &ScaleColor(
    Scale Mode=::NewEnumerator1
    Scale RGBA=/Niagara/DynamicInputs/Helpers/MaskVector4(
     Mask Parameter:NiagaraInt32=$Emitter.SpawnGroup
     Mask:NiagaraInt32=0
     Masked Vector4:Vector4f=&Vector4FromCurve(Vector4Curve=DataInterface() {propertyValues {
        AlphaCurve {keys [<at 0 0 RCIM_Cubic>,<at 0.2932 3 RCIM_Cubic>,<at 1 0 RCIM_Cubic>]}
        BlueCurve {keys [<at 0 0 RCIM_Cubic>,<at 0.2925 3 RCIM_Cubic>,<at 1 0 RCIM_Cubic>]}
        GreenCurve {keys [<at 0 0 RCIM_Cubic>,<at 0.2925 3 RCIM_Cubic>,<at 1 0 RCIM_Cubic>]}
        RedCurve {keys [<at 0 0 RCIM_Cubic>,<at 0.2925 3 RCIM_Cubic>,<at 1 0 RCIM_Cubic>]}
       }})
     Passthrough Vector4:Vector4f=&Vector4FromCurve(Vector4Curve=DataInterface() {propertyValues {
        AlphaCurve {keys [<at 0 0 RCIM_Cubic>,<at 0.6722 5 RCIM_Cubic>,<at 1 0 RCIM_Cubic>]}
        BlueCurve {keys [<at 0 0 RCIM_Cubic>,<at 0.6722 5 RCIM_Cubic>,<at 1 0 RCIM_Cubic>]}
        GreenCurve {keys [<at 0 0 RCIM_Cubic>,<at 0.6722 5 RCIM_Cubic>,<at 1 0 RCIM_Cubic>]}
        RedCurve {keys [<at 0 0 RCIM_Cubic>,<at 0.6722 5 RCIM_Cubic>,<at 1 0 RCIM_Cubic>]}
       }})
    )
    ScaleRGBA=true
   )
   &CameraOffset(Camera Offset Amount=50)
   /SetVariables(Particles.DarkRibbonDynMat:Vector4f=&MakeVector4(
     X=&FloatFromCurve(FloatCurve=DataInterface() {propertyValues {Curve {keys [{tangentMode=RCTM_User, time=0, value=-1}, {tangentMode=RCTM_User, time=1, value=0}]}}})
     Y=&FloatFromCurve(FloatCurve=DataInterface() {propertyValues {Curve {keys [{tangentMode=RCTM_User, time=0, value=0}, {tangentMode=RCTM_User, time=1, value=1}]}}})
     Z=&RandomRangeFloat(Minimum=2.5103, Maximum=7.4856)
     W=&Add_Float(A=&RandomRangeFloat(Minimum=25.1137, Maximum=44.0696), B=0)
    )) {isSetParameters=true}
  ]
  rendererClasses [&NiagaraMeshRendererProperties]
  renderers [
   {rendererClass=&NiagaraMeshRendererProperties, rendererIndex=0}
   {rendererClass=&NiagaraMeshRendererProperties, rendererIndex=1}
  ]
 }
 Rocks_OnSpire001 {
  simTarget=GPUComputeSim
  emitterUpdateScript [
   &EmitterState(
    Enable Distance Culling=true
    MaxDistance=25000
    Module.bApplySpawnCountScaleByDistance=false
    Module.bUseMaxDistance=false
    Module.bUseMinDistance=false
    Scalability Mode=::NewEnumerator1
    Scale Spawn Count=true
    Spawn Count Scale By Distance=DataInterface() {propertyValues {Curve {keys [<at 0 1>,<at 0.1 1>,<at 1 1>]}}}
   )
   &SpawnBurst_Instantaneous(Spawn Count=100, Use Spawn Probability=false, UseLoopCountLimit=false) {enabled=false}
   &SpawnRate(SpawnRate=15, Use Spawn Probability=true)
  ]
  particleSpawnScript [
   &InitializeParticle(
    Color=<rgba 0.3542 0.024 0.3365 1>
    Lifetime Max=4
    Lifetime Min=2
    Lifetime Mode=::NewEnumerator1
    Mesh Scale Max=<6 6 8>
    Mesh Scale Min=<5 5 5>
    Mesh Scale Mode=::NewEnumerator2
    Position Offset=<0 0 -1000>
    Sprite Rotation Mode=::NewEnumerator2
    UsePositionOffset=true
    Write Material Random=true
   ) {module=InitializeParticle001}
   &InitialMeshOrientation(
    Mesh Orientation Mode=::NewEnumerator2
    Rotation=&RandomRangeVector(Minimum=<-0.02 -0.02 0>, Maximum=<0.02 0.02 1>)
    Use Rotation Vector=true
   )
   &MeshRotationForce(Rotation=&RandomRangeVector(Minimum=<0 0 -0.5>, Maximum=<0.1 0.1 0.5>))
   &AddVelocity(Velocity=&RandomRangeVector(Minimum=<-2000 -2000 2500>, Maximum=<2000 2000 5000>))
   &ApplyInitialForces()
   &CameraOffset(Camera Offset Amount=100)
   &ShapeLocation(
    MaskSpawnGroup=false
    Non Uniform Scale=<1 1 0>
    Offset=<0 0 1500>
    Offset Mode=::NewEnumerator0
    Override Seed=false
    Sphere Radius=1500
    UseOwnerScale=true
   ) {module=ShapeLocation002}
   /Niagara/Modules/Update/Utility/SampleGBuffer(Kill Beyond Depth=false, Snap To Depth:ENiagaraGBufferSnapToDepth=::NewEnumerator1, Write To Color=false, Write To Position=false)
   /SetVariables(Particles.InitialPos:NiagaraPosition=$Particles.Position) {isSetParameters=true}
   /SetVariables(Particles.MeshIndex=&RandomRangeInt(Minimum=0, Maximum=3)) {isSetParameters=true}
  ]
  particleUpdateScript [
   &ParticleState()
   &ScaleMeshSize(Scale Factor=&VectorFromFloat(Value=&FloatFromCurve(FloatCurve=DataInterface() {propertyValues {Curve {keys [<at 0.5 1>,<at 1 0>]}}}))) {module=ScaleMeshSize001}
   &UpdateMeshOrientation(Rotation Rate=&RandomRangeFloat(Minimum=-1, Maximum=1), Rotation Vector=<1 0.01 1>)
   &CurlNoiseForce(
    Bias Noise=true
    Bias Noise Field:Vector3f=<0 0 -1>
    Noise Frequency=&RandomRangeFloat(Minimum=0.1, Maximum=0.2)
    Noise Strength=&RandomRangeFloat(Minimum=25, Maximum=250)
    Pan Noise=true
    Pan Noise Field=<1 0 -1>
   )
   &PointAttractionForce(
    Attraction Radius=7500
    Attraction Strength=&Multiply_Float(
     A=&RandomRangeFloat(Minimum=10, Maximum=20)
     B=&FloatFromCurve(FloatCurve=DataInterface() {propertyValues {Curve {defaultValue=1, keys [<at 0.5 0 RCIM_Cubic>,<at 0.75 1 RCIM_Cubic>]}}})
    )
    Falloff Exponent=1.5
    Kill Radius=300
    Kill Within Radius=true
    Module.Use Falloff=true
    UseAttractorOffset=false
   ) {enabled=false, module=PointAttractionForce001}
   &VortexForce(Origin Pull Amount=200, UseOriginOffset=false, Vortex Force Amount=50) {enabled=false}
   &Drag(Use Linear Drag=true, Use Rotational Drag=true)
   &GravityForce(Gravity=<0 0 -1960>)
   &SolveForcesAndVelocity(
    Acceleration Limit=100000
    Clamp Velocity=true
    Limit Acceleration=false
    Speed Limit:NiagaraFloat=&Multiply_Float(
     A=10000
     B=&FloatFromCurve(FloatCurve=DataInterface() {propertyValues {Curve {keys [
          <at 0 0.5>
          {arriveTangent=1.255, interpMode=RCIM_Cubic, leaveTangent=1.255, tangentMode=RCTM_User, time=0.5, value=1}
          {arriveTangent=2.5018, interpMode=RCIM_Cubic, leaveTangent=2.5018, tangentMode=RCTM_User, time=1, value=3}
         ]}}})
    )
   )
   /SetVariables(Particles.SubImageIndex=&FloatFromCurve(FloatCurve=DataInterface() {propertyValues {Curve {keys [<at 0 0>,<at 1 1>]}}}, Scale Curve=16)) {isSetParameters=true}
   &ScaleColor(
    Scale Alpha=&FloatFromCurve(
     FloatCurve=DataInterface() {propertyValues {Curve {
        defaultValue=1
        keys [
         {interpMode=RCIM_Cubic, tangentMode=RCTM_User, time=0, value=0}
         {interpMode=RCIM_Cubic, tangentMode=RCTM_User, time=0.5, value=1}
         {tangentMode=RCTM_User, time=1, value=0}
        ]
       }}}
     Scale Curve=0.05
    )
    ScaleA=true
    ScaleRGB=true
   )
   &ScaleMeshSize(Scale Factor=&Multiply_VectorByFloat(
     Vector=&Multiply_VectorByFloat(Vector=<1 1 1>, Float=$User.ScaleMult)
     Float=&FloatFromCurve(
      FloatCurve=DataInterface() {propertyValues {Curve {
         defaultValue=1
         keys [
          {interpMode=RCIM_Cubic, tangentMode=RCTM_User, time=0, value=0.75}
          {interpMode=RCIM_Cubic, tangentMode=RCTM_User, time=0.5, value=1}
          {tangentMode=RCTM_User, time=1, value=0}
         ]
        }}}
      Scale Curve=/Niagara/DynamicInputs/Scalability/ScalabilityDistanceBasedFloat(
       Curve:NiagaraDataInterfaceCurve=DataInterface() {propertyValues {Curve {keys [<at 0 0>,<at 0.02 1>,<at 0.8 1>,<at 1 0>]}}}
       CurveIndex:NiagaraFloat=$Emitter.Scalability.DistanceFraction
       Bias:NiagaraFloat=0
       Scale:NiagaraFloat=&Multiply_Float(A=3, B=$System.Portal_Intensity)
      )
     )
    )) {enabled=false}
   &DynamicMaterialParameters(
    Index 0 Param 4=&FloatFromCurve(FloatCurve=DataInterface() {propertyValues {Curve {keys [<at 0 1>,<at 1 0>]}}}, Scale Curve=2)
    Module.Index0Param0Write=true
    Module.Index0Param1Write=true
    Module.Index0Param2Write=true
    Module.Index0Param3Write=true
   )
  ]
  rendererClasses [&NiagaraMeshRendererProperties, &NiagaraSpriteRendererProperties]
  renderers [
   {rendererClass=&NiagaraMeshRendererProperties, rendererIndex=0}
   {rendererClass=&NiagaraSpriteRendererProperties, rendererIndex=1}
  ]
 }
 Smoke {
  simTarget=GPUComputeSim
  emitterUpdateScript [
   &EmitterState()
   &SpawnBurst_Instantaneous(Spawn Count=5, Use Spawn Probability=false, UseLoopCountLimit=false)
   &SpawnRate(SpawnRate=10, Use Spawn Probability=false)
   &SpawnRate(SpawnGroup=1, SpawnRate=2, Use Spawn Probability=false) {enabled=false, module=SpawnRate001}
  ]
  particleSpawnScript [
   &InitializeParticle(
    Color=<rgba 1 0 0.3333 1>
    Lifetime=&RandomRangeFloat(Minimum=1, Maximum=4)
    Position Offset=&RandomRangeVector(Minimum=<0 0 1000>, Maximum=<0 0 1250>)
    Sprite Size Mode=::NewEnumerator3
    Sprite UV Mode=::NewEnumerator3
    Uniform Sprite Size=&RandomRangeFloat(Minimum=5, Maximum=5)
    UsePositionOffset=true
    Write Material Random=true
   )
   &ShapeLocation(
    MaskSpawnGroup=true
    Override Seed=false
    Ring Radius=970.7466
    Rotation Mode=::NewEnumerator2
    Shape Primitive=::NewEnumerator4
    UseOwnerScale=true
    Yaw / Pitch / Roll=<0 0.25 0>
   ) {enabled=false}
   &ShapeLocation(
    MaskSpawnGroup=true
    Override Seed=false
    Rotation Mode=::NewEnumerator2
    Shape Origin=&AddVectorToPosition(
     Position=$StackContext.Position
     Vector=&Multiply_Vector(
      A=<207.3846 550.5953 -570.5737>
      B=&MakeVector(
       X=1
       Y=&UniformAOrBFloat(A=-1, Evaluation Type=::NewEnumerator0, Recalculate Random Each Loop=true, Override Seed=false)
       Z=1
      )
     )
    )
    Spawn Group Mask=1
    Sphere Radius=148.652
    UseOwnerScale=true
    Yaw / Pitch / Roll=<0.25 0.25 0>
   ) {enabled=false, module=ShapeLocation001}
   /SetVariables(Particles.SpriteRotation:NiagaraFloat=&RandomRangeFloat(Minimum=-0.2293, Maximum=0.3106)) {isSetParameters=true}
   &AddVelocity(Velocity=$StackContext.ShapeLocation.ShapeVector, Velocity Speed Scale=0)
   /SetVariables(Particles.SpriteSize=&Multiply_Vector2D_ByFloat(
     Float=&Lerp_Float(A=107.586, B=190.6381, Alpha=&MakeFloatFromInt(Int=$Particles.SpawnGroup))
     Vector2D=$Particles.SpriteSize
    )) {isSetParameters=true}
   &ApplyInitialForces() {enabled=false}
   &ShapeLocation(
    Hemisphere Distribution=<1 0.5>
    MaskSpawnGroup=false
    Override Seed=false
    Sphere Radius=1500
    Sphere Surface Distribution=1
    UseOwnerScale=true
   ) {module=ShapeLocation002}
  ]
  particleUpdateScript [
   &Drag(Drag=2, Use Linear Drag=true, Use Rotational Drag=true) {enabled=false}
   &VortexForce(
    Origin Pull Amount=76.4139
    UseOriginOffset=false
    Vortex Axis=<1 0 0>
    Vortex Force Amount=&Lerp_Float(
     A=&RandomRangeFloat(Minimum=93.3802, Maximum=178.6482)
     B=0
     Alpha=&MakeFloatFromInt(Int=$Particles.SpawnGroup)
    )
   ) {enabled=false}
   &GravityForce(Gravity=&Lerp_Vector(A=<0 0 850>, B=<0 0 2000>, Alpha=&MakeFloatFromInt(Int=$Particles.SpawnGroup))) {enabled=false}
   &AccelerationForce(Acceleration=<300 0 0>, Coordinate Space=::Local) {enabled=false}
   &SolveForcesAndVelocity(Clamp Velocity=false, Limit Acceleration=false) {enabled=false}
   &ParticleState()
   &ScaleSpriteSize(
    Uniform Curve Scale=5
    Uniform Curve Sprite Scale=DataInterface() {propertyValues {Curve {keys [{arriveTangent=1.2245, interpMode=RCIM_Cubic, leaveTangent=1.2245, tangentMode=RCTM_User, time=0, value=1.5}, <at 1 2 RCIM_Cubic>]}}}
   )
   /SetVariables(Particles.SubImageIndex=&FloatFromCurve(FloatCurve=DataInterface() {propertyValues {Curve {keys [<at 0 0>,<at 1 1>]}}}, Scale Curve=16)) {isSetParameters=true}
   &ScaleColor(
    Scale Alpha=&FloatFromCurve(FloatCurve=DataInterface() {propertyValues {Curve {keys [
         <at 0 0 RCIM_Cubic>
         {arriveTangent=0.8929, interpMode=RCIM_Cubic, leaveTangent=0.8929, time=0.2, value=1}
         {arriveTangent=-1.25, interpMode=RCIM_Cubic, leaveTangent=-1.25, time=0.7, value=0.625}
         <at 1 0 RCIM_Cubic>
        ]}}})
    Scale RGB=&VectorFromFloat(Value=&FloatFromCurve(FloatCurve=DataInterface() {propertyValues {Curve {keys [
          {arriveTangent=-8.0933, interpMode=RCIM_Cubic, leaveTangent=-8.0933, tangentMode=RCTM_User, time=0, value=1}
          {arriveTangent=-0.9139, interpMode=RCIM_Cubic, leaveTangent=-0.9139, tangentMode=RCTM_User, time=0.2322, value=0.0558}
          {arriveTangent=-0.2701, interpMode=RCIM_Cubic, leaveTangent=-0.2701, time=0.2734, value=0.0702}
          {arriveTangent=-0.3098, interpMode=RCIM_Cubic, leaveTangent=-0.3098, time=0.3339, value=0.0283}
          <at 0.5 0>
         ]}}}))
    ScaleA=true
    ScaleRGB=true
   )
   &DynamicMaterialParameters(
    Index 0 Param 4=&FloatFromCurve(FloatCurve=DataInterface() {propertyValues {Curve {keys [<at 0 1>,<at 1 0>]}}}, Scale Curve=2)
    Module.Index0Param0Write=true
    Module.Index0Param1Write=true
    Module.Index0Param2Write=true
    Module.Index0Param3Write=true
   )
   &CameraOffset(Camera Offset Amount=&FloatFromCurve(FloatCurve=DataInterface() {propertyValues {Curve {keys [<at 0 25>,<at 1 75>]}}}))
   &self:MatchScale() {module=MatchScale}
  ]
  rendererClasses [&NiagaraSpriteRendererProperties]
  renderers [{rendererClass=&NiagaraSpriteRendererProperties}]
 }
}
summary {
 systemName=NS_PistolCanary_Tentacle_GroundPortal_Emerge
 emitters [
  {emitterName=Lightning_Primary001, rendererClasses [&NiagaraMeshRendererProperties]}
  {bEnabled=false, emitterName=MeshRibbon, rendererClasses [&NiagaraMeshRendererProperties]}
  {emitterName=GroundFlow, rendererClasses [&NiagaraMeshRendererProperties]}
  {emitterName=Smoke, simTarget=GPUComputeSim, rendererClasses [&NiagaraSpriteRendererProperties]}
  {
   emitterName=Rocks_OnSpire001
   simTarget=GPUComputeSim
   rendererClasses [&NiagaraMeshRendererProperties, &NiagaraSpriteRendererProperties]
  }
 ]
 userVariables [
  User.bIsFrontEnd {type=NiagaraBool}
  User.bIsFrontEndPreview {type=NiagaraBool}
  User.BlackHoleEnd {type=Vector3f, description="", defaultValue {struct=Vector3f, value=<0 50 -150>}}
  User.BlackHoleStart {type=NiagaraPosition, description="", defaultValue {}}
  User.ColorMult {type=NiagaraFloat, defaultValue=1, description=""}
  User.ExplosionScale {type=NiagaraFloat, defaultValue=1.2, description=""}
  User.Fade {type=NiagaraFloat, defaultValue=1, description=""}
  User.GlobalScale {type=NiagaraFloat, defaultValue=33, description=""}
  User.Intensity {type=NiagaraFloat, defaultValue=50, description=""}
  User.Lifetime {type=NiagaraFloat, defaultValue=3, description=""}
  User.Open {type=NiagaraFloat, defaultValue=2, description=""}
  User.OverallWidth {type=NiagaraFloat, defaultValue=3720.3586, description=""}
  User.ParaGlideLeanAlpha {type=NiagaraFloat, defaultValue=0, description=""}
  User.PlayerSpeed {type=NiagaraFloat, defaultValue=0, description=""}
  User.PortalSize {type=NiagaraFloat, defaultValue=1, description=""}
  User.RemoveBeam {type=NiagaraFloat, defaultValue=0, description=""}
  User.Scale {type=NiagaraFloat, defaultValue=1, description=""}
  User.ScaleMult {type=NiagaraFloat, defaultValue=5, description=""}
  User.SkeletalMesh {
   type=NiagaraDataInterfaceSkeletalMesh
   defaultValue {
    struct=NiagaraExt_VariableValue_DataInterface
    value {dataInterface=&self:NiagaraDataInterfaceSkeletalMesh_0}
   }
  }
  User.SplineUpdatedEndPos {type=Vector3f, description="", defaultValue {struct=Vector3f, value=<0 0 -25920.5664>}}
  User.SplineUpdatedStartPos {type=NiagaraPosition, description="", defaultValue {}}
  User.WidthMult {type=NiagaraFloat, defaultValue=1, description=""}
 ]
}
